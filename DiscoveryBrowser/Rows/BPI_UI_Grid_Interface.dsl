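; Blueprint interface for grid UI in the Discovery Browser rows.
; It declares one function, SetBladeExpandContractAnimationFinished, which takes a boolean saying whether a blade's expand/contract animation has finished.
; The file holds only the declaration, with no implementation.
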(fn SetBladeExpandContractAnimationFinished
  (params (bBladeExpandContractAnimationFinished :type Boolean)))
