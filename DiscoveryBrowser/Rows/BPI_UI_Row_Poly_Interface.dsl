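; Blueprint interface that defines shared calls for polymorphic rows in the Discovery Browser UI, with signatures only and no logic.
; Layout calls compute visibility from top-left and bottom-right bounds, update row padding and toggle the arrow background fade.
; Selection calls set the outer row index, pass in a nested button group selection view model, query subrow index data and navigate to the nearest neighbor subrow, accounting for a prior View All selection.
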(fn ComputeRowVisibilityBase
  (params (Top%20Left :type Vector%202D%20Structure) (Bottom%20Right :type Vector%202D%20Structure)))

(fn GetNewSubrowIndexData
  (return 0 0 0 0 0))

(fn NavigateToNearestNeighborSubrowIndex
  (params (Was%20Last%20Selection%20View%20All :type Boolean) (NearestNeighborIndex :type Integer)))

(fn SetArrowBackgroundFadeVisibility
  (params (isVisible :type Boolean)))

(fn SetOuterItemRowIndex
  (params (Index :type Integer)))

(fn UpdateNestedButtonGroupVM
  (params (NestedVM :type Fort%20Nested%20Button%20Group%20Selection%20VM%20Object%20Reference)))

(fn UpdateRowPadding)
