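# UI material for a Battle Royale row background in the Discovery Browser, drawing an animated wave band with a storm-tech pattern.
# A wave SDF function supplies glow, flow and falloff masks, while MF_UI_BR_StormTech supplies an animated grid pattern.
# These combine with edge masks and vertex alpha.
# Panning blurred noise and an oscillating color function tint the blue wave, composited with alpha-composite blending and an additive intensity control.
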Material /DiscoveryBrowser/Rows/Materials/M_UI_Row_Bg_BR
graph {
 properties {
  bAutomaticallySetUsageInEditor=false
  bUsedWithStaticMesh=false
  blendMode=BLEND_AlphaComposite
  materialDomain=MD_UI
  usedShadingModels=Default Lit
 }
 nodes [
  n1 = NamedRerouteDeclaration(None=MaterialFunctionCall() {materialFunction=&GetUserInterfaceUV}.Normalized UV) {name=UV, nodeColor=<rgba 0 0 0 1>}
  n2 = Panner(Coordinate=@n1) {speedX=-0.1}
  n3 = NamedRerouteDeclaration(None=TextureObject() {samplerType=SAMPLERTYPE_Grayscale, texture=&LowResBlurredNoise}) {name=Blur Noise, nodeColor=<rgba 0 0 0 1>}
  n4 = NamedRerouteDeclaration(None=MaterialFunctionCall(UV=@n1) {materialFunction=&MF_UI_UVScaling}.ScaleY) {name=Scale Y UV, nodeColor=<rgba 0 0 0 1>}
  Add(A=Multiply(A=TextureSample(UVs=@n2, Tex=@n3) {samplerSource=SSM_Wrap_WorldGroupSettings}.R) {constB=0.1}, B=@n4)
  Constant2Vector() {g=0.5, r=1}
  Constant2Vector() {g=0.06, r=0.03}
  Constant() {r=0.1}
  LinearInterpolate(
   A=ScalarParameter() {group=Wave, parameterName=Wave Min Thickness}
   B=ScalarParameter() {defaultValue=0.05, group=Wave, parameterName=Wave Max Thickness}
   Alpha=NamedRerouteUsage()
  ) {constB=0.05}
  n5 = NamedRerouteUsage()
  LinearInterpolate(Alpha=@n5) {constB=16}
  n6 = Constant2Vector() {g=0.5, r=0.25}
  LinearInterpolate(A=Multiply(A=@n6) {constB=0.5}, B=@n6, Alpha=@n5)
  LinearInterpolate(A=Constant() {r=0.25}, B=Constant() {r=0.5}, Alpha=NamedRerouteUsage()) {constA=0.125, constB=0.5}
  n7 = Constant()
  n8 = MaterialFunctionCall!(
   Screen UV
   Tiling=ScalarParameter() {defaultValue=16, group=Pattern, parameterName=Tiling}
   Rotation
   Main Min/Max Opacity=Constant()
   Main Min/Max Scale=Constant2Vector() {g=0.85, r=0.925}
   Main Min/Max Perspective=@n7
   Main Outline Thickness
   Main Speed=Constant() {r=0.05}
   Main Reduction=ScalarParameter() {group=Pattern, parameterName=Main Reduction}
   Sub Min/Max Opacity
   Sub Min/Max Scale
   Sub Min/Max Perspective=@n7
   Sub Outline Thickness
   Sub Speed=Constant() {r=0.025}
   Sub Reduction=ScalarParameter() {group=Pattern, parameterName=Sub Reduction}
   Random Seed=Constant() {r=0.01}
   Time=Add(A=Time(), B=@n5)
  ) {materialFunction=/FortUILibrary/MaterialFunctions/BR/MF_UI_BR_StormTech}
  Multiply(A=@n8.Sub Cells, B=@n8.Sub Cells)
  n9 = NamedRerouteUsage()
  n10 = ScalarParameter() {defaultValue=4, parameterName=Wave Flow Falloff}
  n11 = NamedRerouteUsage()
  n12 = Multiply(A=MaterialFunctionCall!(DistanceField=@n11, GlowMin=Constant(), GlowMax=Constant() {r=0.5}) {materialFunction=/FortUILibrary/MaterialFunctions/Shapes/Modifiers/MF_UI_SDFGlowOrAA}.Result)
  n13 = Multiply(
   A=Multiply(A=Power(Base=@n9, Exp=@n10) {constExponent=4}, B=ScalarParameter() {defaultValue=16, parameterName=Wave Flow Intensity}) {constB=16}
   B=@n12
  )
  n14 = Constant() {r=2}
  n15 = NamedRerouteUsage()
  n16 = Multiply(A=OneMinus(None=SmoothStep(Max=@n14, Value=@n15) {constMax=0.5})) {constB=0.1}
  n17 = SmoothStep(Min=@n14, Value=@n15) {constMax=0, constMin=1}
  n18 = Multiply(A=Multiply(A=@n17, B=@n17), B=@n17)
  n19 = LinearInterpolate(A=@n16, Alpha=Add(A=@n16, B=Multiply(A=@n18) {constB=0.25}))
  n20 = Constant() {r=0.3}
  n21 = ScalarParameter() {defaultValue=0.1, parameterName=Center Wave Opacity, sliderMax=1}
  n22 = Saturate(None=Add(A=Multiply(A=@n19, B=@n20) {constB=0.75}, B=Multiply(A=@n12, B=@n21) {constB=0.375}))
  n23 = NamedRerouteDeclaration(None=OneMinus(None=SmoothStep(Value=@n15) {constMax=2})) {name=Tech Wave Mask, nodeColor=<rgba 0 0 0 1>}
  n24 = Multiply(A=@n23, B=@n23)
  n25 = Multiply(A=NamedRerouteUsage(), B=ScalarParameter() {defaultValue=0.5, group=Pattern, parameterName=Pattern Opacity, sliderMax=1}) {constB=0.5}
  n26 = Add(
   A=Multiply(A=Add(A=@n13, B=@n22), B=ScalarParameter() {defaultValue=1, group=Wave, parameterName=Wave Opacity, sliderMax=1})
   B=Multiply(A=Multiply(A=@n24, B=@n25), B=NamedRerouteUsage())
  )
  n27 = Multiply(A=Multiply(A=Saturate(None=@n26), B=NamedRerouteUsage()), B=VertexColor().A)
  n28 = NamedRerouteUsage()
  Multiply(
   A=Multiply(A=@n27, B=LinearInterpolate(Alpha=@n28) {constA=0.25, constB=0}) {constB=0.25}
   B=ScalarParameter() {defaultValue=1, parameterName=Additive Intensity}
  )
  n29 = NamedRerouteUsage()
  n30 = OneMinus(None=@n29)
  n31 = SmoothStep(Value=Multiply(A=Min(A=@n29, B=@n30)) {constB=2})
  NamedRerouteDeclaration(None=ComponentMask(None=@n31) {r=true}) {name=Horizontal Mask, nodeColor=<rgba 0 0 0 1>}
  n32 = MaterialFunctionCall!(
   UV=@n4
   Wave Height=Constant() {r=0.5}
   Wave Elevation=Constant() {r=0.25}
   Wave Length=ScalarParameter() {defaultValue=1, group=Wave, parameterName=Wave Length}
   Wave Speed=ScalarParameter() {defaultValue=0.1, group=Wave, parameterName=Wave Speed}
   GlowMin
   GlowMax
   Outline Thickness=Constant() {r=0}
   Outline Alignment
   Custom Curve
  ) {materialFunction=/FortUILibrary/MaterialFunctions/Shapes/MF_UI_Wave}
  NamedRerouteDeclaration(None=@n32.Raw UV) {name=Wave UV, nodeColor=<rgba 0 0 0 1>}
  n33 = Constant3Vector() {constant=<rgba 0.0729 0.2019 1 1>}
  n34 = MaterialFunctionCall() {materialFunction=&GetUserInterfaceUV}
  n35 = NamedRerouteUsage()
  n36 = Constant() {r=1}
  n37 = NamedRerouteUsage()
  n38 = Constant2Vector() {g=1, r=0.1}
  n39 = TextureSample(UVs=Panner(Coordinate=Multiply(A=@n37, B=@n38)) {speedX=-0.1}, Tex=NamedRerouteUsage()) {samplerSource=SSM_Wrap_WorldGroupSettings}
  n40 = LinearInterpolate(
   A=LinearInterpolate(
    A=NamedRerouteDeclaration(None=@n33) {name=Wave Base Color, nodeColor=<rgba 0 0 0 1>}
    B=MaterialFunctionCall!(PrimaryUV=@n34.Normalized UV, SecondaryUV=@n35, UVMix=@n36, PixelSize) {materialFunction=/FortUILibrary/MaterialFunctions/CoreUI/MF_UI_CoreUI_OscillatingColor}.Result
   ) {constAlpha=0.25}
   B=NamedRerouteUsage()
   Alpha=Multiply(A=NamedRerouteUsage(), B=NamedRerouteDeclaration(None=@n39.R) {name=Wave Pattern, nodeColor=<rgba 0 0 0 1>})
  )
  n41 = Add(A=@n40) {constA=0.5, constB=0.5}
  n42 = NamedRerouteUsage()
  NamedRerouteDeclaration(None=LinearInterpolate(A=@n40, B=Saturate(None=@n41), Alpha=ComponentMask(None=@n42) {r=true}) {constA=1}) {name=Wave Color, nodeColor=<rgba 0 0 0 1>}
  NamedRerouteDeclaration(None=@n32.SDF Outline) {name=Wave SDF Outline, nodeColor=<rgba 0 0 0 1>}
  NamedRerouteDeclaration(None=ComponentMask(None=@n31) {g=true}) {name=Vertical Mask, nodeColor=<rgba 0 0 0 1>}
  NamedRerouteDeclaration(None=@n8.Grid) {name=Storm Tech Pattern Mask, nodeColor=<rgba 0 0 0 1>}
  NamedRerouteUsage()
  OneMinus(None=SmoothStep(Max=Constant() {r=0.25}, Value=@n11) {constMax=0.375})
  n43 = ScalarParameter() {defaultValue=0.5, parameterName=Color Intensity}
  Saturate(None=MaterialFunctionCall!(Base=Multiply(A=@n43, B=@n27), Add=NamedRerouteUsage()) {materialFunction=/Engine/Functions/Engine_MaterialFunctions02/Math/MultiplyAdd}.Result)
 ]
}
