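# UI material for the Discovery Browser's sponsored tag, drawing a tinted rounded-corner shape.
# A core UI rounded-mask function with a fixed pixel corner radius is multiplied by vertex alpha to form the base mask.
# That mask separately scales a dark grey Color parameter and an Additive scalar.
# Alpha-composite blending suits this premultiplied color, allowing a blend between additive and opaque looks.
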Material /DiscoveryBrowser/Rows/Materials/M_UI_Sponsored_Tag
graph {
 properties {
  bAutomaticallySetUsageInEditor=false
  bUsedWithStaticMesh=false
  blendMode=BLEND_AlphaComposite
  materialDomain=MD_UI
  usedShadingModels=Default Lit
 }
 nodes [
  n1 = Multiply(
   A=VertexColor().A
   B=MaterialFunctionCall(Corner Radius=Constant() {r=4}, Radius in Pixels=Constant() {r=1}) {materialFunction=&MF_UI_CoreUI_RoundedMask}.Shape Mask
  )
  Multiply(A=VectorParameter() {parameterName=Color, defaultValue=<rgba 0.125 0.125 0.125 1>}.RGB, B=@n1) {constA=1}
  Multiply(A=@n1, B=ScalarParameter() {defaultValue=0.25, parameterName=Additive, sliderMax=1})
 ]
}
