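# Common UI row header widget for the Discovery Browser, showing an uppercase title, a grey subheader and an optional Sponsored tag.
# A padded horizontal box holds the stacked texts beside a named slot for extra header content.
# Three setter functions fill the header text, toggle the tag between visible and collapsed, and set the subheader, collapsing it when empty.
# The event graph has only unwired event stubs.
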WidgetBlueprint /DiscoveryBrowser/Rows/WBP_Row_Header
graphs [
 EventGraph {nodes [
   Event PreConstruct()
   Event Construct()
   Event Tick()
  ]}
 SetHeaderText {nodes [
   n1 = SetHeaderText!() -> (then, HeaderText:Text) {is=title K2Node_FunctionEntry}
   > Set Text(Text=@n1, self=Get Text_HeaderRow() -> (Text_HeaderRow:Common Text Object Reference))
  ]}
 SetSponsoredIconVisibility {nodes [
   n1 = SetSponsoredIconVisibility!() -> (then, IsVisible:bool=~) {is=title K2Node_FunctionEntry}
   > Set Visibility(
    Visibility=Select(Option 0=Collapsed, Option 1=Visible, Index=@n1)
    self=Get WBP_Tag_Sponsored() -> (WBP_Tag_Sponsored:WBP Tag Sponsored Object Reference)
   )
  ]}
 SetSubHeaderText {nodes [
   n1 = SetSubHeaderText!() -> (then, SubHeaderText:Text) {is=title K2Node_FunctionEntry}
   > n2 = Set Text(Text=@n1, self=Get SubText_HeaderRow() -> (SubText_HeaderRow:Common Text Object Reference))
   > Set Visibility(Visibility=Select(Option 0=HitTestInvisible, Option 1=Collapsed, Index=TextIsEmpty(InText=@n2)), self=Get SubText_HeaderRow())
  ]}
]
variables []
widgets {
 info {inheritedWidgetCount=0, parentClass=&CommonUserWidget, rootWidgetClass=&GridPanel, widgetCount=8}
 widgets [
  GridRoot_ForHeaderPadding {widgetClassPath=&GridPanel, widget=GridRoot_ForHeaderPadding, properties {columnFill [1], rowFill [1]}}
  HB_Header {
   widgetClassPath=&HorizontalBox
   parent=GridRoot_ForHeaderPadding
   widget=HB_Header
   properties {flowDirectionPreference=Culture}
   slot {class=GridSlot, properties {padding=<ltrb 80 0 80 0>}}
  }
  StackBox_0 {
   widgetClassPath=&StackBox
   parent=HB_Header
   widget=StackBox_0
   properties {orientation=Orient_Vertical}
   slot {class=HorizontalBoxSlot}
  }
  Text_HeaderRow {
   widgetClassPath=&CommonTextBlock
   bIsVariable=true
   parent=StackBox_0
   widget=Text_HeaderRow
   properties {
    bAutoCollapseWithEmptyText=true
    clipping=OnDemand
    mobileFontSizeMultiplier=1.4
    scrollStyle=&TextScrollStyle-Base_C
    text="{Header Row}"
    textTransformPolicy=ToUpper
    font {
     bForceMonospaced=false
     bMaterialIsStencil=false
     fontMaterial=null
     fontObject=&HeadingNow
     letterSpacing=10
     monospacedWidth=1
     size=24
     skewAmount=0
     typefaceFontName=86 Bold
     outlineSettings {
      bApplyOutlineToDropShadows=false
      bMiteredCorners=false
      bSeparateFillAlpha=false
      outlineMaterial=null
      outlineSize=0
      outlineColor=<rgba 0 0 0 1>
     }
    }
    shadowOffset=<0 0>
   }
   slot {class=StackBoxSlot}
  }
  HB_SubtextContent {widgetClassPath=&HorizontalBox, parent=StackBox_0, widget=HB_SubtextContent, slot {class=StackBoxSlot}}
  WBP_Tag_Sponsored {
   widgetClassPath=&/WBP_Tag_Sponsored
   bIsVariable=true
   parent=HB_SubtextContent
   widget=WBP_Tag_Sponsored
   properties {visibility=Collapsed}
   slot {class=HorizontalBoxSlot, properties {horizontalAlignment=HAlign_Left, verticalAlignment=VAlign_Center, padding=<ltrb 0 0 8 0>}}
  }
  SubText_HeaderRow {
   widgetClassPath=&CommonTextBlock
   bIsVariable=true
   parent=HB_SubtextContent
   widget=SubText_HeaderRow
   properties {
    applyLineHeightToBottomLine=false
    bAutoCollapseWithEmptyText=true
    clipping=OnDemand
    mobileFontSizeMultiplier=1.6
    scrollStyle=&TextScrollStyle-Base_C
    text="{Sub Header}"
    textTransformPolicy=ToUpper
    colorAndOpacity {colorUseRule=UseColor_Specified, specifiedColor=<rgba 0.5776 0.5776 0.5776 1>}
    font {
     bForceMonospaced=false
     bMaterialIsStencil=false
     fontMaterial=null
     fontObject=&HeadingNow
     letterSpacing=10
     monospacedWidth=1
     size=24
     skewAmount=0
     typefaceFontName=86 Bold
     outlineSettings {
      bApplyOutlineToDropShadows=false
      bMiteredCorners=false
      bSeparateFillAlpha=false
      outlineMaterial=null
      outlineSize=0
      outlineColor=<rgba 0 0 0 1>
     }
    }
    shadowOffset=<0 0>
   }
   slot {class=HorizontalBoxSlot, properties {horizontalAlignment=HAlign_Left, verticalAlignment=VAlign_Center, size {sizeRule=Fill, value=1}}}
  }
  Slot_HeaderExtraContent {
   widgetClassPath=&NamedSlot
   bIsVariable=true
   parent=HB_Header
   widget=Slot_HeaderExtraContent
   slot {class=HorizontalBoxSlot, properties {padding=<ltrb 8 0 0 0>, size {sizeRule=Fill, value=1}}}
  }
 ]
}
