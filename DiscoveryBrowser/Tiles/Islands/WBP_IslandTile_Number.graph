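# Island tile variant for the Discovery browser that extends WBP_IslandTile_Full with a number derived from the tile's item index.
# Delegates bound on the base tile drive the number text and the container size override.
# The size override adds the number widget's spacer width to the dynamic tile width.
# Selection changes the outside-content slot's grid layer, toggles the number's outline material mask and plays a short scale-down animation, reversed on deselection.
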WidgetBlueprint /DiscoveryBrowser/Tiles/Islands/WBP_IslandTile_Number
animations [Anim_OnFocus {
  binding_count=1
  length_seconds=0.15
  object=&self:Anim_OnFocus
  sequence {bindings [WBP_Discovery_Number {
     id=#1
     possessed_class=WBP_Discovery_Number_C
     tracks [{
       type=MovieScene2DTransformTrack
       sections [{
         type=MovieScene2DTransformSection
         end_frame=3
         start_frame=0
         channels [
          Scale.X {type=MovieSceneScriptingFloatChannel, keys [[0,1],[3,0.9]]}
          Scale.Y {type=MovieSceneScriptingFloatChannel, keys [[0,1],[3,0.9]]}
         ]
        }]
      }]
    }]}
 }]
graphs [
 EventGraph {nodes [
   OnTileSizeSetOverrideContainerSize() {is=title K2Node_CustomEvent}
   n1 = Get WBP_Discover_Tile_Base() -> (WBP_Discover_Tile_Base:WBP Discover Tile Base Object Reference)
   n2 = BreakVector2D(InVec=Get Dynamic Tile Size!(self:WBP Discover Tile Base Object Reference=@n1) -> (DynamicTileSize:Vector 2D Structure) {is=title K2Node_CallFunction})
   n3 = Get WBP_Discovery_Number() -> (WBP_Discovery_Number:WBP Discovery Number Object Reference)
   n4 = BreakVector2D(InVec=Get Size!(self:Spacer Object Reference=Get SpacerWidth!(self:WBP Discovery Number Object Reference=@n3) -> (SpacerWidth:Spacer Object Reference)) -> (Size:Vector 2D Structure))
   > Set Tile Container Size Override!(
    execute
    self:WBP Discover Tile Base Object Reference=@n1
    TileSizeOverrideWithOutsideContent:Vector 2D Structure=MakeVector2D(X=float + float(A=@n2, B=@n4), Y=@n2.Y)
   ) -> (then) {is=title K2Node_CallFunction}
   OnSetItemIndex() {is=title K2Node_CustomEvent}
   n5 = Get TileItemIndex!(self:WBP Discover Tile Base Object Reference=Get WBP_Discover_Tile_Base()) -> (TileItemIndex:int=0)
   > "SetText (Text)"(
    self=Get Text_Number!(self:WBP Discovery Number Object Reference=Get WBP_Discovery_Number()) -> (Text_Number:Common Text Object Reference)
    InText="To Text (Integer)"(Value=int + int(A=@n5))
   )
   n6 = Event PreConstruct()
   > Parent: PreConstruct(IsDesignTime=@n6.IsDesignTime)
   Event Construct()
   > Parent: Construct()
   n7 = Create Event()
   n8 = Get WBP_Discover_Tile_Base()
   > Bind Event to On Set Tile Index(self:WBP Discover Tile Base Object Reference=@n8, Delegate=@n7) {is=title K2Node_AddDelegate}
   n9 = Get WBP_Discover_Tile_Base()
   > Bind Event to On Set Tile Container Size Override(self:WBP Discover Tile Base Object Reference=@n9, Delegate=Create Event()) {is=title K2Node_AddDelegate}
   Event Destruct()
   > Unbind Event from On Set Tile Index(self:WBP Discover Tile Base Object Reference=@n8, Delegate=@n7) {is=title K2Node_RemoveDelegate}
   > Unbind all Events from On Set Tile Container Size Override!(execute:Exec, self:WBP Discover Tile Base Object Reference=@n9) -> (then:Exec) {is=title K2Node_ClearDelegate}
   Get Anim_Focus() -> (Anim_Focus:Widget Animation Object Reference)
  ]}
 OnTileSelectionActive {nodes [
   OnTileSelectionActive() {is=title K2Node_FunctionEntry}
   > Parent: OnTileSelectionActive() {is=title K2Node_CallParentFunction}
   > Set Layer!(
    execute
    Layer:int=-40
    self:Grid Slot Object Reference=SlotAsGridSlot(Widget=Get Slot_OutsideContent() -> (Slot_OutsideContent:Named Slot Object Reference))
   ) -> (then, Output_Get:int=0)
   > n1 = GetDynamicOutlineMaterial!(execute, self:Text Object Reference=Get Text_Number(self=Get WBP_Discovery_Number())) -> (then, ReturnValue:Material Instance Dynamic Object Reference)
   > SetScalarParameterValue(self=@n1, ParameterName=Display Mask, Value=1)
   > PlayAnimationForward(InAnimation=Get Anim_OnFocus() -> (Anim_OnFocus:Widget Animation Object Reference))
  ]}
 OnTileSelectionNotActive {nodes [
   OnTileSelectionNotActive() {is=title K2Node_FunctionEntry}
   > Parent: OnTileSelectionNotActive() {is=title K2Node_CallParentFunction}
   > Set Layer(Layer=2, self=SlotAsGridSlot(Widget=Get Slot_OutsideContent())) -> (Output_Get=0)
   > n1 = GetDynamicOutlineMaterial(self=Get Text_Number(self=Get WBP_Discovery_Number()))
   > SetScalarParameterValue(self=@n1, ParameterName=Display Mask, Value=0)
   > PlayAnimationReverse(InAnimation=Get Anim_OnFocus())
  ]}
]
variables []
widgets {
 info {
  inheritedWidgetCount=0
  namedSlotCount=1
  parentClass=&/WBP_IslandTile_Full.WBP_IslandTile_Full_C
  rootWidgetClass=null
  widgetCount=1
 }
 widgets [WBP_Discovery_Number {
   widgetClassPath=&/WBP_Discovery_Number
   bIsVariable=true
   widget=WBP_Discovery_Number
   properties {renderTransformPivot=<1 0.5>}
  }]
}
