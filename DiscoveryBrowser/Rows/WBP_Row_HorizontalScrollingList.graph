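# Discovery browser row widget: a header, a horizontal polymorphic list of tiles, left/right page arrows and a loading-more indicator.
# An override selector or the item view model picks each tile class, and entries receive index, panel name and focus bindings that feed OnTileFocusGained.
# Arrows page by the count of fully on-screen tiles, measured from cached geometry, and toggle visibility as scrolling changes.
# While the row is selected, a longer item list is appended rather than reset.
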WidgetBlueprint /DiscoveryBrowser/Rows/WBP_Row_HorizontalScrollingList
graphs [
 EventGraph {nodes [
   n1 = "On Entry Initialized (ListView)"!() -> (OutputDelegate, then, Item:Object Reference, Widget:User Widget Object Reference) {is=title K2Node_ComponentBoundEvent}
   > n2 = Cast To WBP_IslandTile_Minimal!(execute, Object=@n1.Widget) -> (then, CastFailed, AsWBP Island Tile Minimal:WBP Island Tile Minimal Object Reference) {is=title K2Node_DynamicCast}
   n3 = GetIndexForItem!(self:List View Object Reference=Get ListView() -> (ListView:Common List View Object Reference), Item:Object Reference=@n1.Item) -> (ReturnValue:int=0)
   > UpdateBaseTileProperties!(
    execute
    self
    ItemIndex:int=@n3
    TileBaseWidget:WBP Discover Tile Base Object Reference=Get WBP_Discover_Tile_Base!(self:WBP Island Tile Minimal Object Reference=@n2) -> (WBP_Discover_Tile_Base:WBP Discover Tile Base Object Reference)
   ) -> (then)
   n4 = Get Is Using Focus for Outline Visuals() -> (Is Using Focus for Outline Visuals:bool)
   > Set IsUsingFocusForOutlineVisuals!(execute, IsUsingFocusForOutlineVisuals:bool=@n4, self:WBP Island Tile Minimal Object Reference=@n2) -> (then, Output_Get:bool)
   n5 = Cast To WBP_Discover_Tile_Button!(execute=@n2.CastFailed, Object=@n1.Widget) -> (then, CastFailed, AsWBP Discover Tile Button:WBP Discover Tile Button Object Reference) {is=title K2Node_DynamicCast}
   > UpdateBaseTileProperties(ItemIndex=@n3, TileBaseWidget=Get WBP_Discover_Tile_Base(self:WBP Discover Tile Button Object Reference=@n5))
   > Set IsTileUsingFocusForOutlineVisuals!(execute, IsTileUsingFocusForOutlineVisuals:bool=@n4, self:WBP Discover Tile Button Object Reference=@n5) -> (then, Output_Get:bool)
   "On Clicked (Button_PageRight)"!() -> (OutputDelegate, then, Button:Common Button Base Object Reference) {is=title K2Node_ComponentBoundEvent}
   > PaginateToNextOrPrevious!(execute, self, IsForward:bool=true) -> (then)
   n6 = "On Entry Generated (ListView)"!() -> (OutputDelegate, then, Widget:User Widget Object Reference) {is=title K2Node_ComponentBoundEvent}
   > SetArrowVisibility!(execute, self, ListViewScrollOffset:double=GetScrollOffset!(self:List View Base Object Reference=Get ListView()) -> (ReturnValue:float)) -> (then)
   > n7 = Cast To WBP_IslandTile_Minimal(Object=@n6.Widget)
   > Bind Event to On Focus Gained(self:WBP Island Tile Minimal Object Reference=@n7, Delegate=Create Event()) {is=title K2Node_AddDelegate}
   > Bind Event to On Context Menu Toggle(
    self:WBP Discover Tile Base Object Reference=Get WBP_Discover_Tile_Base(self:WBP Island Tile Minimal Object Reference=@n7)
    Delegate=Create Event()
   ) {is=title K2Node_AddDelegate}
   n8 = Cast To WBP_Discover_Tile_Button(execute=@n7.CastFailed, Object=@n6.Widget)
   > Bind Event to On Focus Gained(self:WBP Discover Tile Button Object Reference=@n8, Delegate=Create Event())
   > Bind Event to On Context Menu Toggle(self=Get WBP_Discover_Tile_Base(self:WBP Discover Tile Button Object Reference=@n8), Delegate=Create Event())
   n9 = "On List View Scrolled (ListView)"!() -> (OutputDelegate, then, ItemOffset:float, DistanceRemaining:float) {is=title K2Node_ComponentBoundEvent}
   > SetArrowVisibility(ListViewScrollOffset=@n9.ItemOffset)
   "On Clicked (Button_PageLeft)"!() -> (OutputDelegate, then, Button:Common Button Base Object Reference) {is=title K2Node_ComponentBoundEvent}
   > PaginateToNextOrPrevious(IsForward=false)
   n10 = OnItemFocused!() -> (OutputDelegate, then, FocusedItem:WBP Island Tile Minimal Object Reference) {is=title K2Node_CustomEvent}
   n11 = IsUsingTouch(PlayerController=GetOwningPlayer())
   > n12 = Branch(Condition=@n11)
   Call On Tile Focus Gained!(
    execute:Exec=@n12.else
    self:WBP Row Horizontal Scrolling List Object Reference
    ItemFocused:int=Get TileItemIndex!(self:WBP Discover Tile Base Object Reference=Get WBP_Discover_Tile_Base(self:WBP Island Tile Minimal Object Reference=@n10.FocusedItem)) -> (TileItemIndex:int=0)
   ) -> (then:Exec) {is=title K2Node_CallDelegate}
   n13 = OnTileButtonFocused!() -> (OutputDelegate, then, FocusedItem:WBP Discover Tile Button Object Reference) {is=title K2Node_CustomEvent}
   > n14 = Branch(Condition=@n11)
   Call On Tile Focus Gained(
    execute=@n14.else
    ItemFocused=Get TileItemIndex(self=Get WBP_Discover_Tile_Base(self:WBP Discover Tile Button Object Reference=@n13.FocusedItem)) -> (TileItemIndex=0)
   )
   n15 = OnContextMenuToggle!() -> (OutputDelegate, then, DiscoverTile:WBP Discover Tile Base Object Reference) {is=title K2Node_CustomEvent}
   > Call On Tile Focus Gained(ItemFocused=Get TileItemIndex(self=@n15.DiscoverTile) -> (TileItemIndex=0))
   n16 = Event PreConstruct()
   > Branch(Condition=@n16.IsDesignTime)
   > UpdateRowPadding!(execute, self, IsTopTenRow:bool=false) -> (then)
   Event Construct()
   > n17 = Branch Natural Nav Rollback Cvar!(In:Exec) -> (Use Natural Nav:Exec, Rollback:Exec) {is=title K2Node_MacroInstance}
   SetEnableProximateEntryNavigation!(execute=@n17.Rollback, self:List View Object Reference=Get ListView(), bNewValue:bool=false) -> (then)
  ]}
 GetDesiredEntryClassForItem {nodes [
   n1 = GetDesiredEntryClassForItem!() -> (then, Item:Object Reference) {is=title K2Node_FunctionEntry}
   > n2 = Cast To FortDiscoverTileItemVM!(execute, Object=@n1) -> (then, CastFailed, AsDiscover Tile Item View Model:Discover Tile Item View Model Object Reference) {is=title K2Node_DynamicCast}
   > n3 = Is Valid(InputObject=Get OverrideItemSelector() -> (OverrideItemSelector:Fort Discover Item Selector Base Object Reference))
   n4 = Branch(execute=@n3.Is Not Valid, Condition=GetIsBannerItem!(self:Discover Tile Item View Model Object Reference=@n2) -> (ReturnValue:bool))
   > Set TileFocusWeight(TileFocusWeight:double=1) -> (Output_Get:double)
   > Return Node!(
    execute
    ReturnValue:User Widget Class Reference=/DiscoveryBrowser/Tiles/Billboard/WBP_Discovery_Tile_Billboard.WBP_Discovery_Tile_Billboard_C
   )
   n5 = Byte to Enum EActivityItemType!(Byte:Byte=Get ItemType!(self:Discover Tile Item View Model Object Reference=@n2) -> (ItemType:Byte=0)) -> (ReturnValue:EActivityItemType Enum) {is=title K2Node_CastByteToEnum}
   n6 = Get bUseTopTenStyle!(self:Discover Tile Item View Model Object Reference=@n2) -> (bUseTopTenStyle:bool)
   n7 = Set TileFocusWeight(
    execute=@n4.else
    TileFocusWeight=Select(
     Option 0:double=Select!(
      Island:double=1
      Creator:double=0.5
      PlayingNow:double=1
      SurfaceTile:double=1.5
      IPTile:double=1.5
      PanelTile:double=1
      CommunityHighlight:double=1
      Index:EActivityItemType Enum=@n5
     ) -> (ReturnValue)
     Option 1:double=1.15
     Index=@n6
    ) -> (ReturnValue:double)
   )
   n8 = GetRuntimeOptions()
   n9 = Select(
    Island:User Widget Class Reference=/DiscoveryBrowser/Tiles/Islands/WBP_IslandTile_Full.WBP_IslandTile_Full_C
    Creator:User Widget Class Reference=/DiscoveryBrowser/Tiles/Creator/WBP_CreatorTile.WBP_CreatorTile_C
    PlayingNow:User Widget Class Reference=/DiscoveryBrowser/Tiles/PlayingNow/WBP_PlayingNow_Tile.WBP_PlayingNow_Tile_C
    SurfaceTile:User Widget Class Reference=/DiscoveryBrowser/Tiles/Categories/WBP_Tile_Category.WBP_Tile_Category_C
    IPTile:User Widget Class Reference=/DiscoveryBrowser/Tiles/Ip/WBP_Tile_IP_NEW.WBP_Tile_IP_NEW_C
    PanelTile:User Widget Class Reference=/DiscoveryBrowser/Tiles/Genre/WBP_Tile_Genre.WBP_Tile_Genre_C
    CommunityHighlight:User Widget Class Reference=Select!(
     Option 0:User Widget Class Reference=/DiscoveryBrowser/Tiles/Community/WBP_CommunityPost_TileV1.WBP_CommunityPost_TileV1_C
     Option 1:User Widget Class Reference=/DiscoveryBrowser/Tiles/Community/WBP_CommunityPost_Tile.WBP_CommunityPost_Tile_C
     Index:bool=Get bEnableCreatorEvents!(self:Fort Runtime Options Object Reference=@n8) -> (bEnableCreatorEvents:bool)
    ) -> (ReturnValue)
    Index=@n5
   ) -> (ReturnValue:User Widget Class Reference)
   Return Node(
    execute=[@n12.else, @n7.then, @n13.Is Not Valid]
    ReturnValue=Select(
     Option 0=Select(Option 0=@n9, Option 1=/DiscoveryBrowser/Tiles/Islands/WBP_IslandTile_Number.WBP_IslandTile_Number_C, Index=@n6)
     Option 1=&WBP_Tile_Genre_C
     Index=Get bDisplayAsMiniGenreTile!(self:Discover Tile Item View Model Object Reference=@n2) -> (bDisplayAsMiniGenreTile:bool)
    )
   )
   n10 = Get OverrideItemSelector()
   n11 = GetTileItemClass!(self:Fort Discover Item Selector Base Object Reference=@n10, InTileItemViewModel:Discover Tile Item View Model Object Reference=@n2) -> (ReturnValue:User Widget Class Reference)
   n12 = Branch(execute=[@n13.Is Valid, @n3.Is Valid], Condition=IsValidClass!(Class:Object Class Reference=@n11) -> (ReturnValue:bool))
   > Return Node(ReturnValue=@n11)
   n13 = Is Valid(exec=@n2.CastFailed, InputObject=@n10)
  ]}
 GetFirstVisibleItemIndex {nodes [
   GetFirstVisibleItemIndex() {is=title K2Node_FunctionEntry}
   n1 = GetScrollOffset(self=Get ListView())
   > Return Node!(
    execute
    FirstVisibleItemIndex:int="To Integer (Integer64)"!(InInt:Integer64=Floor to Integer64!(A:double=@n1) -> (ReturnValue:Integer64=0)) -> (ReturnValue:int=0)
   )
  ]}
 GetMaxScrollOffset {nodes [
   n1 = GetMaxScrollOffset!() -> (then, NumItemsPerPage:int) {is=title K2Node_FunctionEntry}
   n2 = Get ListView()
   n3 = float - float(A="To Float (Integer)"(InInt=GetNumItems!(self:List View Object Reference=@n2) -> (ReturnValue:int=0)), B:int=@n1)
   > Return Node!(execute, MaxItemOffset:double=SelectFloat(A=0, B=@n3, bPickA="float < float"(A=@n3)))
  ]}
 GetNumItemsPerPageNotClipped {nodes [
   GetNumItemsPerPageNotClipped() {is=title K2Node_FunctionEntry}
   n1 = GetCachedGeometry!(self:Widget Object Reference=Self-Reference()) -> (ReturnValue:"Geometry Structure (by ref)")
   n2 = GetAbsolutePosition!(Geometry:"Geometry Structure (by ref)"=@n1) -> (ReturnValue_X:double, ReturnValue_Y:double)
   > Set ListLeftX(ListLeftX:double=@n2) -> (Output_Get:double)
   > Set ListRightX(ListRightX:double=float + float(A=@n2, B=GetAbsoluteSize!(Geometry:"Geometry Structure (by ref)"=@n1) -> (ReturnValue_X:double, ReturnValue_Y:double))) -> (Output_Get:double)
   > n3 = Set L_ListDisplayedEntries(L_ListDisplayedEntries:Array of User Widget Object References=GetDisplayedEntryWidgets!(self:List View Base Object Reference=Get ListView()) -> (ReturnValue:Array of User Widget Object References)) -> (Output_Get:Array of User Widget Object References)
   > n4 = For Each Loop(Array:Array of User Widget Object References=@n3) -> (Array Element:User Widget Object Reference)
   n5 = GetCachedGeometry(self=@n4)
   n6 = GetAbsolutePosition(Geometry=@n5)
   > Set TileLeftX(TileLeftX:double=@n6) -> (Output_Get:double)
   > Set TileRightX(TileRightX:double=float + float(A=@n6, B=GetAbsoluteSize(Geometry=@n5))) -> (Output_Get:double)
   n7 = Get TileLeftX() -> (TileLeftX:double)
   n8 = Get ListLeftX() -> (ListLeftX:double)
   n9 = Get TileRightX() -> (TileRightX:double)
   n10 = Get ListRightX() -> (ListRightX:double)
   > Branch(Condition=AND Boolean(A="float >= float"(A=@n7, B=@n8), B="float <= float"(A=@n9, B=@n10) -> (ReturnValue=~)))
   > Increment Int(Value=Get TileIndex() -> (TileIndex:int=0))
   Return Node!(execute=@n4.Completed, NumFullyOnscreenItemsPerPage:int=Get TileIndex() -> (TileIndex=0))
  ]}
 IsLastItemFullyVisible {nodes [
   IsLastItemFullyVisible() {is=title K2Node_FunctionEntry}
   > n1 = GetNumItemsPerPageNotClipped!(execute, self) -> (then, NumFullyOnscreenItemsPerPage:int=0) {is=title K2Node_CallFunction}
   n2 = Get ListView()
   n3 = Ceil to Integer64!(A:double=GetScrollOffset(self=Get ListView())) -> (ReturnValue:Integer64=0)
   > Return Node!(
    execute
    IsLastTileVisible:bool="integer >= integer"(
     A=integer64 + integer64(A:int=@n1, B:Integer64=@n3) -> (ReturnValue:int) {is=title K2Node_PromotableOperator}
     B=GetNumItems(self=@n2) -> (ReturnValue=0)
    )
   )
  ]}
 OnTileFocusGained {nodes [OnTileFocusGained!() -> (then, ItemFocused:int) {is=title K2Node_FunctionEntry}]}
 PaginateToNextOrPrevious {nodes [
   PaginateToNextOrPrevious!() -> (then, IsForward:bool=~) {is=title K2Node_FunctionEntry}
   n1 = GetDisplayedEntryWidgets(self=Get ListView())
   > n2 = Branch(Condition="integer > integer"(A=Length(TargetArray:Array of User Widget Object References=@n1)))
   n3 = GetScrollOffset(self=Get ListView())
   > Set FirstVisibleIndex(FirstVisibleIndex:int="To Integer (Integer64)"(InInt=Ceil to Integer64(A=@n3) -> (ReturnValue=0)) -> (ReturnValue=0)) -> (Output_Get:int=0)
   > n4 = GetNumItemsPerPageNotClipped() -> (NumFullyOnscreenItemsPerPage=0)
   > n5 = GetMaxScrollOffset!(execute, self, NumItemsPerPage:int=@n4) -> (then, MaxItemOffset:double) {is=title K2Node_CallFunction}
   n6 = SelectInt(
    A=int + int(A=Get FirstVisibleIndex() -> (FirstVisibleIndex:int=0), B=@n4)
    B=int - int(A=Get FirstVisibleIndex() -> (FirstVisibleIndex=0), B=@n4)
    bPickA=Get IsForward() -> (IsForward:bool)
   )
   > SetScrollOffset!(
    execute
    self:List View Base Object Reference=Get ListView()
    InScrollOffset:float="Clamp (Float)"(Value="To Float (Integer)"(InInt=@n6), Min=0, Max=@n5)
   ) -> (then)
   n7 = Branch(execute=@n2.else, Condition=Get IsForward())
   n8 = Get ListView()
   > ScrollToBottom(self:List View Base Object Reference=@n8)
   ScrollToTop(execute=@n7.else, self:List View Base Object Reference=@n8)
  ]}
 Set List View Items {nodes [
   n1 = Set List View Items!() -> (then, ListItems:Array of Discover Tile Item View Model Object References) {is=title K2Node_FunctionEntry}
   n2 = Get IsRowSelected() -> (IsRowSelected:bool)
   n3 = GetNumItems(self=Get ListView()) -> (ReturnValue=0)
   > n4 = Branch(Condition=OR Boolean(A="Equal (Integer)"(A=@n3), B=NOT Boolean(A=@n2)))
   Set List Items!(execute=[@n4.then, @n7.else], self:List View Object Reference=Get ListView(), InListItems:Array of Object References=@n1) -> (then) {is=title K2Node_CallFunction}
   n5 = Length(TargetArray:Array of Discover Tile Item View Model Object References=@n1)
   n6 = GetNumItems(self=Get ListView()) -> (ReturnValue=0)
   n7 = Branch(execute=@n4.else, Condition="integer > integer"(A=@n5, B=@n6))
   > n8 = For Loop(FirstIndex=@n6, LastIndex=int - int(A=@n5))
   > AddItem!(
    execute
    self:List View Object Reference=Get ListView()
    Item:Object Reference="Get (a copy)"(Array:Array of Discover Tile Item View Model Object References=@n1, Dimension 1=@n8) -> (Output:Discover Tile Item View Model Object Reference)
   ) -> (then)
  ]}
 SetArrowBackgroundFadeVisibility {nodes [
   n1 = SetArrowBackgroundFadeVisibility!() -> (then, IsVisible:bool=~) {is=title K2Node_FunctionEntry}
   > SetUseArrowBackgroundFade!(
    execute
    self:Activity Browser Arrow Button Object Reference=[
     Get Button_PageLeft() -> (Button_PageLeft:Activity Browser Arrow Button Object Reference)
     Get Button_PageRight() -> (Button_PageRight:Activity Browser Arrow Button Object Reference)
    ]
    IsVisible:bool=@n1
   ) -> (then)
  ]}
 SetArrowVisibility {nodes [
   n1 = SetArrowVisibility!() -> (then, ListViewScrollOffset:double=~) {is=title K2Node_FunctionEntry}
   > SetVisibility(
    self=Get Button_PageLeft()
    InVisibility=Select(
     Option 0:ESlateVisibility Enum=SelfHitTestInvisible
     Option 1:ESlateVisibility Enum=Collapsed
     Index="float <= float"(A=@n1) -> (ReturnValue=~)
    ) -> (ReturnValue:ESlateVisibility Enum)
   )
   > n2 = IsLastItemFullyVisible!(execute, self) -> (then, IsLastTileVisible:bool) {is=title K2Node_CallFunction}
   > SetVisibility(self=Get Button_PageRight(), InVisibility=Select(Option 0=SelfHitTestInvisible, Option 1=Collapsed, Index=@n2))
  ]}
 SetLeftArrowPadding {nodes [
   SetLeftArrowPadding!() -> (then, IsTop10:bool=~) {is=title K2Node_FunctionEntry}
   > Set Padding!(
    execute
    Padding:Margin Structure=Select!(
     Option 0:Margin Structure
     Option 1_Left:float=72
     Option 1_Top:float=0
     Option 1_Right:float=0
     Option 1_Bottom:float=0
     Index=Get IsTop10() -> (IsTop10:bool)
    ) -> (ReturnValue:Margin Structure)
    self:Overlay Slot Object Reference=SlotAsOverlaySlot(Widget=Get FortVisualAttachment_265() -> (FortVisualAttachment_265:Fort Visual Attachment Object Reference))
   ) -> (then, Output_Get:Margin Structure)
  ]}
 SetLoadingMore {nodes [
   n1 = SetLoadingMore!() -> (then, LoadingState:EDiscoverLoadingState Enum=Unset) {is=title K2Node_FunctionEntry}
   > SetVisibility(
    self=Get WBP_LoadingMorePages() -> (WBP_LoadingMorePages:WBP Loading More Pages Object Reference)
    InVisibility=Select!(
     Option 0:ESlateVisibility Enum=Hidden
     Option 1:ESlateVisibility Enum=HitTestInvisible
     Index="Equal (Enum)"(A:EDiscoverLoadingState Enum=@n1, B:EDiscoverLoadingState Enum=Loading)
    ) -> (ReturnValue:ESlateVisibility Enum)
   )
  ]}
 SetRowPanelName {nodes [
   n1 = SetRowPanelName!() -> (then, PanelName:String) {is=title K2Node_FunctionEntry}
   > Set RowPanelName(RowPanelName:String=@n1) -> (Output_Get:String)
  ]}
 UpdateBaseTileProperties {nodes [
   n1 = UpdateBaseTileProperties!() -> (then, ItemIndex:int, TileBaseWidget:WBP Discover Tile Base Object Reference) {is=title K2Node_FunctionEntry}
   > SetTileIndex!(execute, self:WBP Discover Tile Base Object Reference=@n1.TileBaseWidget, TileIndex:int=@n1) -> (then)
   > SetParentPanelName!(execute, self:WBP Discover Tile Base Object Reference=@n1.TileBaseWidget, PanelName:String=Get RowPanelName() -> (RowPanelName:String)) -> (then)
  ]}
 UpdateRowPadding {nodes [
   n1 = UpdateRowPadding!() -> (then, IsTopTenRow:bool=~) {is=title K2Node_FunctionEntry}
   n2 = GetScalarParameterDefaultValue!(
    self:Material Parameter Collection Object Reference=Get DiscoverySharedValues() -> (DiscoverySharedValues:Material Parameter Collection Object Reference)
    ParameterName:Name=TopTenRowLeftPadding
   ) -> (bParameterFound:bool, ReturnValue:float)
   > Branch(Condition=@n2)
   n3 = SlotAsGridSlot(Widget=Get ItemsOverlay() -> (ItemsOverlay:Overlay Object Reference))
   n4 = Get!(self:Grid Slot Object Reference=@n3) -> (Padding_Left:float=~, Padding_Top:float=~, Padding_Right:float=~, Padding_Bottom:float=~)
   > SetPadding!(
    execute
    self:Grid Slot Object Reference=@n3
    InPadding_Left:float=SelectFloat(A=@n2.ReturnValue, B=68, bPickA=@n1)
    InPadding_Top:float=@n4.Padding_Top
    InPadding_Right:float=@n4.Padding_Right
    InPadding_Bottom:float=@n4.Padding_Bottom
   ) -> (then)
  ]}
]
variables [
 DiscoverySharedValues {type=MaterialParameterCollection, default=/DiscoveryBrowser/DataAssets/DiscoverySharedValues}
 Is Using Focus for Outline Visuals {type=boolean}
 IsRowSelected {type=boolean}
 MaxRowsCached {type=integer}
 OnTileFocusGained {}
 OuterItemRowIndex {type=integer}
 OverrideItemSelector {type=FortDiscoverItemSelectorBase, default=null}
 RowHasViewAll {type=boolean}
 RowPanelName {type=string, default=False}
 TileFocusWeight {type=number, default=1}
]
widgets {
 info {
  inheritedWidgetCount=1
  namedSlotCount=1
  parentClass=/Script/FortniteUI.FortPolymorphicListView
  rootWidgetClass=&GridPanel
  widgetCount=13
 }
 widgets [
  GridPanel_31 {widgetClassPath=&GridPanel, widget=GridPanel_31, properties {columnFill [1], rowFill [0,1]}}
  WBP_Row_Header {
   widgetClassPath=&/WBP_Row_Header
   bIsVariable=true
   parent=GridPanel_31
   widget=WBP_Row_Header
   slot {class=GridSlot, properties {rowSpan=1, verticalAlignment=VAlign_Top}}
  }
  Overlay {widgetClassPath=&Overlay, bIsVariable=true, namedSlotHost=WBP_Row_Header, widget=Overlay}
  FortVisualAttachment_LoadingMore {
   widgetClassPath=&FortVisualAttachment
   parent=Overlay
   widget=FortVisualAttachment_LoadingMore
   properties {contentAnchor=<1 1>}
   slot {class=OverlaySlot, properties {horizontalAlignment=HAlign_Right, verticalAlignment=VAlign_Bottom, padding=<ltrb 0 0 0 2>}}
  }
  WBP_LoadingMorePages {
   widgetClassPath=/DiscoveryBrowser/TabScreens/WBP_LoadingMorePages
   bIsVariable=true
   parent=FortVisualAttachment_LoadingMore
   widget=WBP_LoadingMorePages
   properties {visibility=Hidden}
   slot {class=SizeBoxSlot}
  }
  ItemsOverlay {
   widgetClassPath=&Overlay
   bIsVariable=true
   parent=GridPanel_31
   widget=ItemsOverlay
   slot {class=GridSlot, properties {layer=1, row=1, rowSpan=1, verticalAlignment=VAlign_Top, padding=<ltrb 68 16 68 0>}}
  }
  SizeBox_CapMaxRows {
   widgetClassPath=&SizeBox
   bIsVariable=true
   parent=ItemsOverlay
   widget=SizeBox_CapMaxRows
   properties {maxDesiredHeight=199}
   slot {class=OverlaySlot, properties {horizontalAlignment=HAlign_Fill, padding=<ltrb 0 0 0 48>}}
  }
  ListView {
   widgetClassPath=&CommonListView
   bInherited=true
   bIsVariable=true
   parent=SizeBox_CapMaxRows
   widget=ListView
   properties {
    bEnableProximateEntryNavigation=true
    clipping=Inherit
    entryWidgetClass=&WBP_IslandTile_Full_C
    horizontalEntrySpacing=24
    numDesignerPreviewEntries=20
    orientation=Orient_Horizontal
    verticalEntrySpacing=24
    visibility=SelfHitTestInvisible
    wheelScrollMultiplier=0
    scrollBarStyle {
     thickness=0
     draggedThumbImage {
      bIsDynamicallyLoaded=false
      drawAs=NoDrawType
      imageType=FullColor
      mirroring=NoMirror
      resourceName=../../../Engine/Content/Slate/Common/Scrollbar_Thumb.png
      resourceObject=null
      tiling=NoTile
      imageSize=<0 0>
      margin=<ltrb 0.25 0.25 0.25 0.25>
      outlineSettings {
       bUseBrushTransparency=false
       roundingType=HalfHeightRadius
       width=0
       color {colorUseRule=UseColor_Specified, specifiedColor=<rgba 0 0 0 0>}
       cornerRadii=<0 0 0 0>
      }
      tintColor {colorUseRule=UseColor_Specified, specifiedColor=<rgba 1 1 1 1>}
      uVRegion {bIsValid=false, max=<0 0>, min=<0 0>}
     }
     horizontalBackgroundImage {
      bIsDynamicallyLoaded=false
      drawAs=NoDrawType
      imageType=NoImage
      mirroring=NoMirror
      resourceName=None
      resourceObject=null
      tiling=NoTile
      imageSize=<0 0>
      margin=<ltrb 0 0 0 0>
      outlineSettings {
       bUseBrushTransparency=false
       roundingType=HalfHeightRadius
       width=0
       color {colorUseRule=UseColor_Specified, specifiedColor=<rgba 0 0 0 0>}
       cornerRadii=<0 0 0 0>
      }
      tintColor {colorUseRule=UseColor_Specified, specifiedColor=<rgba 1 1 1 1>}
      uVRegion {bIsValid=false, max=<0 0>, min=<0 0>}
     }
     horizontalBottomSlotImage {
      bIsDynamicallyLoaded=false
      drawAs=NoDrawType
      imageType=FullColor
      mirroring=NoMirror
      resourceName=../../../Engine/Content/Slate/Common/Scrollbar_Background_Horizontal.png
      resourceObject=null
      tiling=NoTile
      imageSize=<8 8>
      margin=<ltrb 0 0 0 0>
      outlineSettings {
       bUseBrushTransparency=false
       roundingType=HalfHeightRadius
       width=0
       color {colorUseRule=UseColor_Specified, specifiedColor=<rgba 0 0 0 0>}
       cornerRadii=<0 0 0 0>
      }
      tintColor {colorUseRule=UseColor_Specified, specifiedColor=<rgba 1 1 1 1>}
      uVRegion {bIsValid=false, max=<0 0>, min=<0 0>}
     }
     horizontalTopSlotImage {
      bIsDynamicallyLoaded=false
      drawAs=NoDrawType
      imageType=FullColor
      mirroring=NoMirror
      resourceName=../../../Engine/Content/Slate/Common/Scrollbar_Background_Horizontal.png
      resourceObject=null
      tiling=NoTile
      imageSize=<8 8>
      margin=<ltrb 0 0 0 0>
      outlineSettings {
       bUseBrushTransparency=false
       roundingType=HalfHeightRadius
       width=0
       color {colorUseRule=UseColor_Specified, specifiedColor=<rgba 0 0 0 0>}
       cornerRadii=<0 0 0 0>
      }
      tintColor {colorUseRule=UseColor_Specified, specifiedColor=<rgba 1 1 1 1>}
      uVRegion {bIsValid=false, max=<0 0>, min=<0 0>}
     }
     hoveredThumbImage {
      bIsDynamicallyLoaded=false
      drawAs=NoDrawType
      imageType=FullColor
      mirroring=NoMirror
      resourceName=../../../Engine/Content/Slate/Common/Scrollbar_Thumb.png
      resourceObject=null
      tiling=NoTile
      imageSize=<0 0>
      margin=<ltrb 0.25 0.25 0.25 0.25>
      outlineSettings {
       bUseBrushTransparency=false
       roundingType=HalfHeightRadius
       width=0
       color {colorUseRule=UseColor_Specified, specifiedColor=<rgba 0 0 0 0>}
       cornerRadii=<0 0 0 0>
      }
      tintColor {colorUseRule=UseColor_Specified, specifiedColor=<rgba 1 1 1 1>}
      uVRegion {bIsValid=false, max=<0 0>, min=<0 0>}
     }
     normalThumbImage {
      bIsDynamicallyLoaded=false
      drawAs=NoDrawType
      imageType=FullColor
      mirroring=NoMirror
      resourceName=../../../Engine/Content/Slate/Common/Scrollbar_Thumb.png
      resourceObject=null
      tiling=NoTile
      imageSize=<0 0>
      margin=<ltrb 0.25 0.25 0.25 0.25>
      outlineSettings {
       bUseBrushTransparency=false
       roundingType=HalfHeightRadius
       width=0
       color {colorUseRule=UseColor_Specified, specifiedColor=<rgba 0 0 0 0>}
       cornerRadii=<0 0 0 0>
      }
      tintColor {colorUseRule=UseColor_Specified, specifiedColor=<rgba 1 1 1 1>}
      uVRegion {bIsValid=false, max=<0 0>, min=<0 0>}
     }
     verticalBackgroundImage {
      bIsDynamicallyLoaded=false
      drawAs=NoDrawType
      imageType=NoImage
      mirroring=NoMirror
      resourceName=None
      resourceObject=null
      tiling=NoTile
      imageSize=<0 0>
      margin=<ltrb 0 0 0 0>
      outlineSettings {
       bUseBrushTransparency=false
       roundingType=HalfHeightRadius
       width=0
       color {colorUseRule=UseColor_Specified, specifiedColor=<rgba 0 0 0 0>}
       cornerRadii=<0 0 0 0>
      }
      tintColor {colorUseRule=UseColor_Specified, specifiedColor=<rgba 1 1 1 1>}
      uVRegion {bIsValid=false, max=<0 0>, min=<0 0>}
     }
     verticalBottomSlotImage {
      bIsDynamicallyLoaded=false
      drawAs=NoDrawType
      imageType=FullColor
      mirroring=NoMirror
      resourceName=../../../Engine/Content/Slate/Common/Scrollbar_Background_Vertical.png
      resourceObject=null
      tiling=NoTile
      imageSize=<8 8>
      margin=<ltrb 0 0 0 0>
      outlineSettings {
       bUseBrushTransparency=false
       roundingType=HalfHeightRadius
       width=0
       color {colorUseRule=UseColor_Specified, specifiedColor=<rgba 0 0 0 0>}
       cornerRadii=<0 0 0 0>
      }
      tintColor {colorUseRule=UseColor_Specified, specifiedColor=<rgba 1 1 1 1>}
      uVRegion {bIsValid=false, max=<0 0>, min=<0 0>}
     }
     verticalTopSlotImage {
      bIsDynamicallyLoaded=false
      drawAs=NoDrawType
      imageType=FullColor
      mirroring=NoMirror
      resourceName=../../../Engine/Content/Slate/Common/Scrollbar_Background_Vertical.png
      resourceObject=null
      tiling=NoTile
      imageSize=<8 8>
      margin=<ltrb 0 0 0 0>
      outlineSettings {
       bUseBrushTransparency=false
       roundingType=HalfHeightRadius
       width=0
       color {colorUseRule=UseColor_Specified, specifiedColor=<rgba 0 0 0 0>}
       cornerRadii=<0 0 0 0>
      }
      tintColor {colorUseRule=UseColor_Specified, specifiedColor=<rgba 1 1 1 1>}
      uVRegion {bIsValid=false, max=<0 0>, min=<0 0>}
     }
    }
   }
   slot {class=SizeBoxSlot, properties {verticalAlignment=VAlign_Top}}
  }
  Overlay_ArrowButtons {
   widgetClassPath=&Overlay
   parent=ItemsOverlay
   widget=Overlay_ArrowButtons
   slot {class=OverlaySlot, properties {horizontalAlignment=HAlign_Fill, verticalAlignment=VAlign_Center, padding=<ltrb 0 0 16 24>}}
  }
  FortVisualAttachment_265 {
   widgetClassPath=&FortVisualAttachment
   bIsVariable=true
   parent=Overlay_ArrowButtons
   widget=FortVisualAttachment_265
   properties {contentAnchor=<0.5 0.5>}
   slot {class=OverlaySlot, properties {verticalAlignment=VAlign_Center}}
  }
  Button_PageLeft {
   widgetClassPath=/DiscoveryBrowser/ActivityBrowserArrowButton
   bIsVariable=true
   parent=FortVisualAttachment_265
   widget=Button_PageLeft
   properties {
    flip=true
    onButtonBaseClicked=(&self:WidgetTree.Button_PageLeft.HandleOnButtonClicked)
    onButtonBaseDoubleClicked=(&self:WidgetTree.Button_PageLeft.HandleOnButtonDoubleClicked)
    onButtonBaseHovered=(&self:WidgetTree.Button_PageLeft.HandleOnButtonHovered)
    onButtonBaseUnhovered=(&self:WidgetTree.Button_PageLeft.HandleOnButtonUnhovered)
    onSelectedChangedBase=(&self:WidgetTree.Button_PageLeft.HandleOnSelectedChanged)
   }
   slot {class=SizeBoxSlot}
  }
  FortVisualAttachment_190 {
   widgetClassPath=&FortVisualAttachment
   parent=Overlay_ArrowButtons
   widget=FortVisualAttachment_190
   properties {contentAnchor=<0.5 0.5>}
   slot {class=OverlaySlot, properties {horizontalAlignment=HAlign_Right, verticalAlignment=VAlign_Center}}
  }
  Button_PageRight {
   widgetClassPath=&ActivityBrowserArrowButton
   bIsVariable=true
   parent=FortVisualAttachment_190
   widget=Button_PageRight
   properties {
    onButtonBaseClicked=(&self:WidgetTree.Button_PageRight.HandleOnButtonClicked)
    onButtonBaseDoubleClicked=(&self:WidgetTree.Button_PageRight.HandleOnButtonDoubleClicked)
    onButtonBaseHovered=(&self:WidgetTree.Button_PageRight.HandleOnButtonHovered)
    onButtonBaseUnhovered=(&self:WidgetTree.Button_PageRight.HandleOnButtonUnhovered)
    onSelectedChangedBase=(&self:WidgetTree.Button_PageRight.HandleOnSelectedChanged)
   }
   slot {class=SizeBoxSlot}
  }
 ]
}
