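# Standalone Discovery browser row: an activatable list entry that wraps a horizontal scrolling list and takes a Discover provider view model.
# Initializing entries triggers next-page fetches, and generated tiles receive the nested button selection view model.
# OnPaint computes row visibility from the list's screen rect; releasing the row stops the visible impression behind a console variable.
# Focus path changes toggle activation, and touch input clears stale selection.
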WidgetBlueprint /DiscoveryBrowser/Rows/WBP_Row_HorizontalList_Standalone
graphs [
 BP_GetDesiredFocusTarget {nodes [
   Get Desired Focus Target()
   > Return Node!(
    execute
    ReturnValue:Widget Object Reference=Get HorizontalScrollingList() -> (HorizontalScrollingList:WBP Row Horizontal Scrolling List Object Reference)
   )
  ]}
 EventGraph {nodes [
   n1 = OnEntryInit!() -> (OutputDelegate, then, Item:Object Reference, Widget:User Widget Object Reference) {is=title K2Node_CustomEvent}
   n2 = Get HorizontalScrollingList()
   > CheckAndFetchNextPage!(
    execute
    self:Discover Provider View Model Object Reference=Get FortDiscoverProviderViewModel() -> (FortDiscoverProviderViewModel:Discover Provider View Model Object Reference)
    SelectionIndex:int=GetIndexForItem!(
     self:List View Object Reference=Get ListView!(self:Fort Polymorphic List View Object Reference=@n2) -> (ListView:Common List View Object Reference)
     Item:Object Reference=@n1.Widget
    ) -> (ReturnValue:int=0)
    bForceFetchNextPage:bool=false
   ) -> (then)
   n3 = HandleInputMethodChanged!() -> (OutputDelegate, then, InputType:ECommonInputType Enum) {is=title K2Node_CustomEvent}
   > n4 = Switch on ECommonInputType!(execute, Selection:ECommonInputType Enum=@n3.InputType) -> (MouseAndKeyboard:Exec, Gamepad:Exec, Touch:Exec, Count:Exec) {is=title K2Node_SwitchEnum}
   Clear Selection(execute=@n4.Touch, self:List View Object Reference=Get ListView(self=Get HorizontalScrollingList()))
   n5 = OnTileItemGenerated!() -> (OutputDelegate, then, Widget:User Widget Object Reference) {is=title K2Node_CustomEvent}
   > SetNestedButtonGroupSelectionVM!(
    execute:Exec
    self:Object Reference=@n5.Widget
    InVM:Fort Nested Button Group Selection VM Object Reference=Get NestedButtonGroupSelectionVM() -> (NestedButtonGroupSelectionVM:Fort Nested Button Group Selection VM Object Reference)
   ) -> (then:Exec) {is=title K2Node_Message}
   Event OnAddedToFocusPath!() -> (OutputDelegate, then, InFocusEvent:Focus Event Structure) {is=title K2Node_Event}
   > ActivateWidget()
   n6 = Event On Item Selection Changed!() -> (OutputDelegate, then, bIsSelected:bool) {is=title K2Node_Event}
   > Set IsRowSelected!(execute, IsRowSelected:bool=@n6.bIsSelected, self:WBP Row Horizontal Scrolling List Object Reference=Get HorizontalScrollingList()) -> (then, Output_Get:bool)
   n7 = Event OnListItemObjectSet()
   > n8 = Cast To FortDiscoverProviderViewModel!(execute, Object=@n7.ListItemObject) -> (then, CastFailed, AsDiscover Provider View Model:Discover Provider View Model Object Reference) {is=title K2Node_DynamicCast}
   > Set FortDiscoverProviderViewModel(FortDiscoverProviderViewModel:Discover Provider View Model Object Reference=@n8) -> (Output_Get:Discover Provider View Model Object Reference)
   Event Destruct()
   n9 = Get ListView(self=Get HorizontalScrollingList())
   > Unbind all Events from On Entry Initialized!(execute:Exec, self:List View Object Reference=@n9) -> (then:Exec) {is=title K2Node_ClearDelegate}
   n10 = Create Event()
   n11 = Get CommonInputSubsystem()
   > Unbind Event from On Input Method Changed(self=@n11, Delegate=@n10)
   n12 = Create Event()
   n13 = Get ListView(self=Get HorizontalScrollingList())
   > Unbind Event from On Entry Generated(self:List View Base Object Reference=@n13, Delegate=@n12) {is=title K2Node_RemoveDelegate}
   Event OnRemovedFromFocusPath!() -> (OutputDelegate, then, InFocusEvent:Focus Event Structure) {is=title K2Node_Event}
   > DeactivateWidget()
   Event Construct()
   > Bind Event to On Entry Initialized(self:List View Object Reference=@n9, Delegate=Create Event()) {is=title K2Node_AddDelegate}
   > Bind Event to On Input Method Changed(self=@n11, Delegate=@n10)
   > Bind Event to On Entry Generated(self:List View Base Object Reference=@n13, Delegate=@n12) {is=title K2Node_AddDelegate}
   > SetDesiredFocusWidget!(execute, self:User Widget Object Reference, Widget:Widget Object Reference=Get HorizontalScrollingList()) -> (then, ReturnValue:bool)
   Event On Entry Released() {is=title K2Node_Event}
   > StopVisibilityOnEntryReleased()
   > n14 = Cast To UserWidget!(
    execute
    Object:Fort Nested Button Group VMItem Interface=Get CurrentSelectedButton!(self:Fort Nested Button Group Selection VM Object Reference=Get NestedButtonGroupSelectionVM()) -> (CurrentSelectedButton:Fort Nested Button Group VMItem Interface)
   ) -> (then, CastFailed, AsUser Widget:User Widget Object Reference) {is=title K2Node_DynamicCast}
   > Is Valid(InputObject=Get NestedButtonGroupSelectionVM())
   n15 = GetDisplayedEntryWidgets!(self:List View Base Object Reference=Get ListView(self=Get HorizontalScrollingList())) -> (ReturnValue:Array of User Widget Object References)
   n16 = GetOwningPlayer()
   > Branch(Condition=AND Boolean(
     A=IsUsingTouch(PlayerController=@n16)
     B=Contains Item(TargetArray:Array of User Widget Object References=@n15, ItemToFind:"User Widget Object Reference (by ref)"=@n14)
    ))
   n17 = Get NestedButtonGroupSelectionVM()
   > SetIsSelected!(
    execute
    self:Fort Nested Button Group Selection VM Object Reference=@n17
    InButton:Fort Nested Button Group VMItem Interface=Get CurrentSelectedButton(self=@n17)
    bInSelected:bool=false
   ) -> (then)
  ]}
 OnPaint {nodes [
   OnPaint!() -> (then, Context:"Paint Context Structure (by ref)") {is=title K2Node_FunctionEntry}
   n1 = Get FortDiscoverProviderViewModel()
   > Is Valid(InputObject=@n1)
   n2 = Get HorizontalScrollingList()
   > n3 = GetWidgetAbsoluteRect!(execute, Widget:Widget Object Reference=@n2) -> (then, TopLeft:Vector 2D Structure, BottomRight:Vector 2D Structure, Size:Vector 2D Structure, DebugStringOutput:String)
   > ComputeRowVisibility!(
    execute
    ListView:List View Object Reference=Get ListView(self=@n2)
    Provider VM:Discover Provider View Model Object Reference=@n1
    TopLeft:Vector 2D Structure=@n3
    BottomRight:Vector 2D Structure=@n3.BottomRight
   ) -> (then)
   > Return Node!(execute)
  ]}
 StopVisibilityOnEntryReleased {nodes [
   StopVisibilityOnEntryReleased() {is=title K2Node_FunctionEntry}
   > n1 = GetConsoleVariableBoolValue!(execute, VariableName:String=Athena.DiscoverSurface.RowStopVisibilityOnEntryReleased) -> (then, ReturnValue:bool)
   > Branch(Condition=@n1)
   > StopVisibleImpression(self:Discover Provider View Model Object Reference=Get FortDiscoverProviderViewModel())
  ]}
]
variables [NestedButtonGroupSelectionVM {type=FortNestedButtonGroupSelectionVM, default=null}]
widgets {
 info {
  inheritedWidgetCount=0
  parentClass=&CommonActivatableWidget
  rootWidgetClass=&/WBP_Row_HorizontalScrollingList.WBP_Row_HorizontalScrollingList_C
  widgetCount=1
 }
 widgets [HorizontalScrollingList {
   widgetClassPath=&/WBP_Row_HorizontalScrollingList
   bIsVariable=true
   widget=HorizontalScrollingList
   properties {bIsFocusable=true}
  }]
}
