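# Invisible CommonUI button for Discovery browser tiles, apparently a hover target that emulates row hover, judging by its name.
# Derives from CommonButtonBase, with a single root Image that has no resource and a zero-alpha tint, so it draws nothing.
# The event graph holds only unconnected PreConstruct, Construct and Tick events, so any hover response comes from outside this widget's graph.
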WidgetBlueprint /DiscoveryBrowser/Tiles/WBP_Discover_ButtonEmulateRowHover
graphs [EventGraph {nodes [
   Event PreConstruct()
   Event Construct()
   Event Tick()
  ]}]
variables []
widgets {
 info {inheritedWidgetCount=1, parentClass=&CommonButtonBase, rootWidgetClass=&Image, widgetCount=2}
 widgets [
  Image_26 {
   widgetClassPath=&Image
   bIsVariable=true
   widget=Image_26
   properties {brush {
     bIsDynamicallyLoaded=false
     drawAs=Image
     imageType=NoImage
     mirroring=NoMirror
     resourceName=None
     resourceObject=null
     tiling=NoTile
     imageSize=<32 32>
     margin=<ltrb 0 0 0 0>
     outlineSettings {
      bUseBrushTransparency=false
      roundingType=HalfHeightRadius
      width=0
      color {colorUseRule=UseColor_Specified, specifiedColor=<rgba 0 0 0 0>}
      cornerRadii=<0 0 0 0>
     }
     tintColor {colorUseRule=UseColor_Specified, specifiedColor=<rgba 1 1 1 0>}
     uVRegion {bIsValid=false, max=<0 0>, min=<0 0>}
    }}
  }
  InputActionWidget {widgetClassPath=&CommonActionWidget, bInherited=true, widget=null}
 ]
}
