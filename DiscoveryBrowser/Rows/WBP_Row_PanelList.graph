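# Discovery browser row widget: a header above a non-scrolling, row-capped tile view with a View All button.
# Selects each entry's class from an item selector or its view model's type, then sizes entries from tile size or forced column count.
# View All shows when items overflow and input is touch, or the row is hovered or button selected.
# Tile and button focus are relayed through dispatchers; OpenViewAllScreen pushes WBP_ViewAllPage with the provider view model.
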WidgetBlueprint /DiscoveryBrowser/Rows/WBP_Row_PanelList
graphs [
 EventGraph {nodes [
   n1 = "On Entry Initialized (TileView)"!() -> (OutputDelegate, then, Item:Object Reference, Widget:User Widget Object Reference) {is=title K2Node_ComponentBoundEvent}
   > n2 = Cast To WBP_IslandTile_Minimal!(execute, Object=@n1.Widget) -> (then, CastFailed, AsWBP Island Tile Minimal:WBP Island Tile Minimal Object Reference) {is=title K2Node_DynamicCast}
   n3 = GetIndexForItem!(self:List View Object Reference=Get TileView() -> (TileView:Fort Tile View Object Reference), Item:Object Reference=@n1.Item) -> (ReturnValue:int=0)
   > UpdateBaseTileProperties!(
    execute
    self
    ItemIndex:int=@n3
    TileBaseWidget:WBP Discover Tile Base Object Reference=Get WBP_Discover_Tile_Base!(self:WBP Island Tile Minimal Object Reference=@n2) -> (WBP_Discover_Tile_Base:WBP Discover Tile Base Object Reference)
   ) -> (then)
   n4 = Get UseFocusToTriggerOutlineVisuals() -> (UseFocusToTriggerOutlineVisuals:bool)
   > Set IsUsingFocusForOutlineVisuals!(execute, IsUsingFocusForOutlineVisuals:bool=@n4, self:WBP Island Tile Minimal Object Reference=@n2) -> (then, Output_Get:bool)
   n5 = Cast To WBP_Discover_Tile_Button!(execute=@n2.CastFailed, Object=@n1.Widget) -> (then, CastFailed, AsWBP Discover Tile Button:WBP Discover Tile Button Object Reference) {is=title K2Node_DynamicCast}
   > UpdateBaseTileProperties(ItemIndex=@n3, TileBaseWidget=Get WBP_Discover_Tile_Base(self:WBP Discover Tile Button Object Reference=@n5))
   > Set IsTileUsingFocusForOutlineVisuals!(execute, IsTileUsingFocusForOutlineVisuals:bool=@n4, self:WBP Discover Tile Button Object Reference=@n5) -> (then, Output_Get:bool)
   "On List View Scrolled (TileView)"!() -> (OutputDelegate, then, ItemOffset:float, DistanceRemaining:float) {is=title K2Node_ComponentBoundEvent}
   > SetScrollOffset!(execute, self:List View Base Object Reference=Get TileView(), InScrollOffset:float=0) -> (then)
   n6 = "On Entry Generated (TileView)"!() -> (OutputDelegate, then, Widget:User Widget Object Reference) {is=title K2Node_ComponentBoundEvent}
   > n7 = Cast To WBP_IslandTile_Minimal(Object=@n6.Widget)
   > Set HandlesOwnStartVisibilityImpression!(
    execute
    HandlesOwnStartVisibilityImpression:bool=Get TilesHandleTheirOwnStartVisibilityImpression() -> (TilesHandleTheirOwnStartVisibilityImpression:bool)
    self:WBP Island Tile Minimal Object Reference=@n7
   ) -> (then, Output_Get:bool)
   n8 = Get FortActivityViewModel!(self:WBP Island Tile Minimal Object Reference=@n7) -> (FortActivityViewModel:Discover Activity View Model Object Reference)
   > n9 = Branch(Condition=IsValid(Object=@n8))
   > UpdateAnalyticsCategoryVisibleRowsAndItemsPerRow!(
    execute
    self:Discover Tile Item View Model Object Reference=@n8
    InCategoryVisibleRows:int=GetNumDisplayedRows!() -> (DisplayedRows:int=0)
    InCategoryVisibleItemsPerRow:int=GetNumberItemsPerRow!(self:Fort Tile View Object Reference=Get TileView()) -> (ReturnValue:int=0)
   ) -> (then)
   > n10 = StopVisibilityImpression(self:Discover Tile Item View Model Object Reference=@n8)
   Bind Event to On Focus Gained(execute=[@n10.then, @n9.else], self:WBP Island Tile Minimal Object Reference=@n7, Delegate=Create Event()) {is=title K2Node_AddDelegate}
   n11 = Get WBP_Discover_Tile_Base(self:WBP Island Tile Minimal Object Reference=@n7)
   > UpdateTileViewLayout!(execute, self, TileBaseWidget:WBP Discover Tile Base Object Reference=@n11) -> (then)
   > Bind Event to On Context Menu Toggle(self:WBP Discover Tile Base Object Reference=@n11, Delegate=Create Event()) {is=title K2Node_AddDelegate}
   n12 = Cast To WBP_Discover_Tile_Button(execute=@n7.CastFailed, Object=@n6.Widget)
   > Bind Event to On Focus Gained(self:WBP Discover Tile Button Object Reference=@n12, Delegate=Create Event())
   n13 = Get WBP_Discover_Tile_Base(self:WBP Discover Tile Button Object Reference=@n12)
   > UpdateTileViewLayout(TileBaseWidget=@n13)
   > Bind Event to On Context Menu Toggle(self=@n13, Delegate=Create Event())
   "On Clicked (Button_ViewAll_PaddingHalfOfTileEntrySpacing)"!() -> (OutputDelegate, then, Button:Common Button Base Object Reference) {is=title K2Node_ComponentBoundEvent}
   > Call On View All Send Provider View Model!(execute:Exec, self:WBP Row Panel List Object Reference) -> (then:Exec) {is=title K2Node_CallDelegate}
   "On Entries Generated (TileView)"!() -> (OutputDelegate, then, NumEntries:int=0) {is=title K2Node_ComponentBoundEvent}
   > UpdateHasViewAll()
   "On Entry Released (TileView)"!() -> (OutputDelegate, then, Widget:User Widget Object Reference) {is=title K2Node_ComponentBoundEvent}
   > UpdateHasViewAll()
   "On Focused (Button_ViewAll_PaddingHalfOfTileEntrySpacing)"!() -> (OutputDelegate, then, Button:Common Button Base Object Reference) {is=title K2Node_ComponentBoundEvent}
   n14 = Get TileView()
   > Set LastSelectedIndex(LastSelectedIndex:int=GetIndexForItem(self=@n14, Item=Get Selected Item!(self:List View Object Reference=@n14) -> (ReturnValue:Object Reference) {is=title K2Node_CallFunction}) -> (ReturnValue=0)) -> (Output_Get:int=0)
   n15 = OpenViewAllScreen!() -> (OutputDelegate, then, ProviderViewModel:Discover Provider View Model Object Reference) {is=title K2Node_CustomEvent}
   > n16 = PushPrimaryContentClassAsync!(execute:Exec, WidgetClass:Common Activatable Widget Class Reference=/DiscoveryBrowser/WBP_ViewAllPage.WBP_ViewAllPage_C) -> (then:Exec, OnPrimaryContentPushed:Exec, ActivatableWidget:Common Activatable Widget Object Reference) {is=title K2Node_AsyncAction}
   n17 = GetCurrentInputType(self=Get CommonInputSubsystem())
   > Set IsViewAllSelected(IsViewAllSelected:bool=NOT Boolean(A="Equal (Enum)"(A=@n17, B=MouseAndKeyboard))) -> (Output_Get:bool)
   n18 = Cast To WBP_ViewAllPage!(execute=@n16.OnPrimaryContentPushed, Object=@n16) -> (then, CastFailed, AsWBP View All Page:WBP View All Page Object Reference) {is=title K2Node_DynamicCast}
   > SetFortDiscoverProviderViewModel!(execute, self:WBP View All Page Object Reference=@n18, Viewmodel:Discover Provider View Model Object Reference=@n15.ProviderViewModel) -> (then)
   n19 = Get FortActivitySelector() -> (ReturnValue:Fort Activity Selector Object Reference) {is=title K2Node_GetSubsystem}
   > Branch(Condition=IsShowingPlayerPage!(self:Fort Activity Selector Object Reference=@n19) -> (ReturnValue:bool))
   > SetPlayerPageViewAllButtonClicked!(execute, self:Fort Activity Selector Object Reference=@n19, bInIsButtonClicked:bool=true) -> (then)
   n20 = OnItemFocused!() -> (OutputDelegate, then, FocusedItem:WBP Island Tile Minimal Object Reference) {is=title K2Node_CustomEvent}
   n21 = IsUsingTouch(PlayerController=GetOwningPlayer())
   > n22 = Branch(Condition=@n21)
   Set IsViewAllSelected(execute=@n22.else, IsViewAllSelected=false)
   > Call On Tile Focus Gained!(
    execute:Exec
    self:WBP Row Panel List Object Reference
    ItemFocused:int=Get TileItemIndex!(self:WBP Discover Tile Base Object Reference=Get WBP_Discover_Tile_Base(self:WBP Island Tile Minimal Object Reference=@n20.FocusedItem)) -> (TileItemIndex:int=0)
   ) -> (then:Exec) {is=title K2Node_CallDelegate}
   n23 = OnTileButtonFocused!() -> (OutputDelegate, then, FocusedItem:WBP Discover Tile Button Object Reference) {is=title K2Node_CustomEvent}
   > n24 = Branch(Condition=@n21)
   Set IsViewAllSelected(execute=@n24.else, IsViewAllSelected=false)
   > Call On Tile Focus Gained(ItemFocused=Get TileItemIndex(self=Get WBP_Discover_Tile_Base(self:WBP Discover Tile Button Object Reference=@n23.FocusedItem)) -> (TileItemIndex=0))
   HandleOnInputMethodChanged!() -> (OutputDelegate, then, NewInputType:ECommonInputType Enum) {is=title K2Node_CustomEvent}
   > UpdateGridPanelVisibility()
   OnViewAllButtonFocusGained() {is=title K2Node_CustomEvent}
   > n25 = Branch(Condition=@n21)
   Set IsViewAllSelected(execute=@n25.else, IsViewAllSelected=true)
   > Call On View All Focus Gained!(execute:Exec, self:WBP Row Panel List Object Reference) -> (then:Exec) {is=title K2Node_CallDelegate}
   OnRowUnselected() {is=title K2Node_CustomEvent}
   > Set IsViewAllSelected(IsViewAllSelected=false)
   n26 = OnTileContextMenuToggled!() -> (OutputDelegate, then, DiscoverTile:WBP Discover Tile Base Object Reference) {is=title K2Node_CustomEvent}
   > Call On Tile Focus Gained(ItemFocused=Get TileItemIndex(self=@n26.DiscoverTile) -> (TileItemIndex=0))
   n27 = Event PreConstruct()
   > Branch(Condition=@n27.IsDesignTime)
   > SetMaxVisibleRows!(execute, self, MaxRows:int=2) -> (then)
   > UpdateRowPadding!(execute, self, IsTopTenRow:bool=false) -> (then)
   > n28 = Branch Natural Nav Rollback Cvar!(In:Exec) -> (Use Natural Nav:Exec, Rollback:Exec) {is=title K2Node_MacroInstance}
   SetNavigationRuleCustom!(
    execute=@n28.Rollback
    self:Widget Object Reference=Get O_ViewAllContainer() -> (O_ViewAllContainer:Overlay Object Reference)
    Direction:EUINavigation Enum=Left
    InCustomDelegate:Delegate=Create Event()
   ) -> (then)
   > SetReturnFocusToSelection!(execute, self:List View Object Reference=Get TileView(), bEnabled:bool=true) -> (then)
   > SetEnableProximateEntryNavigation!(execute, self:List View Object Reference=Get TileView(), bNewValue:bool=false) -> (then)
   Event Construct()
   n29 = Get Button_ViewAll_PaddingHalfOfTileEntrySpacing() -> (Button_ViewAll_PaddingHalfOfTileEntrySpacing:WBP Discover Button Generic Object Reference)
   > Bind Event to On Discover Button Focus Gained(self:WBP Discover Button Generic Object Reference=@n29, Delegate=Create Event()) {is=title K2Node_AddDelegate}
   n30 = Create Event()
   n31 = Get CommonInputSubsystem()
   > Bind Event to On Input Method Changed(self=@n31, Delegate=@n30)
   > UpdateGridPanelVisibility()
   > SetDesiredFocusWidget!(execute, self:User Widget Object Reference, Widget:Widget Object Reference=GetDesiredFocusWidget!() -> (FocusWidget:Widget Object Reference)) -> (then, ReturnValue:bool)
   Event Destruct()
   > Unbind all Events from On Discover Button Focus Gained!(execute:Exec, self:WBP Discover Button Generic Object Reference=@n29) -> (then:Exec) {is=title K2Node_ClearDelegate}
   > Unbind Event from On Input Method Changed(self=@n31, Delegate=@n30)
   Event OnMouseEnter()
   > n32 = SetIsRowHovered!(execute, self, IsHoveredRow:bool=true) -> (then)
   UpdateHasViewAll(execute=[@n32.then, @n33.then])
   Event OnMouseLeave!() -> (OutputDelegate, then, MouseEvent:"Pointer Event Structure (by ref)") {is=title K2Node_Event}
   > n33 = SetIsRowHovered(IsHoveredRow=false)
  ]}
 ForceToMinColumnCount {nodes [
   ForceToMinColumnCount() {is=title K2Node_FunctionEntry}
   > Branch(Condition="integer > integer"(A=Get ForcedMinColumnCount() -> (ForcedMinColumnCount:int=0)))
   > n1 = GetWidgetAbsoluteRect!(execute, Widget:Widget Object Reference=Get TileView()) -> (then, TopLeft:Vector 2D Structure, BottomRight:Vector 2D Structure, Size:Vector 2D Structure, DebugStringOutput:String)
   n2 = BreakVector2D(InVec=Get DesiredTileSize() -> (DesiredTileSize:Vector 2D Structure))
   n3 = float * float(
    A=GetHorizontalEntrySpacing!(self:List View Object Reference=Get TileView()) -> (ReturnValue:float)
    B=float + float(A:int=Get ForcedMinColumnCount() -> (ForcedMinColumnCount=0), B:int=1)
   ) {is=title K2Node_PromotableOperator}
   n4 = float / float(A=BreakVector2D(InVec=@n1.Size), B=GetViewportScale!() -> (ReturnValue:float))
   n5 = "Min (Float)"(A:double=float / float(A=float - float(A=@n4, B=@n3), B:int=Get ForcedMinColumnCount() -> (ForcedMinColumnCount=0)), B:double=@n2) -> (ReturnValue:double) {is=title K2Node_CommutativeAssociativeBinaryOperator}
   > SetEntryWidth!(execute, self:Tile View Object Reference=Get TileView(), NewWidth:float=@n5) -> (then)
   > SetEntryHeight!(execute, self:Tile View Object Reference=Get TileView(), NewHeight:float=float * float(A=float / float(A=@n5, B:double=@n2), B=@n2.Y)) -> (then)
  ]}
 GetDesiredEntryClassForItem {nodes [
   n1 = GetDesiredEntryClassForItem!() -> (then, Item:Object Reference) {is=title K2Node_FunctionEntry}
   > n2 = Cast To FortDiscoverTileItemVM!(execute, Object=@n1) -> (then, CastFailed, AsDiscover Tile Item View Model:Discover Tile Item View Model Object Reference) {is=title K2Node_DynamicCast}
   > n3 = Cast To FortDiscoverSurfaceTileItemVM!(execute, Object=@n2) -> (then, CastFailed, AsFort Discover Surface Tile Item VM:Fort Discover Surface Tile Item VM Object Reference) {is=title K2Node_DynamicCast}
   > Set TileFocusWeight(TileFocusWeight:double=1.5) -> (Output_Get:double)
   n4 = IsIPTile!(self:Fort Discover Surface Tile Item VM Object Reference=@n3) -> (ReturnValue:bool)
   > Return Node!(
    execute
    ReturnValue:User Widget Class Reference=Select(
     Option 0:User Widget Class Reference=Select(
      Option 0=/DiscoveryBrowser/Tiles/Categories/WBP_Tile_Category.WBP_Tile_Category_C
      Option 1=/DiscoveryBrowser/Tiles/Ip/WBP_Tile_IP_NEW.WBP_Tile_IP_NEW_C
      Index=@n4
     )
     Option 1:User Widget Class Reference=/DiscoveryBrowser/Tiles/Genre/WBP_Tile_Genre.WBP_Tile_Genre_C
     Index=Get bDisplayAsMiniGenreTile!(self:Discover Tile Item View Model Object Reference=@n2) -> (bDisplayAsMiniGenreTile:bool)
    ) -> (ReturnValue:User Widget Class Reference)
   )
   n5 = Get OverrideItemSelector() -> (OverrideItemSelector:Fort Discover Item Selector Base Object Reference)
   n6 = Is Valid(exec=@n3.CastFailed, InputObject=@n5)
   n7 = GetTileItemClass!(self:Fort Discover Item Selector Base Object Reference=@n5, InTileItemViewModel:Discover Tile Item View Model Object Reference=@n2) -> (ReturnValue:User Widget Class Reference)
   > n8 = Branch(Condition=IsValidClass!(Class:Object Class Reference=@n7) -> (ReturnValue:bool))
   > Return Node(ReturnValue=@n7)
   n9 = Byte to Enum EActivityItemType!(Byte:Byte=Get ItemType!(self:Discover Tile Item View Model Object Reference=@n2) -> (ItemType:Byte=0)) -> (ReturnValue:EActivityItemType Enum) {is=title K2Node_CastByteToEnum}
   Set TileFocusWeight(
    execute=[@n8.else, @n6.Is Not Valid]
    TileFocusWeight=Select(
     Option 0:double=Select!(
      Island:double=1
      Creator:double=0.75
      PlayingNow:double=1
      SurfaceTile:double=1.5
      IPTile:double=1.5
      PanelTile:double=1
      CommunityHighlight:double=1
      Index:EActivityItemType Enum=@n9
     ) -> (ReturnValue)
     Option 1:double=1.15
     Index=false
    ) -> (ReturnValue:double)
   )
   n10 = Select!(
    Option 0:User Widget Class Reference=/DiscoveryBrowser/Tiles/Community/WBP_CommunityPost_TileV1.WBP_CommunityPost_TileV1_C
    Option 1:User Widget Class Reference=/DiscoveryBrowser/Tiles/Community/WBP_CommunityPost_Tile.WBP_CommunityPost_Tile_C
    Index:bool=Get bEnableCreatorEvents!(self:Fort Runtime Options Object Reference=GetRuntimeOptions()) -> (bEnableCreatorEvents:bool)
   ) -> (ReturnValue:User Widget Class Reference)
   > Return Node(ReturnValue=Select(
     Option 0=Select!(
      Island:User Widget Class Reference=/DiscoveryBrowser/Tiles/Islands/WBP_IslandTile_Full.WBP_IslandTile_Full_C
      Creator:User Widget Class Reference=/DiscoveryBrowser/Tiles/Creator/WBP_CreatorTile.WBP_CreatorTile_C
      PlayingNow:User Widget Class Reference=/DiscoveryBrowser/Tiles/PlayingNow/WBP_PlayingNow_Tile.WBP_PlayingNow_Tile_C
      SurfaceTile:User Widget Class Reference
      IPTile:User Widget Class Reference
      PanelTile:User Widget Class Reference=&WBP_Tile_Genre_C
      CommunityHighlight:User Widget Class Reference=@n10
      Index:EActivityItemType Enum=@n9
     ) -> (ReturnValue)
     Option 1=/DiscoveryBrowser/Tiles/Islands/WBP_IslandTile_Number.WBP_IslandTile_Number_C
     Index=Get bUseTopTenStyle!(self:Discover Tile Item View Model Object Reference=@n2) -> (bUseTopTenStyle:bool)
    ))
  ]}
 GetDesiredFocusWidget {nodes [
   GetDesiredFocusWidget() {is=title K2Node_FunctionEntry}
   n1 = IsVisible(self=Get O_ViewAllContainer())
   n2 = Get IsViewAllSelected() -> (IsViewAllSelected:bool)
   > Return Node!(
    execute
    FocusWidget:Widget Object Reference=Select!(
     Option 0:Widget Object Reference=Get TileView()
     Option 1:Widget Object Reference=Get Button_ViewAll_PaddingHalfOfTileEntrySpacing()
     Index:bool=AND Boolean(A=@n2, B=@n1)
    ) -> (ReturnValue:Widget Object Reference)
   )
  ]}
 GetNumDisplayedRows {nodes [
   GetNumDisplayedRows() {is=title K2Node_FunctionEntry}
   n1 = Get TileView()
   n2 = GetNumberItemsPerRow(self=@n1) -> (ReturnValue=0)
   > n3 = Branch(Condition="integer > integer"(A=@n2))
   n4 = float / float(A="To Float (Integer)"(InInt=GetNumItems!(self:List View Object Reference=@n1) -> (ReturnValue:int=0)), B:int=@n2)
   > n5 = Set Result(Result:int="Min (Integer)"(A=Ceil(A=@n4) {is=title K2Node_CallFunction}, B=Get MaxRowsCached() -> (MaxRowsCached:int=0)) -> (ReturnValue=0)) -> (Output_Get:int=0)
   Return Node!(execute=[@n6.then, @n5.then], DisplayedRows:int=Get Result() -> (Result:int=0))
   n6 = Set Result(execute=@n3.else, Result=1) -> (Output_Get=0)
  ]}
 GetNumVisibleRowItems {nodes [
   GetNumVisibleRowItems() {is=title K2Node_FunctionEntry}
   > n1 = For Each Loop(Array:Array of User Widget Object References=GetDisplayedEntryWidgets!(self:List View Base Object Reference=Get TileView()) -> (ReturnValue:Array of User Widget Object References)) -> (Array Element:User Widget Object Reference)
   n2 = Get TileView()
   n3 = GetCachedGeometry!(self:Widget Object Reference=@n2) -> (ReturnValue:"Geometry Structure (by ref)")
   n4 = float + float!(
    A:double=GetLocalSize!(Geometry:"Geometry Structure (by ref)"=GetCachedGeometry(self=@n1)) -> (ReturnValue_X:double, ReturnValue_Y:double)
    B:double=Get TileViewContainerLocalSize() -> (TileViewContainerLocalSize:double)
    C:float=Get HorizontalEntrySpacing!(self:List View Object Reference=@n2) -> (HorizontalEntrySpacing:float)
   ) -> (ReturnValue)
   > Branch(Condition="float < float"(A=@n4, B=GetLocalSize(Geometry=@n3)))
   > Set TileViewContainerLocalSize(TileViewContainerLocalSize:double=@n4) -> (Output_Get:double)
   > Set TileIndex(TileIndex:int=@n1.Array Index) -> (Output_Get:int=0)
   n5 = Get TileView()
   n6 = int - int(A=Length(TargetArray:Array of User Widget Object References=GetDisplayedEntryWidgets(self=@n5)))
   n7 = Get MaxRowsCached() -> (MaxRowsCached=0)
   n8 = Branch(execute=@n1.Completed, Condition="Equal (Integer)"(A=Get TileIndex() -> (TileIndex:int=0), B=int / int(A=@n6, B=@n7)))
   n9 = Get MaxRowsCached() -> (MaxRowsCached=0)
   n10 = Get TileIndex() -> (TileIndex=0)
   > Return Node!(execute, NumItems:int=int + int(A=int * int(A=@n10, B=@n9)))
   Return Node(execute=@n8.else, NumItems=Get TileIndex() -> (TileIndex=0))
  ]}
 HasViewAllButton {nodes [
   HasViewAllButton() {is=title K2Node_FunctionEntry}
   n1 = Get TileView()
   n2 = Length(TargetArray=GetDisplayedEntryWidgets(self=@n1))
   n3 = GetNumItems(self=@n1) -> (ReturnValue=0)
   > Return Node!(execute, HasViewAll:bool=AND Boolean(A=Get RowHasViewAll() -> (RowHasViewAll:bool), B="integer > integer"(A=@n3, B=@n2)))
  ]}
 NavigateFromViewAllButtonToTileView {nodes [
   NavigateFromViewAllButtonToTileView!() -> (then, Navigation:EUINavigation Enum) {is=title K2Node_FunctionEntry}
   > Return Node!(execute, ReturnValue:Widget Object Reference=Get TileView())
  ]}
 OnFocusReceived {nodes [
   OnFocusReceived!() -> (then, MyGeometry:Geometry Structure, InFocusEvent:Focus Event Structure) {is=title K2Node_FunctionEntry}
   > n1 = Sequence()
   > Branch(Condition="integer >= integer"(A=Get LastSelectedIndex() -> (LastSelectedIndex:int=0)))
   n2 = Get TileView()
   > Set Selected Item!(
    execute
    self:List View Object Reference=@n2
    Item:Object Reference=GetItemAt!(self:List View Object Reference=@n2, Index:int=Get LastSelectedIndex() -> (LastSelectedIndex=0)) -> (ReturnValue:Object Reference)
   ) -> (then) {is=title K2Node_CallFunction}
   > Set LastSelectedIndex(LastSelectedIndex=-1) -> (Output_Get=0)
   n3 = Parent: OnFocusReceived!(
    execute=@n1.then_1
    MyGeometry:Geometry Structure=Get MyGeometry() -> (MyGeometry:Geometry Structure)
    InFocusEvent:Focus Event Structure=Get InFocusEvent() -> (InFocusEvent:Focus Event Structure)
   ) -> (then, ReturnValue:Event Reply Structure) {is=title K2Node_CallParentFunction}
   > Return Node(ReturnValue:Event Reply Structure=@n3)
  ]}
 OnMaxRowsSet {nodes [OnMaxRowsSet() {is=title K2Node_FunctionEntry}]}
 OnTileFocusGained {nodes [OnTileFocusGained!() -> (then, ItemFocused:int) {is=title K2Node_FunctionEntry}]}
 OnViewAllFocusGained {nodes [OnViewAllFocusGained() {is=title K2Node_FunctionEntry}]}
 OnViewAllSendProviderViewModel {nodes [OnViewAllSendProviderViewModel() {is=title K2Node_FunctionEntry}]}
 RefreshAfterMaxRowsSet {nodes [
   RefreshAfterMaxRowsSet() {is=title K2Node_FunctionEntry}
   n1 = Get TileView()
   n2 = Get MaxRowsCached() -> (MaxRowsCached=0)
   n3 = float * float(A=GetEntryHeight!(self:Tile View Object Reference=@n1) -> (ReturnValue:float), B:int=@n2)
   n4 = float * float(A=GetVerticalEntrySpacing!(self:List View Object Reference=@n1) -> (ReturnValue:float), B=@n2)
   > SetMaxDesiredHeight!(
    execute
    self:Size Box Object Reference=Get SizeBox_CapMaxRows() -> (SizeBox_CapMaxRows:Size Box Object Reference)
    InMaxDesiredHeight:float=float - float(A=float + float!(A=@n3, B=@n4) -> (ReturnValue), B=1)
   ) -> (then)
   > SetIsPointerScrollingEnabled!(execute, self:List View Base Object Reference=@n1, bInIsPointerScrollingEnabled:bool=false) -> (then)
   > SetIsGamepadScrollingEnabled!(execute, self:List View Base Object Reference=@n1, bInIsGamepadScrollingEnabled:bool=false) -> (then)
   > SetScrollbarVisibility!(execute, self:List View Base Object Reference=@n1, InVisibility:ESlateVisibility Enum=Collapsed) -> (then)
   > SetScrollOffset(self=@n1, InScrollOffset=0)
  ]}
 SetHasViewAll {nodes [
   n1 = SetHasViewAll!() -> (then, Has view all:bool=~) {is=title K2Node_FunctionEntry}
   > Set RowHasViewAll(RowHasViewAll:bool=@n1) -> (Output_Get:bool)
   n2 = Get CommonInputSubsystem()
   n3 = OR Boolean!(A="Equal (Enum)"(A=GetCurrentInputType(self=@n2), B=Touch), B=Get IsRowHovered() -> (IsRowHovered:bool), C:bool=Get IsViewAllSelected()) -> (ReturnValue)
   n4 = Get RowHasMoreItemsToDisplayInViewAll() -> (RowHasMoreItemsToDisplayInViewAll:bool)
   n5 = Get RowHasViewAll()
   > SetVisibility(
    self=Get O_ViewAllContainer()
    InVisibility=Select(
     Option 0:ESlateVisibility Enum=Collapsed
     Option 1:ESlateVisibility Enum=Visible
     Index=AND Boolean!(A=@n5, B=@n3, C:bool=@n4) -> (ReturnValue)
    ) -> (ReturnValue:ESlateVisibility Enum)
   )
  ]}
 SetIsRowHovered {nodes [
   n1 = SetIsRowHovered!() -> (then, IsHoveredRow:bool=~) {is=title K2Node_FunctionEntry}
   > n2 = Set IsRowHovered(IsRowHovered:bool=@n1) -> (Output_Get:bool)
   > n3 = Branch(Condition=@n2)
   Set IsViewAllSelected(execute=@n3.else, IsViewAllSelected=false)
  ]}
 SetMaxVisibleRows {nodes [
   n1 = SetMaxVisibleRows!() -> (then, MaxRows:int) {is=title K2Node_FunctionEntry}
   > Set MaxRowsCached(MaxRowsCached:int=@n1) -> (Output_Get:int=0)
   > Call On Max Rows Set!(execute:Exec, self:WBP Row Panel List Object Reference) -> (then:Exec) {is=title K2Node_CallDelegate}
   > RefreshAfterMaxRowsSet() {is=title K2Node_CallFunction}
  ]}
 SetRowPanelName {nodes [
   n1 = SetRowPanelName!() -> (then, PanelName:String) {is=title K2Node_FunctionEntry}
   > Set RowPanelName(RowPanelName:String=@n1) -> (Output_Get:String)
  ]}
 UpdateBaseTileProperties {nodes [
   n1 = UpdateBaseTileProperties!() -> (then, ItemIndex:int, TileBaseWidget:WBP Discover Tile Base Object Reference) {is=title K2Node_FunctionEntry}
   > SetTileIndex!(execute, self:WBP Discover Tile Base Object Reference=@n1.TileBaseWidget, TileIndex:int=@n1) -> (then)
   > Set RowHasViewAll!(execute, RowHasViewAll=Get RowHasViewAll(), self:WBP Discover Tile Base Object Reference=@n1.TileBaseWidget) -> (then, Output_Get)
   > SetParentPanelName!(execute, self:WBP Discover Tile Base Object Reference=@n1.TileBaseWidget, PanelName:String=Get RowPanelName() -> (RowPanelName:String)) -> (then)
  ]}
 UpdateGridPanelVisibility {nodes [
   UpdateGridPanelVisibility() {is=title K2Node_FunctionEntry}
   n1 = GetCurrentInputType(self=Get CommonInputSubsystem())
   > SetVisibility(
    self=Get GridPanel_Main() -> (GridPanel_Main:Grid Panel Object Reference)
    InVisibility=Select!(
     MouseAndKeyboard:ESlateVisibility Enum=SelfHitTestInvisible
     Gamepad:ESlateVisibility Enum=Visible
     Touch:ESlateVisibility Enum=SelfHitTestInvisible
     Count:ESlateVisibility Enum=Visible
     Index:ECommonInputType Enum=@n1
    ) -> (ReturnValue)
   )
   > SetVisibility(
    self=Get ButtonToEmulateRowHovering() -> (ButtonToEmulateRowHovering:WBP Discover Button Emulate Row Hover Object Reference)
    InVisibility=Select(MouseAndKeyboard=Visible, Gamepad=Collapsed, Touch=Collapsed, Count=Collapsed, Index=@n1)
   )
  ]}
 UpdateHasViewAll {nodes [
   UpdateHasViewAll() {is=title K2Node_FunctionEntry}
   n1 = GetDisplayedEntryWidgets(self=Get TileView())
   n2 = Get TileView()
   > n3 = Set RowHasMoreItemsToDisplayInViewAll(RowHasMoreItemsToDisplayInViewAll:bool="integer > integer"(A=GetNumItems(self=@n2) -> (ReturnValue=0), B=Length(TargetArray=@n1))) -> (Output_Get:bool)
   > SetHasViewAll!(execute, self, Has view all:bool=@n3) -> (then) {is=title K2Node_CallFunction}
  ]}
 UpdateRowPadding {nodes [
   n1 = UpdateRowPadding!() -> (then, IsTopTenRow:bool=~) {is=title K2Node_FunctionEntry}
   n2 = GetScalarParameterDefaultValue!(
    self:Material Parameter Collection Object Reference=Get DiscoverySharedValues() -> (DiscoverySharedValues:Material Parameter Collection Object Reference)
    ParameterName:Name=TopTenRowLeftPadding
   ) -> (bParameterFound:bool, ReturnValue:float)
   > Branch(Condition=@n2)
   n3 = SlotAsGridSlot(Widget=Get OverlayRoot() -> (OverlayRoot:Overlay Object Reference))
   n4 = Get!(self:Grid Slot Object Reference=@n3) -> (Padding_Left:float=~, Padding_Top:float=~, Padding_Right:float=~, Padding_Bottom:float=~)
   > SetPadding!(
    execute
    self:Grid Slot Object Reference=@n3
    InPadding_Left:float=SelectFloat(A=@n2.ReturnValue, B=64, bPickA=@n1)
    InPadding_Top:float=@n4.Padding_Top
    InPadding_Right:float=@n4.Padding_Right
    InPadding_Bottom:float=@n4.Padding_Bottom
   ) -> (then)
  ]}
 UpdateTileViewLayout {nodes [
   n1 = UpdateTileViewLayout!() -> (then, TileBaseWidget:WBP Discover Tile Base Object Reference) {is=title K2Node_FunctionEntry}
   > Set DesiredTileSize(DesiredTileSize:Vector 2D Structure=Get Dynamic Tile Size!(self:WBP Discover Tile Base Object Reference=@n1) -> (DynamicTileSize:Vector 2D Structure) {is=title K2Node_CallFunction}) -> (Output_Get:Vector 2D Structure)
   > n2 = Branch(Condition="integer <= integer"(A=Get ForcedMinColumnCount() -> (ForcedMinColumnCount=0)))
   n3 = BreakVector2D(InVec=Get DesiredTileSize())
   n4 = Get TileView()
   > SetEntryWidth(self=@n4, NewWidth=@n3)
   > n5 = SetEntryHeight(self=@n4, NewHeight=@n3.Y)
   SetMaxVisibleRows(execute=[@n5.then, @n6.then], MaxRows=Get MaxRowsCached() -> (MaxRowsCached=0)) {is=title K2Node_CallFunction}
   > RequestRefresh(self:List View Base Object Reference=@n4)
   n6 = ForceToMinColumnCount(execute=@n2.else) {is=title K2Node_CallFunction}
  ]}
]
variables [
 DesiredTileSize {type=Vector2D, default=<0 0>}
 DiscoverySharedValues {type=MaterialParameterCollection, default=/DiscoveryBrowser/DataAssets/DiscoverySharedValues}
 ForcedMinColumnCount {type=integer}
 IsRowHovered {type=boolean}
 IsViewAllSelected {type=boolean}
 LastSelectedIndex {type=integer, default=-1}
 MaxRowsCached {type=integer}
 OnMaxRowsSet {}
 OnTileFocusGained {}
 OnViewAllFocusGained {}
 OnViewAllSendProviderViewModel {}
 OuterItemRowIndex {type=integer}
 OverrideItemSelector {type=FortDiscoverItemSelectorBase, default=null}
 RowHasMoreItemsToDisplayInViewAll {type=boolean}
 RowHasViewAll {type=boolean}
 RowPanelName {type=string, default=False}
 TileFocusWeight {type=number, default=1}
 TilesHandleTheirOwnStartVisibilityImpression {type=boolean, default=true}
 UseFocusToTriggerOutlineVisuals {type=boolean}
]
widgets {
 info {inheritedWidgetCount=1, parentClass=/Script/FortniteUI.FortPolymorphicTileView, rootWidgetClass=&GridPanel, widgetCount=11}
 widgets [
  GridPanel_0 {widgetClassPath=&GridPanel, widget=GridPanel_0, properties {columnFill [1], rowFill [1]}}
  GridPanel_Main {
   widgetClassPath=&GridPanel
   bIsVariable=true
   parent=GridPanel_0
   widget=GridPanel_Main
   properties {visibility=Visible, columnFill [1], rowFill [0,1]}
   slot {class=GridSlot, properties {padding=<ltrb 0 0 0 48>}}
  }
  WBP_Row_Header {
   widgetClassPath=&/WBP_Row_Header
   bIsVariable=true
   parent=GridPanel_Main
   widget=WBP_Row_Header
   slot {class=GridSlot, properties {layer=1, rowSpan=1, verticalAlignment=VAlign_Top}}
  }
  OverlayRoot {
   widgetClassPath=&Overlay
   bIsVariable=true
   parent=GridPanel_Main
   widget=OverlayRoot
   slot {class=GridSlot, properties {layer=1, row=1, rowSpan=1, verticalAlignment=VAlign_Top, padding=<ltrb 64 8 0 0>}}
  }
  SizeBox_CapMaxRows {
   widgetClassPath=&SizeBox
   bIsVariable=true
   parent=OverlayRoot
   widget=SizeBox_CapMaxRows
   properties {maxDesiredHeight=383}
   slot {class=OverlaySlot, properties {horizontalAlignment=HAlign_Fill, verticalAlignment=VAlign_Fill, padding=<ltrb 0 0 80 0>}}
  }
  HorizontalBox_1 {widgetClassPath=&HorizontalBox, parent=SizeBox_CapMaxRows, widget=HorizontalBox_1, slot {class=SizeBoxSlot}}
  TileView {
   widgetClassPath=/Script/FortniteUI.FortTileView
   bInherited=true
   bIsVariable=true
   parent=HorizontalBox_1
   widget=TileView
   properties {
    allowOverscroll=false
    bAllowDragging=false
    bConsumeTouchInputWhenScrolling=false
    bEnableProximateEntryNavigation=true
    bEnableRightClickScrolling=false
    bEnableScrollAnimation=false
    bEnableTouchScrolling=false
    bEntrySizeIncludesEntrySpacing=false
    bNeverConsumeTouchInputWhenScrolling=true
    clipping=Inherit
    entryHeight=176
    entryWidgetClass=&WBP_IslandTile_Full_C
    entryWidth=312
    horizontalEntrySpacing=24
    numDesignerPreviewEntries=20
    tileAlignment=LeftAligned
    verticalEntrySpacing=16
    visibility=SelfHitTestInvisible
    wheelScrollMultiplier=0
   }
   slot {class=HorizontalBoxSlot, properties {size {sizeRule=Fill, value=1}}}
  }
  O_ViewAllContainer {
   widgetClassPath=&Overlay
   bIsVariable=true
   parent=HorizontalBox_1
   widget=O_ViewAllContainer
   slot {class=HorizontalBoxSlot, properties {padding=<ltrb 0 8 0 8>}}
  }
  WBP_CaptureForPostBufferUpdate {
   widgetClassPath=&WBP_CaptureForPostBufferUpdate
   bIsVariable=true
   parent=O_ViewAllContainer
   slot {class=OverlaySlot, properties {horizontalAlignment=HAlign_Fill, verticalAlignment=VAlign_Fill}}
  }
  Button_ViewAll_PaddingHalfOfTileEntrySpacing {
   widgetClassPath=/DiscoveryBrowser/Tiles/WBP_Discover_Button_Generic
   bIsVariable=true
   parent=O_ViewAllContainer
   widget=Button_ViewAll_PaddingHalfOfTileEntrySpacing
   properties {
    bHideInputAction=true
    bShouldSelectUponReceivingFocus=true
    bSimulateHoverOnTouchInput=false
    iconMaterial=/FortUILibrary/Textures/MaterialInstances/MI_UIKit_Icon_Chevron
    minHeight=112
    minWidth=80
    touchMethod=PreciseTap
    useTextBlock=false
    buttonPadding=<ltrb 12 0 12 0>
   }
   slot {class=OverlaySlot, properties {horizontalAlignment=HAlign_Fill, verticalAlignment=VAlign_Fill}}
  }
  ButtonToEmulateRowHovering {
   widgetClassPath=/DiscoveryBrowser/Tiles/WBP_Discover_ButtonEmulateRowHover
   bIsVariable=true
   parent=GridPanel_Main
   widget=ButtonToEmulateRowHovering
   slot {class=GridSlot, properties {rowSpan=2}}
  }
 ]
}
