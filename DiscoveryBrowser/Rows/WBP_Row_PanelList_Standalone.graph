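# Standalone Discovery browser row: an activatable list-entry widget that wraps WBP_Row_PanelList and stores its list item as a Discover provider view model.
# Panel events drive tile selection from focus or hover, open View All with that view model, and share a nested-button selection view model with generated tiles.
# Focus-path changes activate or deactivate the row, and touch handling clears selections.
# OnPaint feeds the panel's screen rectangle into row visibility computation.
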WidgetBlueprint /DiscoveryBrowser/Rows/WBP_Row_PanelList_Standalone
graphs [
 BP_GetDesiredFocusTarget {nodes [
   Get Desired Focus Target()
   > Return Node!(execute, ReturnValue:Widget Object Reference=Get PanelList() -> (PanelList:WBP Row Panel List Object Reference))
  ]}
 EventGraph {nodes [
   n1 = OnItemFocused!() -> (OutputDelegate, then, ItemFocused:int) {is=title K2Node_CustomEvent}
   n2 = Get TileView!(self:Fort Polymorphic Tile View Object Reference=Get PanelList()) -> (TileView:Fort Tile View Object Reference)
   > SetSelectedIndex!(execute, self:List View Object Reference=@n2, Index:int=@n1.ItemFocused) -> (then)
   n3 = OnItemHovered!() -> (OutputDelegate, then, Item:Object Reference, bIsHovered:bool=~) {is=title K2Node_CustomEvent}
   > Branch(Condition=@n3.bIsHovered)
   > SetSelectedIndex(self=@n2, Index=GetIndexForItem!(self:List View Object Reference=@n2, Item:Object Reference=@n3.Item) -> (ReturnValue:int=0))
   n4 = HandleInputMethodChanged!() -> (OutputDelegate, then, bNewInputType:ECommonInputType Enum) {is=title K2Node_CustomEvent}
   > n5 = Switch on ECommonInputType!(execute, Selection:ECommonInputType Enum=@n4.bNewInputType) -> (MouseAndKeyboard:Exec, Gamepad:Exec, Touch:Exec, Count:Exec) {is=title K2Node_SwitchEnum}
   Clear Selection(execute=@n5.Touch, self:List View Object Reference=Get TileView(self=Get PanelList()))
   OnSendProviderViewModel() {is=title K2Node_CustomEvent}
   > OpenViewAllScreen!(
    execute
    self:WBP Row Panel List Object Reference=Get PanelList()
    ProviderViewModel:Discover Provider View Model Object Reference=Get FortDiscoverProviderViewModel() -> (FortDiscoverProviderViewModel:Discover Provider View Model Object Reference)
   ) -> (then)
   n6 = OnTileItemGenerated!() -> (OutputDelegate, then, Widget:User Widget Object Reference) {is=title K2Node_CustomEvent}
   > SetNestedButtonGroupSelectionVM!(
    execute:Exec
    self:Object Reference=@n6.Widget
    InVM:Fort Nested Button Group Selection VM Object Reference=Get NestedButtonGroupSelectionVM() -> (NestedButtonGroupSelectionVM:Fort Nested Button Group Selection VM Object Reference)
   ) -> (then:Exec) {is=title K2Node_Message}
   Event Construct()
   n7 = Create Event()
   n8 = Get CommonInputSubsystem()
   > Bind Event to On Input Method Changed(self=@n8, Delegate=@n7)
   n9 = Get PanelList()
   > Bind Event to On Tile Focus Gained(self:WBP Row Panel List Object Reference=@n9, Delegate=Create Event()) {is=title K2Node_AddDelegate}
   n10 = Get TileView(self=@n9)
   > Bind Event to On Item Is Hovered Changed(self:List View Object Reference=@n10, Delegate=Create Event()) {is=title K2Node_AddDelegate}
   > Bind Event to On Entry Generated(self:List View Base Object Reference=@n10, Delegate=Create Event()) {is=title K2Node_AddDelegate}
   > Bind Event to On View All Send Provider View Model(self:WBP Row Panel List Object Reference=@n9, Delegate=Create Event()) {is=title K2Node_AddDelegate}
   > SetDesiredFocusWidget!(execute, self:User Widget Object Reference, Widget:Widget Object Reference=Get PanelList()) -> (then, ReturnValue:bool)
   Event OnAddedToFocusPath!() -> (OutputDelegate, then, InFocusEvent:Focus Event Structure) {is=title K2Node_Event}
   > ActivateWidget()
   n11 = Event OnListItemObjectSet()
   > n12 = Cast To FortDiscoverProviderViewModel!(execute, Object=@n11.ListItemObject) -> (then, CastFailed, AsDiscover Provider View Model:Discover Provider View Model Object Reference) {is=title K2Node_DynamicCast}
   > Set FortDiscoverProviderViewModel(FortDiscoverProviderViewModel:Discover Provider View Model Object Reference=@n12) -> (Output_Get:Discover Provider View Model Object Reference)
   Event Destruct()
   > Unbind Event from On Input Method Changed(self=@n8, Delegate=@n7)
   n13 = Get PanelList()
   > Unbind Event from On Entry Generated(self:List View Base Object Reference=Get TileView(self=@n13), Delegate=Create Event()) {is=title K2Node_RemoveDelegate}
   > Set IsViewAllSelected!(execute, IsViewAllSelected:bool=false, self:WBP Row Panel List Object Reference=@n13) -> (then, Output_Get:bool)
   Event OnRemovedFromFocusPath!() -> (OutputDelegate, then, InFocusEvent:Focus Event Structure) {is=title K2Node_Event}
   > DeactivateWidget()
   Event On Entry Released() {is=title K2Node_Event}
   > n14 = Cast To UserWidget!(
    execute
    Object:Fort Nested Button Group VMItem Interface=Get CurrentSelectedButton!(self:Fort Nested Button Group Selection VM Object Reference=Get NestedButtonGroupSelectionVM()) -> (CurrentSelectedButton:Fort Nested Button Group VMItem Interface)
   ) -> (then, CastFailed, AsUser Widget:User Widget Object Reference) {is=title K2Node_DynamicCast}
   > Is Valid(InputObject=Get NestedButtonGroupSelectionVM())
   n15 = GetDisplayedEntryWidgets!(self:List View Base Object Reference=Get TileView(self=Get PanelList())) -> (ReturnValue:Array of User Widget Object References)
   n16 = GetOwningPlayer()
   > Branch(Condition=AND Boolean(
     A=IsUsingTouch(PlayerController=@n16)
     B=Contains Item(TargetArray:Array of User Widget Object References=@n15, ItemToFind:"User Widget Object Reference (by ref)"=@n14)
    ))
   n17 = Get NestedButtonGroupSelectionVM()
   > SetIsSelected!(
    execute
    self:Fort Nested Button Group Selection VM Object Reference=@n17
    InButton:Fort Nested Button Group VMItem Interface=Get CurrentSelectedButton(self=@n17)
    bInSelected:bool=false
   ) -> (then)
   Event On Item Selection Changed!() -> (OutputDelegate, then, bIsSelected:bool) {is=title K2Node_Event}
  ]}
 OnPaint {nodes [
   OnPaint!() -> (then, Context:"Paint Context Structure (by ref)") {is=title K2Node_FunctionEntry}
   n1 = Get FortDiscoverProviderViewModel()
   > Is Valid(InputObject=@n1)
   > n2 = GetWidgetAbsoluteRect!(execute, Widget:Widget Object Reference=Get PanelList()) -> (then, TopLeft:Vector 2D Structure, BottomRight:Vector 2D Structure, Size:Vector 2D Structure, DebugStringOutput:String)
   > ComputeRowVisibility!(
    execute
    ListView:List View Object Reference=Get TileView(self=Get PanelList())
    Provider VM:Discover Provider View Model Object Reference=@n1
    TopLeft:Vector 2D Structure=@n2
    BottomRight:Vector 2D Structure=@n2.BottomRight
   ) -> (then)
   > Return Node!(execute)
  ]}
]
variables [NestedButtonGroupSelectionVM {type=FortNestedButtonGroupSelectionVM, default=null}]
widgets {
 info {
  inheritedWidgetCount=0
  namedSlotCount=1
  parentClass=&CommonActivatableWidget
  rootWidgetClass=&/WBP_Row_PanelList.WBP_Row_PanelList_C
  widgetCount=2
 }
 widgets [
  PanelList {widgetClassPath=&/WBP_Row_PanelList, bIsVariable=true, widget=PanelList}
  WBP_Row_PanelList {widgetClassPath=&WBP_Row_PanelList, bIsVariable=true, widget=WBP_Row_PanelList}
 ]
}
