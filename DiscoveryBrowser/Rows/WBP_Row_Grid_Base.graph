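# Base widget for Discovery browser grid rows, derived from FortDiscoverBrowserGridRow, with no widget tree of its own.
# It tracks input method changes, becoming hit-testable for touch and SelfHitTestInvisible otherwise.
# OpenNestedSurface asynchronously pushes WBP_Discover_NestedSurface, then passes it the tile viewmodel and reference surface name.
# Helper functions broadcast fullscreen changes and, on touch, deselect the nested-group button displayed in a given list.
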WidgetBlueprint /DiscoveryBrowser/Rows/WBP_Row_Grid_Base
graphs [
 DiscoverFullscreenChanged {nodes [DiscoverFullscreenChanged!() -> (then, IsFullscreen:bool=~) {is=title K2Node_FunctionEntry}]}
 EventGraph {nodes [
   HandleOnInputMethodChanged!() -> (OutputDelegate, then, bNewInputType:ECommonInputType Enum) {is=title K2Node_CustomEvent}
   > UpdateVisibility()
   Event Construct()
   n1 = Create Event()
   n2 = Get CommonInputSubsystem()
   > Bind Event to On Input Method Changed(self=@n2, Delegate=@n1)
   > HandleOnInputMethodChanged!(execute, self, bNewInputType:ECommonInputType Enum=GetCurrentInputType(self=@n2)) -> (then) {is=title K2Node_CallFunction}
   > UpdateVisibility()
   Event Destruct()
   > Unbind Event from On Input Method Changed(self=@n2, Delegate=@n1)
   n3 = Event OpenNestedSurface!() -> (OutputDelegate, then, SurfaceTileItemVM:Fort Discover Surface Tile Item VM Object Reference) {is=title K2Node_Event}
   > n4 = PushPrimaryContentClassAsync!(
    execute:Exec
    WidgetClass:Common Activatable Widget Class Reference=/DiscoveryBrowser/WBP_Discover_NestedSurface.WBP_Discover_NestedSurface_C
   ) -> (then:Exec, OnPrimaryContentPushed:Exec, ActivatableWidget:Common Activatable Widget Object Reference) {is=title K2Node_AsyncAction}
   n5 = Cast To WBP_Discover_NestedSurface!(execute=@n4.OnPrimaryContentPushed, Object=@n4) -> (then, CastFailed, AsWBP Discover Nested Surface:WBP Discover Nested Surface Object Reference) {is=title K2Node_DynamicCast}
   > SetFortDiscoverSurfaceTileItemVM!(
    execute
    self:WBP Discover Nested Surface Object Reference=@n5
    Viewmodel:Fort Discover Surface Tile Item VM Object Reference=@n3.SurfaceTileItemVM
   ) -> (then)
   > On Disover Surface Tile Item VM Set(self:WBP Discover Nested Surface Object Reference=@n5)
   > SetSurfaceName!(
    execute
    self:Discover Surface View Model Object Reference=Get FortDiscoverSurfaceViewModel!(self:WBP Discover Nested Surface Object Reference=@n5) -> (FortDiscoverSurfaceViewModel:Discover Surface View Model Object Reference)
    InSurfaceName:Name=GetReferenceSurfaceName!(self:Fort Discover Surface Tile Item VM Object Reference=@n3.SurfaceTileItemVM) -> (ReturnValue:Name=None)
   ) -> (then)
   Event On Entry Released() {is=title K2Node_Event}
  ]}
 SetDiscoverFullscreenState {nodes [
   n1 = SetDiscoverFullscreenState!() -> (then, IsFullscreen:bool=~) {is=title K2Node_FunctionEntry}
   > Call Discover Fullscreen Changed!(execute:Exec, self:WBP Row Grid Base Object Reference, IsFullscreen:bool=@n1) -> (then:Exec) {is=title K2Node_CallDelegate}
  ]}
 SetIsSelectedNestedVM {nodes [
   n1 = SetIsSelectedNestedVM!() -> (then, ListContainer:List View Base Object Reference) {is=title K2Node_FunctionEntry}
   > n2 = Cast To UserWidget!(
    execute
    Object:Fort Nested Button Group VMItem Interface=Get CurrentSelectedButton!(self:Fort Nested Button Group Selection VM Object Reference=Get NestedButtonGroupSelectionVM() -> (NestedButtonGroupSelectionVM:Fort Nested Button Group Selection VM Object Reference)) -> (CurrentSelectedButton:Fort Nested Button Group VMItem Interface)
   ) -> (then, CastFailed, AsUser Widget:User Widget Object Reference) {is=title K2Node_DynamicCast}
   > Is Valid(InputObject=Get NestedButtonGroupSelectionVM())
   n3 = GetDisplayedEntryWidgets!(self:List View Base Object Reference=@n1) -> (ReturnValue:Array of User Widget Object References)
   n4 = GetOwningPlayer()
   > Branch(Condition=AND Boolean(
     A=IsUsingTouch(PlayerController=@n4)
     B=Contains Item(TargetArray:Array of User Widget Object References=@n3, ItemToFind:"User Widget Object Reference (by ref)"=@n2)
    ))
   n5 = Get NestedButtonGroupSelectionVM()
   > SetIsSelected!(
    execute
    self:Fort Nested Button Group Selection VM Object Reference=@n5
    InButton:Fort Nested Button Group VMItem Interface=Get CurrentSelectedButton(self=@n5)
    bInSelected:bool=false
   ) -> (then)
  ]}
 UpdateVisibility {nodes [
   UpdateVisibility() {is=title K2Node_FunctionEntry}
   n1 = GetCurrentInputType(self=Get CommonInputSubsystem())
   > SetVisibility(InVisibility=Select(Option 0=SelfHitTestInvisible, Option 1=Visible, Index="Equal (Enum)"(A=@n1, B=Touch)))
  ]}
]
variables [DiscoverFullscreenChanged {}]
widgets {
 info {inheritedWidgetCount=0, parentClass=/Script/DiscoveryBrowserUI.FortDiscoverBrowserGridRow, rootWidgetClass=null, widgetCount=0}
 widgets []
}
