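# Search tab button for the Discovery browser, derived from the generic UIKit modular button.
# PreConstruct sets selection timings and text materials, then rounds corners: compact UI casts the background block and caps width, otherwise it writes parameters to the dynamic materials.
# Two near-identical count setters drive the status indicator, hiding it at zero; GetNotificationData returns visibility for unseen notifications.
# SetText substitutes empty text, and SetupSurface copies the tile viewmodel's reference surface name into a surface viewmodel.
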WidgetBlueprint /DiscoveryBrowser/WBP_Search_Tab_Button
graphs [
 EventGraph {nodes [
   n1 = Cast To WBP_UIKit_Block_BackgroundTopNavBar!(execute, Object:Wildcard) -> (then, CastFailed, AsWBP UIKit Block Background Top Nav Bar:WBP UIKit Block Background Top Nav Bar Object Reference) {is=title K2Node_DynamicCast}
   > SetCornerRadiuses!(execute, self:WBP UIKit Block Background Top Nav Bar Object Reference=@n1, CornerRadius:float=10, RadiusInPixels:bool=true) -> (then)
   n2 = Event PreConstruct()
   > Parent: PreConstruct(IsDesignTime=@n2.IsDesignTime)
   > SetBlocksTiming!(
    execute
    self:UIKit Modular Button Object Reference
    Timing:"UIKit Block Timing Structure (by ref)"=Make UIKit Block Timing!(
     Hovering:float=0
     Unhovering:float=0
     Focusing:float=0
     Unfocusing:float=0
     Pressing:float=0
     Releasing:float=0
     Disabling:float=0
     Enabling:float=0
     Selecting:float=0.15
     DeselectingIdle:float=0.1
     DeselectingFocused:float=0.1
    ) -> (UIKitBlockTiming:UIKit Block Timing Structure) {is=title K2Node_MakeStruct}
   ) -> (then)
   n3 = Get Block_Text() -> (Block_Text:WBP UIKit Block Text Object Reference)
   > SetFontMaterial!(
    execute
    self:WBP UIKit Block Text Object Reference=@n3
    In Material:Material Instance Object Reference=Get TextMaterial() -> (TextMaterial:Material Instance Object Reference)
   ) -> (then)
   > SetOutlineMaterial!(
    execute
    self:WBP UIKit Block Text Object Reference=@n3
    In Material:Material Instance Object Reference=Get TextOutlineMaterial() -> (TextOutlineMaterial:Material Instance Object Reference)
   ) -> (then)
   > n4 = Branch(Condition=IsUsingCompactUI())
   n5 = Get Block_Background() -> (Block_Background:WBP UIKit Block Background Base Object Reference)
   > n6 = Cast To WBP_UIKit_Block_BackgroundTabSideNav!(execute, Object=@n5) -> (then, CastFailed, AsWBP UIKit Block Background Tab Side Nav:WBP UIKit Block Background Tab Side Nav Object Reference) {is=title K2Node_DynamicCast}
   > n7 = SetCornerRadiuses(self:WBP UIKit Block Background Tab Side Nav Object Reference=@n6, CornerRadius=20, RadiusInPixels=true)
   SetMaxDimensions!(execute=[@n9.then, @n7.then], self:Common Button Base Object Reference, InMaxWidth:int=280, InMaxHeight:int=0) -> (then)
   n8 = Cast To WBP_UIKit_Block_BackgroundTopNavBar(execute=@n6.CastFailed, Object:Object Reference=@n5)
   > n9 = SetCornerRadiuses(self:WBP UIKit Block Background Top Nav Bar Object Reference=@n8, CornerRadius=20, RadiusInPixels=true)
   n10 = Get OutlineDynamicMaterial!(self:WBP UIKit Block Background Base Object Reference=Get Block_Background()) -> (OutlineDynamicMaterial:Material Instance Dynamic Object Reference)
   SetScalarParameterValue(execute=@n4.else, self=@n10, ParameterName=Radius in Pixels, Value=1)
   > SetVectorParameterValue(
    self=@n10
    ParameterName=Corner Radius
    Value=Make Linear Color!(R:float=20, G:float=20, B:float=20, A:float=20) -> (LinearColor:Linear Color Structure) {is=title K2Node_MakeStruct}
   )
   n11 = Get BackgroundDynamicMaterial!(self:WBP UIKit Block Background Base Object Reference=Get Block_Background()) -> (BackgroundDynamicMaterial:Material Instance Dynamic Object Reference)
   > SetScalarParameterValue(self=@n11, ParameterName=Radius in Pixels, Value=1)
   > SetScalarParameterValue(self=@n11, ParameterName=Corner Radius, Value=10)
   Event Construct()
   > Parent: Construct()
   Event On Deselected()
   > Parent: On Deselected() {is=title K2Node_CallParentFunction}
   Event On Selected()
   > Parent: On Selected() {is=title K2Node_CallParentFunction}
  ]}
 GetNotificationData {nodes [
   n1 = GetNotificationData!() -> (then, HasUnseenNotification:bool=~) {is=title K2Node_FunctionEntry}
   n2 = NOT Boolean(A="Equal (Integer)"(A=Get TabNotificationCount() -> (TabNotificationCount:int=0)))
   > Return Node!(execute, Visibility:ESlateVisibility Enum=Select(Option 0=Collapsed, Option 1=SelfHitTestInvisible, Index=AND Boolean(A=@n1, B=@n2)))
  ]}
 SetNotificationCount {nodes [
   n1 = SetNotificationCount!() -> (then, Count:int) {is=title K2Node_FunctionEntry}
   > n2 = Set TabNotificationCount(TabNotificationCount:int=@n1) -> (Output_Get:int=0)
   > OverrideCount!(
    execute
    self:WBP UIKit Status Indicator Base Object Reference=Get StatusIndicator() -> (StatusIndicator:WBP UIKit Status Indicator Object Reference)
    Count:int=@n2
   ) -> (then)
   > n3 = Branch(Condition="Equal (Integer)"(A=Get TabNotificationCount() -> (TabNotificationCount=0)))
   n4 = Get StatusIndicator()
   > SetVisibility(self=@n4, InVisibility=Collapsed)
   SetVisibility(execute=@n3.else, self=@n4, InVisibility=SelfHitTestInvisible)
  ]}
 SetTabNotificationCount {nodes [
   n1 = SetTabNotificationCount!() -> (then, Count:int) {is=title K2Node_FunctionEntry}
   > n2 = Set TabNotificationCount(TabNotificationCount=@n1) -> (Output_Get=0)
   > OverrideCount(self=Get StatusIndicator(), Count=@n2)
   > n3 = Branch(Condition="Equal (Integer)"(A=Get TabNotificationCount() -> (TabNotificationCount=0)))
   n4 = Get StatusIndicator()
   > SetVisibility(self=@n4, InVisibility=Collapsed)
   SetVisibility(execute=@n3.else, self=@n4, InVisibility=SelfHitTestInvisible)
  ]}
 SetText {nodes [
   n1 = SetText!() -> (then, Text:Text) {is=title K2Node_FunctionEntry}
   > n2 = Set TabText(TabText:Text=Select(Option 0:Text=@n1, Option 1:Text="LOCTEXT(\"Empty\")", Index=TextIsEmpty(InText=@n1)) -> (ReturnValue:Text)) -> (Output_Get:Text)
   > Parent: SetText!(execute, Text:Text=@n2) -> (then) {is=title K2Node_CallParentFunction}
  ]}
 SetupSurface {nodes [
   n1 = SetupSurface!() -> (then, SurfaceVM:Discover Surface View Model Object Reference) {is=title K2Node_FunctionEntry}
   n2 = Get FortDiscoverSurfaceTileItemVM() -> (FortDiscoverSurfaceTileItemVM:Fort Discover Surface Tile Item VM Object Reference)
   > SetFortDiscoverSurfaceTileItemVM!(execute, self, Viewmodel:Fort Discover Surface Tile Item VM Object Reference=@n2) -> (then)
   > SetSurfaceName!(
    execute
    self:Discover Surface View Model Object Reference=@n1
    InSurfaceName:Name=GetReferenceSurfaceName!(self:Fort Discover Surface Tile Item VM Object Reference=@n2) -> (ReturnValue:Name=None)
   ) -> (then)
  ]}
]
variables [
 IsMobile {type=boolean}
 MobileTextMaterial {type=MaterialInstance, default=/FortUILibrary/UIKit/Buttons/TabButtons/_MaterialInstances/MI_UI_TopNav_TextMobile}
 MobileTextOutlineMaterial {type=MaterialInstance, default=/FortUILibrary/UIKit/Buttons/TabButtons/_MaterialInstances/MI_UI_TopNav_TextOutlineMobile}
 TabNotificationCount {type=integer}
 TabText {type=string, default=""}
 TextMaterial {type=MaterialInstance, default=/FortUILibrary/UIKit/Buttons/TabButtons/_MaterialInstances/MI_UI_TopNav_Text}
 TextOutlineMaterial {type=MaterialInstance, default=/FortUILibrary/UIKit/Buttons/TabButtons/_MaterialInstances/MI_UI_TopNav_TextOutline}
]
widgets {
 info {
  inheritedWidgetCount=0
  parentClass=/FortUILibrary/Systems/ModularButton/WBP_UIKit_ButtonGeneric_Base.WBP_UIKit_ButtonGeneric_Base_C
  rootWidgetClass=null
  widgetCount=0
 }
 widgets []
}
