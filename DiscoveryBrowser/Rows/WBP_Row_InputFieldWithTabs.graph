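# Discovery browser widget that pairs a search field and tab group with a switcher between the row mega list, search results and empty states.
# View models drive tab selection, Island or Creator searches, playtime-limit states and enter/exit analytics.
# Custom navigation tracks the last focused widget, adapting decoy bottom-bar buttons and the clear button to input method.
# Scrolling triggers mini-header transitions and optional lazy panel hydration, while tab switches animate background color.
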WidgetBlueprint /DiscoveryBrowser/Rows/WBP_Row_InputFieldWithTabs
animations [
 Anim_SearchScrollingFade {
  binding_count=1
  length_seconds=0.3
  object=&self:Anim_SearchScrollingFade
  sequence {bindings [Image_SoftEdgeGradient_Islands {
     id=#1
     possessed_class=Image
     tracks [{
       type=MovieSceneFloatTrack
       display_name=Render Opacity
       sections [{type=MovieSceneFloatSection, end_frame=6, start_frame=0, channels [None {type=MovieSceneScriptingFloatChannel, keys [[0,0],[6,1]]}]}]
      }]
    }]}
 }
 Anim_FadeColorChange {
  binding_count=1
  length_seconds=1
  object=&self:Anim_FadeColorChange
  sequence {bindings [WBP_Row_InputFieldWithTabs_C_47 {
     id=#2
     possessed_class=WBP_Row_InputFieldWithTabs_C
     tracks [{
       type=MovieSceneDoubleTrack
       display_name=Anim Output Background Color
       sections [{type=MovieSceneDoubleSection, end_frame=20, start_frame=0, channels [None {type=MovieSceneScriptingDoubleChannel, keys [[0,0],[20,1]]}]}]
      }]
    }]}
 }
]
graphs [
 BP_GetDesiredFocusTarget {nodes [
   Get Desired Focus Target()
   > Branch(Condition=AND Boolean(
     A=IsActivated!(self:Common Activatable Widget Object Reference) -> (ReturnValue:bool)
     B=Get IsBladeExpandContractAnimationFinished() -> (IsBladeExpandContractAnimationFinished:bool)
    ))
   n1 = Get SearchInputField() -> (SearchInputField:WBP UIKit Filtered Input Field Object Reference)
   > n2 = Branch(Condition=IsPressed!(self:Common Button Base Object Reference=Get Button_Clear!(self:WBP UIKit Filtered Input Field Object Reference=@n1) -> (Button_Clear:WBP UIKit Button Regular Object Reference)) -> (ReturnValue:bool))
   n3 = Get!(execute:Exec=@n2.else) -> (then:Exec, else:Exec, LastFocusedWidget:Widget Object Reference)
   n4 = Get LastFocusedWidget() -> (LastFocusedWidget:Widget Object Reference)
   > PrintString!(
    execute
    InString=Append!(A=Navigation: Last focused widget, B="Focus set to: ", C:String=GetDisplayName(Object=@n4)) -> (ReturnValue)
    bPrintToScreen
    bPrintToLog
    TextColor
    Duration=3
    Key
   ) -> (then)
   > Return Node!(execute, ReturnValue:Widget Object Reference=Get LastFocusedWidget())
   PrintString(execute=@n3.else, InString=Navigation:Focus is refreshed)
   n5 = Get InputField!(self:WBP UIKit Filtered Input Field Object Reference=Get SearchInputField()) -> (InputField:WBP UIKit Input Field Object Reference)
   > Return Node(ReturnValue=Get EditableText!(self:WBP UIKit Block Editable Text Object Reference=Get Block_EditableText!(self:WBP UIKit Input Field Base Object Reference=@n5) -> (Block_EditableText:WBP UIKit Block Editable Text Object Reference)) -> (EditableText:UIKit Editable Text Object Reference))
  ]}
 BindEvents {nodes [
   BindEvents() {is=title K2Node_FunctionEntry}
   n1 = Get TileView!(self:Fort Polymorphic Tile View Object Reference=Get WBP_Results_PolyTileView() -> (WBP_Results_PolyTileView:WBP Results Poly Tile View Object Reference)) -> (TileView:Fort Tile View Object Reference)
   > Bind Event to On Entry Generated(self:List View Base Object Reference=@n1, Delegate=Create Event()) {is=title K2Node_AddDelegate}
   > Bind Event to On List View Scrolled(self:List View Object Reference=@n1, Delegate=Create Event()) {is=title K2Node_AddDelegate}
   > Bind Event to Navigate Up Out Of Mega List(
    self:WBP Poly Mega List Object Reference=Get WBP_Poly_MegaList() -> (WBP_Poly_MegaList:WBP Poly Mega List Object Reference)
    Delegate=Create Event()
   ) {is=title K2Node_AddDelegate}
   > Bind Event to On Back Button Clicked(self:WBP Discover Surface Base Object Reference, Delegate=Create Event()) {is=title K2Node_AddDelegate}
   n2 = Get CommonInputSubsystem()
   > Bind Event to On Input Method Changed(self=@n2, Delegate=Create Event())
   > OnInputMethodChanged!(execute, self, NewInputType:ECommonInputType Enum=GetCurrentInputType(self=@n2)) -> (then)
   n3 = Get SearchInputField()
   > Bind Event to On Clicked(self:Common Button Base Object Reference=Get Button_Clear(self=@n3), Delegate=Create Event()) {is=title K2Node_AddDelegate}
   > Bind Event to On Search Added to Focus Path(self:WBP UIKit Filtered Input Field Object Reference=@n3, Delegate=Create Event()) {is=title K2Node_AddDelegate}
   > Bind Event to On List View Scrolled(
    self=Get ListView!(self:Fort Polymorphic List View Object Reference=Get WBP_Poly_MegaList()) -> (ListView:Common List View Object Reference)
    Delegate=Create Event()
   )
   > BindToAthenaFrontendTabChange!(execute, Context:Widget Object Reference=Self-Reference(), EventToBind:"Delegate (by ref)"=Create Event()) -> (then)
   > Bind Event to On List View Touch End(self:List View Object Reference=Get ListView(self=Get WBP_Poly_MegaList()), Delegate=Create Event()) {is=title K2Node_AddDelegate}
   > Bind Event to On Switcher List Active Widget Changed(self:WBP Discover Surface Base Object Reference, Delegate=Create Event()) {is=title K2Node_AddDelegate}
  ]}
 BindScrollOutEvent {nodes [
   n1 = BindScrollOutEvent!() -> (then, ActivityDiscoverView:Activity Discover View NEW VM Object Reference, Bind:bool=~) {is=title K2Node_FunctionEntry}
   > n2 = Branch(Condition=@n1.Bind)
   > Bind Event to On Transition Discover Out(self:Activity Discover View NEW VM Object Reference=@n1, Delegate=Create Event()) {is=title K2Node_AddDelegate}
   Unbind all Events from On Transition Discover Out!(execute:Exec=@n2.else, self:Activity Discover View NEW VM Object Reference=@n1) -> (then:Exec) {is=title K2Node_ClearDelegate}
  ]}
 BroadcastTabSearchSelectedIndex {nodes [BroadcastTabSearchSelectedIndex!() -> (then, SelectedTabIndex:int) {is=title K2Node_FunctionEntry}]}
 ClearSearch {nodes [
   ClearSearch() {is=title K2Node_FunctionEntry}
   > n1 = GetCurrentText!(execute, self:WBP UIKit Filtered Input Field Object Reference=Get SearchInputField()) -> (then, Text:Text)
   > n2 = Branch(Condition=NOT Boolean(A=TextIsEmpty(InText=@n1)))
   > ClearSearch(self:Discover Search Model Object Reference=Get FortDiscoverSearchViewModel() -> (FortDiscoverSearchViewModel:Discover Search Model Object Reference)) {is=title K2Node_CallFunction}
   > SetInputFieldText!(execute, self:WBP UIKit Filtered Input Field Object Reference=Get SearchInputField(), Text:Text) -> (then)
   > n3 = UpdateClearButtonVisibility(self:WBP UIKit Filtered Input Field Object Reference=Get SearchInputField())
   Return Node!(execute=[@n3.then, @n2.else])
  ]}
 EventGraph {nodes [
   "On Clicked (WBP_BottomBarHack_ClearResults)"!() -> (OutputDelegate, then, Button:Common Button Base Object Reference) {is=title K2Node_ComponentBoundEvent}
   n1 = Get TileView(self=Get WBP_Results_PolyTileView())
   > SetSelectedIndex!(execute, self:List View Object Reference=@n1, Index:int=0) -> (then)
   > ScrollToTop(self:List View Base Object Reference=@n1)
   > ClearSearch(self=Get FortDiscoverSearchViewModel())
   > SetInputFieldText(self=Get SearchInputField())
   > SetActiveWidget(
    self:Widget Switcher Object Reference=Get Switcher_List() -> (Switcher_List:Common Activatable Widget Switcher Object Reference)
    Widget=Select(
     Option 0:Widget Object Reference=Get OverlayList() -> (OverlayList:Overlay Object Reference)
     Option 1:Widget Object Reference=Get Switcher_NoResults() -> (Switcher_NoResults:Common Widget Switcher Legacy Object Reference)
     Index=Get IsPlayTime Out() -> (IsPlayTime Out:bool)
    ) -> (ReturnValue:Widget Object Reference)
   )
   > SelectTabByIndex!(
    execute
    self:UIKit Tab Group Object Reference=Get TabGroup_Search() -> (TabGroup_Search:WBP UIKit Tab Group Object Reference)
    Index:int=0
    bSuppressClickFeedback:bool=true
   ) -> (then, ReturnValue:bool)
   > SetBottomButtonVisibility!(
    execute
    self:WBP Bottom Bar Decoy Button Object Reference=Get WBP_BottomBarHack_ClearResults() -> (WBP_BottomBarHack_ClearResults:WBP Bottom Bar Decoy Button Object Reference)
    ShouldButtonBeVisible:bool=false
   ) -> (then)
   n2 = Get SearchInputField()
   n3 = Get EditableText(self=Get Block_EditableText(self=Get InputField(self=@n2)))
   > Set LastFocusedWidget(LastFocusedWidget:Widget Object Reference=@n3) -> (Output_Get:Widget Object Reference)
   > RequestRefreshFocus(self:Common Activatable Widget Object Reference)
   > ToggleVirtualKeyboard!(execute, self:Editable Text Object Reference=@n3, bShow:bool=true) -> (then)
   n4 = "On Text Committed (SearchInputField)"!() -> (OutputDelegate, then, CommitMethod:ETextCommit Enum=Default) {is=title K2Node_ComponentBoundEvent}
   > n5 = Switch on ETextCommit!(execute, Selection:ETextCommit Enum=@n4.CommitMethod) -> (Default:Exec, OnEnter:Exec, OnUserMovedFocus:Exec, OnCleared:Exec) {is=title K2Node_SwitchEnum}
   n6 = GetCurrentText(execute=@n5.OnEnter, self=Get SearchInputField())
   > Branch(Condition=NOT Boolean(A=TextIsEmpty(InText=@n6)))
   > FireScrollOutAnalytics()
   > ActionSearch(self:Discover Search Model Object Reference=Get FortDiscoverSearchViewModel())
   > SetBottomButtonVisibility(
    self=Get WBP_BottomBarHack_ClearResults()
    ShouldButtonBeVisible=Select!(
     MouseAndKeyboard:bool=false
     Gamepad:bool=true
     Touch:bool=true
     Count:bool=true
     Index:ECommonInputType Enum=Get Current Input Type() -> (Current Input Type:ECommonInputType Enum)
    ) -> (ReturnValue:bool)
   )
   > Branch(Condition=Get bIsActive() -> (bIsActive:bool))
   > n7 = Sequence()
   > Branch(Condition=Get SendSearchTabEnterAnalytics() -> (SendSearchTabEnterAnalytics:bool))
   > FireDiscoverTabEnteredAnalyticEvent(self:Discover Search Model Object Reference=Get FortDiscoverSearchViewModel())
   > Set SendSearchTabEnterAnalytics(SendSearchTabEnterAnalytics:bool=false) -> (Output_Get:bool)
   Set SendAnalyticsOnExit(execute=@n7.then_1, SendAnalyticsOnExit:bool=true) -> (Output_Get:bool)
   "On Clicked (DecoyButton_TabRightTriggerNav)"!() -> (OutputDelegate, then, Button:Common Button Base Object Reference) {is=title K2Node_ComponentBoundEvent}
   n8 = Get IsPlayTime Out()
   > n9 = Branch(Condition=@n8)
   Set bShouldFocusOnMegaListItem(execute=@n9.else, bShouldFocusOnMegaListItem:bool=true) -> (Output_Get:bool)
   > RightTabTriggerPressed()
   "On Clicked (DecoyButton_LeftTabTriggerNav)"!() -> (OutputDelegate, then, Button:Common Button Base Object Reference) {is=title K2Node_ComponentBoundEvent}
   > n10 = Branch(Condition=@n8)
   LeftTabTriggerPressed(execute=@n10.else)
   "On Swipe Up (Discovery_BlankTiles_SwipePanel)"() {is=title K2Node_ComponentBoundEvent}
   > Call On Back Button Clicked!(execute:Exec, self:WBP Discover Surface Base Object Reference) -> (then:Exec) {is=title K2Node_CallDelegate}
   n11 = "On Clicked (ViewTimeLimitsResults_Button)"!() -> (OutputDelegate, then, Button:Common Button Base Object Reference) {is=title K2Node_ComponentBoundEvent}
   RequestOpenTimeLimitsSidebar(
    execute=[@n11.then, @n12.then]
    self:Player Playtime View Model Object Reference=Get PlayerPlaytimeVM() -> (PlayerPlaytimeVM:Player Playtime View Model Object Reference)
   )
   n12 = "On Clicked (ViewTimeLimits_Button)"!() -> (OutputDelegate, then, Button:Common Button Base Object Reference) {is=title K2Node_ComponentBoundEvent}
   "On Tab Button Creation (TabGroup_Search)"!() -> (OutputDelegate, then, TabId:Name=None, TabButton:Common Button Base Object Reference) {is=title K2Node_ComponentBoundEvent}
   > SetPlayTimeState!(execute, self, IsPlayTimeOut:bool=Get IsPlayTime Out()) -> (then)
   n13 = "On Text Changed (SearchInputField)"!() -> (OutputDelegate, then, Text:Text) {is=title K2Node_ComponentBoundEvent}
   > Set SearchTerm!(
    execute
    SearchTerm:String="To String (Text)"(InText=@n13.Text)
    self:Discover Search Model Object Reference=Get FortDiscoverSearchViewModel()
   ) -> (then, Output_Get:String)
   n14 = "On Tab Button Creation (Tabs_Search_Results)"!() -> (OutputDelegate, then, TabId:Name=None, TabButton:Common Button Base Object Reference) {is=title K2Node_ComponentBoundEvent}
   > Populate Additional Tab Content!(execute, self, TabId:Name=@n14.TabId, TabButton:Common Button Base Object Reference=@n14.TabButton) -> (then)
   n15 = "On UIKit Tab Selected (Tabs_Search_Results)"!() -> (OutputDelegate, then, TabID:Name=None, TabVM:MVVM Base Viewmodel Object Reference) {is=title K2Node_ComponentBoundEvent}
   n16 = GetTabIndexByID!(
    self:UIKit Tab Group Object Reference=Get Tabs_Search_Results() -> (Tabs_Search_Results:WBP UIKit Tabs Object Reference)
    TabNameID:Name=@n15.TabID
   ) -> (ReturnValue:int=0)
   > Set with Broadcast SearchMode!(
    execute
    SearchMode:EDiscoverSearchMode Enum=Select!(Option 0:EDiscoverSearchMode Enum=Island, Option 1:EDiscoverSearchMode Enum=Creator, Index:int=@n16) -> (ReturnValue:EDiscoverSearchMode Enum)
   ) -> (then, Output_Get:EDiscoverSearchMode Enum=Island)
   > SetRowHeaderText!(execute, self, TabIndex:int=@n16) -> (then)
   > ScrollToTop(self=Get TileView(self=Get WBP_Results_PolyTileView()))
   > PlayAnimationReverse(InAnimation=Get Anim_SearchScrollingFade() -> (Anim_SearchScrollingFade:Widget Animation Object Reference))
   "On Clicked (BackToDisco_Button)"!() -> (OutputDelegate, then, Button:Common Button Base Object Reference) {is=title K2Node_ComponentBoundEvent}
   > GoBackToDisco()
   "On Active Widget Index Changed BP (SearchResult_Switcher_0)"!() -> (OutputDelegate, then, ActiveWidget:Widget Object Reference, ActiveIndex:int=0) {is=title K2Node_ComponentBoundEvent}
   > Branch(Condition=Get HasListLoaded() -> (HasListLoaded:bool))
   > RequestRefreshFocus()
   > PrintString(InString=Navigation: On active widget index changed SearchResults_Switcher - refresh focus)
   "On Clicked (WBP_BottomBarHack_ClearBadges)"!() -> (OutputDelegate, then, Button:Common Button Base Object Reference) {is=title K2Node_ComponentBoundEvent}
   > ClearAllNotifications(self:Discover Surface View Model Object Reference=Get FortDiscoverSurfaceViewModel() -> (FortDiscoverSurfaceViewModel:Discover Surface View Model Object Reference))
   > SetBottomButtonVisibility(
    self=Get WBP_BottomBarHack_ClearBadges() -> (WBP_BottomBarHack_ClearBadges:WBP Bottom Bar Decoy Button Object Reference)
    ShouldButtonBeVisible=false
   )
   n17 = "On UIKit Tab Selected (TabGroup_Search)"!() -> (OutputDelegate, then, TabID:Name=None, TabVM:MVVM Base Viewmodel Object Reference) {is=title K2Node_ComponentBoundEvent}
   > Set CurrentTabId(CurrentTabId:Name=@n17.TabID) -> (Output_Get:Name=None)
   > n18 = Branch(Condition=Get bIsActive())
   > Branch(Condition=IsValid(Object=Get FortDiscoverSurfaceViewModel()))
   > FireScrollOutAnalytics()
   > SetupTabSurface()
   > ClearSearch(self:Self Object Reference)
   > EndInertialScrolling(self:List View Base Object Reference=Get ListView(self=Get WBP_Poly_MegaList()))
   n19 = Get WBP_BottomBarHack_ClearResults()
   > n20 = Branch(Condition=IsVisible(self=@n19))
   > n21 = SetBottomButtonVisibility(self=@n19, ShouldButtonBeVisible=false)
   Call Broadcast Tab Search Selected Index!(
    execute:Exec=[@n21.then, @n20.else]
    self:WBP Row Input Field With Tabs Object Reference
    SelectedTabIndex:int=GetSelectedTabIndex!(self:UIKit Tab Group Object Reference=Get TabGroup_Search()) -> (ReturnValue:int=0)
   ) -> (then:Exec) {is=title K2Node_CallDelegate}
   > SetBackgroundColorEnums!(execute, self, Tab Index:int=GetSelectedTabIndex(self=Get TabGroup_Search()) -> (ReturnValue=0)) -> (then)
   > SetAnim_OutputBackgroundColor!(execute, self, Anim Progress:double=Get Anim_OutputBackgroundColor() -> (Anim_OutputBackgroundColor:double)) -> (then)
   > PlayAnimation(InAnimation=Get Anim_FadeColorChange() -> (Anim_FadeColorChange:Widget Animation Object Reference))
   > Set CachedTabIndex(CachedTabIndex:int=GetSelectedTabIndex(self=Get TabGroup_Search()) -> (ReturnValue=0)) -> (Output_Get:int=0)
   Call On Request Navigate to Discover!(execute:Exec=@n18.else, self:WBP Row Input Field With Tabs Object Reference) -> (then:Exec) {is=title K2Node_CallDelegate}
   "On Clicked (WBP_UIKit_ViewMore_ArrowButton)"!() -> (OutputDelegate, then, Button:Common Button Base Object Reference) {is=title K2Node_ComponentBoundEvent}
   > Call On Request Navigate to Discover()
   n22 = OnSearchTIleViewEntriesGenerated!() -> (OutputDelegate, then, Widget:User Widget Object Reference) {is=title K2Node_CustomEvent}
   > n23 = Cast To WBP_IslandTile_Minimal!(execute, Object=@n22.Widget) -> (then, CastFailed, AsWBP Island Tile Minimal:WBP Island Tile Minimal Object Reference) {is=title K2Node_DynamicCast}
   > UpdateTileSizeAndShapeThenTileView!(
    execute
    self
    TileBase:WBP Discover Tile Base Object Reference=Get WBP_Discover_Tile_Base!(self:WBP Island Tile Minimal Object Reference=@n23) -> (WBP_Discover_Tile_Base:WBP Discover Tile Base Object Reference)
   ) -> (then)
   n24 = Cast To WBP_Discover_Tile_Button!(execute=@n23.CastFailed, Object=@n22.Widget) -> (then, CastFailed, AsWBP Discover Tile Button:WBP Discover Tile Button Object Reference) {is=title K2Node_DynamicCast}
   > UpdateTileSizeAndShapeThenTileView(TileBase=Get WBP_Discover_Tile_Base(self:WBP Discover Tile Button Object Reference=@n24))
   n25 = OnSearchTileViewListScrolled!() -> (OutputDelegate, then, ItemOffset:double=~, DistanceRemaining:double=~) {is=title K2Node_CustomEvent}
   > Branch(Condition=NOT Boolean(A="Nearly Equal (Float)"(A=@n25.DistanceRemaining, B=0, ErrorTolerance=0)))
   > PlaySearchResultsFadeAnimation!(execute, self, SearchResultsScrollOffset:double=@n25.ItemOffset) -> (then)
   n26 = OnMegaListSwitcherActiveWidgetChanged!() -> (OutputDelegate, then, ActiveWidget:Widget Object Reference) {is=title K2Node_CustomEvent}
   n27 = GetScrollIsAtTopOfList!() -> (bIsAtTop:bool)
   n28 = Get Current Input Type()
   n29 = Get OverlayList()
   > Branch(Condition=AND Boolean!(
     A="Equal (Object)"(A=@n26.ActiveWidget, B=@n29)
     B=NOT Boolean(A=@n27)
     C:bool="Not Equal (Enum)"!(A:ECommonInputType Enum=@n28, B:ECommonInputType Enum=Gamepad) -> (ReturnValue:bool)
    ) -> (ReturnValue))
   > DelayUntilNextFrame!(execute) -> (then)
   > ScrollToTop(self=Get ListView(self=Get WBP_Poly_MegaList()))
   n30 = OnInputMethodChanged!() -> (OutputDelegate, then, NewInputType:ECommonInputType Enum) {is=title K2Node_CustomEvent}
   > Set Current Input Type(Current Input Type:ECommonInputType Enum=@n30.NewInputType) -> (Output_Get:ECommonInputType Enum)
   > UpdateTabPromptVisibility()
   n31 = "Equal (Object)"(
    A=GetActiveWidget!(self:Widget Switcher Object Reference=Get Switcher_List()) -> (ReturnValue:Widget Object Reference)
    B=Get NamedSlot_SwitcherListContent() -> (NamedSlot_SwitcherListContent:Named Slot Object Reference)
   )
   > SetBottomButtonVisibility(
    self=Get WBP_BottomBarHack_ClearResults()
    ShouldButtonBeVisible=Select!(MouseAndKeyboard:bool=false, Gamepad:bool=@n31, Touch:bool=@n31, Count:bool=@n31, Index:ECommonInputType Enum=Get Current Input Type()) -> (ReturnValue:bool)
   )
   > n32 = GetCurrentText(self=Get SearchInputField())
   n33 = NOT Boolean(A=TextIsEmpty(InText=@n32))
   n34 = "Equal (Enum)"(A=Get Current Input Type(), B=MouseAndKeyboard)
   > SetVisibility(
    self=Get Button_Clear(self=Get SearchInputField())
    InVisibility=Select!(Option 0:ESlateVisibility Enum=Collapsed, Option 1:ESlateVisibility Enum=Visible, Index:bool=AND Boolean!(A=@n33, B=@n34) -> (ReturnValue)) -> (ReturnValue:ESlateVisibility Enum)
   )
   > RequestRefreshFocus()
   OnNavigateOutOfMegalist() {is=title K2Node_CustomEvent}
   > NavigateOutOfMegalist()
   OnInputFieldButtonClearClicked!() -> (OutputDelegate, then, Button:Common Button Base Object Reference) {is=title K2Node_CustomEvent}
   > ClearSearch(self:Discover Search Model Object Reference=Get FortDiscoverSearchViewModel())
   > SetInputFieldText(self=Get SearchInputField())
   > UpdateClearButtonVisibility(self=Get SearchInputField())
   > SetActiveWidget(
    self=Get Switcher_List()
    Widget=Select(Option 0:Widget Object Reference=Get OverlayList(), Option 1:Widget Object Reference=Get Switcher_NoResults(), Index=Get IsPlayTime Out()) -> (ReturnValue:Widget Object Reference)
   )
   > SelectTabByIndex(self=Get TabGroup_Search(), Index=0, bSuppressClickFeedback=true)
   n35 = Get InputField(self=Get SearchInputField())
   > Set LastFocusedWidget(LastFocusedWidget=Get EditableText(self=Get Block_EditableText(self=@n35)))
   > RequestRefreshFocus()
   OnBackButtonClicked() {is=title K2Node_CustomEvent}
   > n36 = Switch on Int!(execute:Exec, Selection:int=GetActiveWidgetIndex!(self:Widget Switcher Object Reference=Get Switcher_List()) -> (ReturnValue:int=0)) -> (0:Exec, 1:Exec, 2:Exec, 3:Exec) {bHasDefaultPin=false}
   ClearSearch(execute=@n36.3, self=Get FortDiscoverSearchViewModel())
   > SetInputFieldText(self=Get SearchInputField())
   > n37 = SetActiveWidget(self=Get Switcher_List(), Widget=Select(Option 0=Get OverlayList(), Option 1=Get Switcher_NoResults(), Index=Get IsPlayTime Out()))
   n38 = Get TabGroup_Search()
   n39 = Get SearchInputField()
   n40 = Branch(
    execute=[@n37.then, @n36.0, @n36.1, @n36.2]
    Condition=OR Boolean!(A=HasAnyUserFocus!(self:Widget Object Reference=@n39) -> (ReturnValue:bool), B=HasAnyUserFocus(self=@n38), C:bool=GetScrollIsAtTopOfList()) -> (ReturnValue)
   )
   > Set LastFocusedWidget()
   n41 = Get AnimationSpacer() -> (AnimationSpacer:Spacer Object Reference)
   > n42 = Branch(Condition=IsZero2D!(A:"Vector 2D Structure (by ref)"=Get Size!(self:Spacer Object Reference=@n41) -> (Size:Vector 2D Structure)) -> (ReturnValue:bool))
   > n43 = ResetHeader(self:WBP Discover Surface Base Object Reference)
   Call On Back Button Hack Clicked!(
    execute:Exec=[@n43.then, @n42.else]
    self:WBP Row Input Field With Tabs Object Reference
    ButtonObject:WBP Bottom Bar Decoy Button Object Reference=Get WBP_BottomBarHack_Back() -> (WBP_BottomBarHack_Back:WBP Bottom Bar Decoy Button Object Reference)
   ) -> (then:Exec) {is=title K2Node_CallDelegate}
   > DeactivateWidget()
   RefreshCurrentPageToTop(execute=@n40.else)
   OnInputFieldFocused() {is=title K2Node_CustomEvent}
   > n44 = Branch(Condition=Get bIsActive())
   Call On Request Navigate to Discover(execute=@n44.else)
   n45 = OnMegaListScrolled!() -> (OutputDelegate, then, ItemOffset:float=~, DistanceRemaining:float=~) {is=title K2Node_CustomEvent}
   > Set MegalistItem Offset(MegalistItem Offset:float=@n45.ItemOffset) -> (Output_Get:float)
   > n46 = Branch(Condition=Get bUseLazyLoadSurface!(self:Fort Runtime Options Object Reference=GetRuntimeOptions()) -> (bUseLazyLoadSurface:bool))
   > n47 = Request Panels Hydration()
   Branch(execute=[@n47.then, @n46.else], Condition="Equal (Enum)"(A=Get Current Input Type(), B=Touch))
   > n48 = Branch(Condition="Equal (Float)"(A=Get MegalistItem Offset() -> (MegalistItem Offset:float)))
   n49 = Get AnimationSpacer()
   Branch(execute=[@n65.then, @n48.then], Condition=IsZero2D(A=Get Size(self=@n49)))
   > TransitionToMiniHeader!(execute, self:WBP Discover Surface Base Object Reference, IsMiniHeaderWithFullList:bool=false) -> (then)
   HandlePossibleTransitionToMiniHeader(execute=@n48.else)
   n50 = OnFrontendTabChange!() -> (OutputDelegate, then, LeavingFeature:EFortUIFeature Enum=ShowHome, EnteringFeature:EFortUIFeature Enum=ShowHome) {is=title K2Node_CustomEvent}
   > Branch(Condition=AND Boolean(A="Equal (Enum)"(A:EFortUIFeature Enum=@n50.EnteringFeature, B:EFortUIFeature Enum=ShowPlay), B=IsActivated()))
   > ClearSearch(self=Get FortDiscoverSearchViewModel())
   > SetInputFieldText(self=Get SearchInputField())
   > SetActiveWidget(self=Get Switcher_List(), Widget=Select(Option 0=Get OverlayList(), Option 1=Get Switcher_NoResults(), Index=Get IsPlayTime Out()))
   > Set bShouldFocusOnMegaListItem(bShouldFocusOnMegaListItem=true)
   n51 = Get TabGroup_Search()
   > n52 = Branch(Condition="Equal (Integer)"(A=GetSelectedTabIndex(self=@n51) -> (ReturnValue=0)))
   n53 = Get FortDiscoverSurfaceViewModel()
   > n54 = Branch(Condition=IsSurfaceDirty!(self:Discover Surface View Model Object Reference=@n53) -> (ReturnValue:bool))
   > n55 = UpdateCachedSurfaceData!(execute, self:Discover Surface View Model Object Reference=@n53, bShouldForceUpdate:bool=true) -> (then)
   SetNavOnTabSwitched!(execute=[@n55.then, @n54.else], self, LoadingState:EDiscoverLoadingState Enum=Loaded) -> (then)
   SelectTabByIndex(execute=@n52.else, self=Get TabGroup_Search(), Index=0, bSuppressClickFeedback=true)
   OnMegaListTouchEnd() {is=title K2Node_CustomEvent}
   n56 = Get MegalistItem Offset()
   n57 = Get Current Input Type()
   n58 = GetOverscroll!(self:List View Base Object Reference=Get ListView(self=Get WBP_Poly_MegaList())) -> (ReturnValue:float)
   > Branch(Condition=AND Boolean!(
     A="Equal (Enum)"(A:ECommonInputType Enum=@n57, B:ECommonInputType Enum=Touch)
     B="Nearly Equal (Float)"(A=@n56, B=0, ErrorTolerance=0.01)
     C:bool="float < float"(A=@n58)
    ) -> (ReturnValue))
   > Call On Back Button Clicked()
   n59 = Event PreConstruct()
   > Parent: PreConstruct(IsDesignTime=@n59.IsDesignTime)
   > n60 = Sequence()
   > UpdateVisual()
   > SetCompactUIStyle(self:WBP Discover Surface Base Object Reference) {is=title K2Node_CallFunction}
   > SetResultsTerm!(execute, self, ResultsTerm:String=$Search Result) -> (then)
   > PopulateSearchResultTabs()
   > Set Material Instance Parameters() {is=title K2Node_CallFunction}
   > SetMegaListMinMaxDesiredHeight()
   > SetBottomButtonVisibility(
    self=[
     Get WBP_BottomBarHack_Back()
     Get WBP_BottomBarHack_ClearResults()
     Get WBP_BottomBarHack_BackToTop() -> (WBP_BottomBarHack_BackToTop:WBP Bottom Bar Decoy Button Object Reference)
    ]
    ShouldButtonBeVisible=false
   )
   > Set bUseScrollOffsetOverrideRedux!(execute, bUseScrollOffsetOverrideRedux:bool=true, self:WBP Poly Mega List Object Reference=Get WBP_Poly_MegaList()) -> (then, Output_Get:bool)
   > PopulateSpacer!(execute, self:WBP Poly Mega List Object Reference=Get WBP_Poly_MegaList(), SpacerSize:Vector 2D Structure=(X=0,Y=250)) -> (then)
   > Set DeactivateWIdgetOnClickedBackButton(DeactivateWIdgetOnClickedBackButton:bool=false) -> (Output_Get:bool)
   > OverrideTabInputAction()
   Branch(execute=@n60.then_1, Condition=@n59.IsDesignTime)
   Event Construct()
   > Parent: Construct()
   > BindEvents() {is=title K2Node_CallFunction}
   > Set with Broadcast SearchMode(SearchMode=Island) -> (Output_Get=Island)
   > n61 = Create Widget!(
    execute:Exec
    Class:User Widget Class Reference=/DiscoveryBrowser/Tiles/WBP_Discover_Tile_Base.WBP_Discover_Tile_Base_C
    OwningPlayer:Player Controller Object Reference
   ) -> (then:Exec, ReturnValue:WBP Discover Tile Base Object Reference)
   > Set DiscoverTileBaseReference(DiscoverTileBaseReference:WBP Discover Tile Base Object Reference=@n61) -> (Output_Get:WBP Discover Tile Base Object Reference)
   > Set LastFocusedWidget()
   > SetupNavigationRules()
   > SetInitialDiscoverTabIndex(self:Discover Surface View Model Object Reference=Get FortDiscoverSurfaceViewModel())
   > Init(self:Player Playtime View Model Object Reference=Get PlayerPlaytimeVM())
   Event OnHeaderAnimationFinished() {is=title K2Node_Event}
   > Branch(Condition=Get RequestRefreshFocusAfterAnimation() -> (RequestRefreshFocusAfterAnimation:bool))
   > Set RequestRefreshFocusAfterAnimation(RequestRefreshFocusAfterAnimation:bool=false) -> (Output_Get:bool)
   > RequestRefreshFocus()
   Event Destruct()
   > Parent: Destruct()
   > UnbindEvents()
   Event On Activated()
   > OnActivated()
   n62 = GetOwningPlayer()
   > Branch(Condition=AND Boolean!(
     A=Get IsFirstActivation() -> (IsFirstActivation:bool)
     B=GetDefaultDiscoverLandingExperimentValue!(PlayerController:Player Controller Object Reference=@n62) -> (ReturnValue:bool)
    ) -> (ReturnValue))
   > DelayUntilNextFrame()
   > GoBackToDisco()
   > Set IsFirstActivation(IsFirstActivation:bool=false) -> (Output_Get:bool)
   > Set LastFocusedWidget()
   > Set with Broadcast IsBladeExpandContractAnimationFinished!(execute, IsBladeExpandContractAnimationFinished:bool=true) -> (then, Output_Get:bool)
   Event On Deactivated()
   > SetSelectedDiscoverTabIndex!(
    execute
    self:Discover Surface View Model Object Reference=Get FortDiscoverSurfaceViewModel()
    InSelectionIndex:int=Get CachedTabIndex() -> (CachedTabIndex:int=0)
   ) -> (then)
   > DeselectAll(self:UIKit Tab Group Object Reference=Get TabGroup_Search())
   n63 = Event OnUserNavigatedMegalist!() -> (OutputDelegate, then, IsDirectionUp:bool, NewDesiredOuterIndex:int=0) {is=title K2Node_Event}
   > n64 = UpdateScrollOffset!(
    execute
    self:WBP Poly Mega List Object Reference=Get WBP_Poly_MegaList()
    IsDirectionUp:bool=@n63.IsDirectionUp
    RowIndex:int=@n63.NewDesiredOuterIndex
   ) -> (then, CurrentPage:int=0)
   > n65 = Branch(Condition="Equal (Integer)"(A=@n64))
   HandlePossibleTransitionToMiniHeader(execute=@n65.else)
  ]}
 FillOutList {nodes [
   n1 = FillOutList!() -> (then, ProvidersArray:Array of Discover Provider View Model Object References) {is=title K2Node_FunctionEntry}
   > Parent: FillOutList!(execute, ProvidersArray:Array of Discover Provider View Model Object References=@n1) -> (then) {is=title K2Node_CallParentFunction}
   n2 = GetActiveWidget(self=Get Switcher_List())
   > n3 = Branch(Condition="Equal (Object)"(A=Get Last Active Widget() -> (Last Active Widget:Widget Object Reference), B=@n2))
   > n4 = OnMegaListSwitcherActiveWidgetChanged!(execute, self, ActiveWidget:Widget Object Reference=@n2) -> (then) {is=title K2Node_CallFunction}
   Return Node(execute=[@n4.then, @n3.else])
  ]}
 FillOutSearchResults {nodes [
   n1 = FillOutSearchResults!() -> (then, SearchArray:Array of Discover Tile Item View Model Object References) {is=title K2Node_FunctionEntry}
   > ClearListItems(self=Get TileView(self=Get WBP_Results_PolyTileView()))
   > n2 = Set with Broadcast IsTileViewSearchResultsNotEmpty!(
    execute
    IsTileViewSearchResultsNotEmpty:bool=Is Not Empty!(TargetArray:Array of Discover Tile Item View Model Object References=Get SearchArray() -> (SearchArray:Array of Discover Tile Item View Model Object References)) -> (ReturnValue:bool)
   ) -> (then, Output_Get:bool)
   > n3 = Branch(Condition=@n2)
   > UpdateTileSizeAndShapeThenTileView(TileBase=Get DiscoverTileBaseReference() -> (DiscoverTileBaseReference:WBP Discover Tile Base Object Reference))
   > n4 = Set List Items!(
    execute
    self:List View Object Reference=Get TileView(self=Get WBP_Results_PolyTileView())
    InListItems:Array of Object References=Get SearchArray()
   ) -> (then) {is=title K2Node_CallFunction}
   Reroute Node(InputPin=[@n4.then, @n3.else])
   Branch(Condition=Get bIsActive())
   > n5 = Sequence()
   > Branch(Condition=Get SendSearchTabEnterAnalytics())
   > FireDiscoverTabEnteredAnalyticEvent(self=Get FortDiscoverSearchViewModel())
   > Set SendSearchTabEnterAnalytics(SendSearchTabEnterAnalytics=false)
   Set SendAnalyticsOnExit(execute=@n5.then_1, SendAnalyticsOnExit=true)
   Reroute Node(InputPin:Array of Discover Tile Item View Model Object References=@n1) -> (OutputPin:Array of Discover Tile Item View Model Object References)
  ]}
 FillOutTabsData {nodes [
   n1 = FillOutTabsData!() -> (then, TabsData:Array of Discover Tile Item View Model Object References) {is=title K2Node_FunctionEntry}
   > RemoveAllTabs(self:Common Tab List Widget Base Object Reference=Get TabGroup_Search())
   > Branch(Condition=Is Not Empty(TargetArray=@n1))
   > n2 = For Each Loop(Array:Array of Discover Tile Item View Model Object References=@n1) -> (Array Element:Discover Tile Item View Model Object Reference)
   > AddTab!(
    execute
    self:WBP UIKit Tab Group Base Object Reference=Get TabGroup_Search()
    Tab VM:MVVM Base Viewmodel Object Reference=@n2
    Tab ID:Name=String To Name(InString=GetDisplayName(Object=@n2))
    Index:int=@n2.Array Index
   ) -> (then)
  ]}
 FireExitTabAnalytics {nodes [
   n1 = FireExitTabAnalytics!() -> (then, ActiveWidget:Widget Object Reference) {is=title K2Node_FunctionEntry}
   > Set Last Active Widget(Last Active Widget:Widget Object Reference=@n1) -> (Output_Get:Widget Object Reference)
  ]}
 FireScrollOutAnalytics {nodes [
   FireScrollOutAnalytics() {is=title K2Node_FunctionEntry}
   > Branch(Condition=Get SendAnalyticsOnExit() -> (SendAnalyticsOnExit:bool))
   > Set SendAnalyticsOnExit(SendAnalyticsOnExit=false)
   > n1 = Branch(Condition="Equal (Object)"(A=Get Last Active Widget(), B=Get NamedSlot_SwitcherListContent()))
   > FireDiscoverTabExitedAnalyticEvent(self:Discover Search Model Object Reference=Get FortDiscoverSearchViewModel())
   FireDiscoverTabExitedAnalyticEvent(execute=@n1.else, self:Discover Surface View Model Object Reference=Get FortDiscoverSurfaceViewModel())
  ]}
 FocusOnSearchField {nodes [
   FocusOnSearchField() {is=title K2Node_FunctionEntry}
   > Set bShouldFocusOnMegaListItem(bShouldFocusOnMegaListItem=false)
   n1 = Get InputField(self=Get SearchInputField())
   > Set LastFocusedWidget(LastFocusedWidget=Get EditableText(self=Get Block_EditableText(self=@n1)))
   > RequestRefreshFocus()
  ]}
 FromSearchResultTabsToTileResults {nodes [
   FromSearchResultTabsToTileResults!() -> (then, Navigation:EUINavigation Enum) {is=title K2Node_FunctionEntry}
   > Branch(Condition=Get IsTileViewSearchResultsNotEmpty() -> (IsTileViewSearchResultsNotEmpty:bool))
   n1 = Get TileView(self=Get WBP_Results_PolyTileView())
   > SetSelectedIndex(self=@n1, Index=0)
   > NavigateToIndex!(execute, self:List View Object Reference=@n1, Index:int=0) -> (then)
   > Return Node!(execute, ReturnValue:Widget Object Reference)
  ]}
 FromTabsToMegaListOrSearch {nodes [
   FromTabsToMegaListOrSearch!() -> (then, Navigation:EUINavigation Enum) {is=title K2Node_FunctionEntry}
   > n1 = Switch on Int(Selection=GetActiveWidgetIndex(self=Get Switcher_List()) -> (ReturnValue=0)) {bHasDefaultPin=false}
   > Branch(Condition=NOT Boolean(A=Get IsListEmpty() -> (IsListEmpty:bool)))
   > PrintString(InString=Navigation: nav down from tabs to mega list)
   n2 = Get WBP_Poly_MegaList()
   > Set bFirstRowSnapIsZeroIndex!(execute, bFirstRowSnapIsZeroIndex:bool=true, self:WBP Poly Mega List Object Reference=@n2) -> (then, Output_Get:bool)
   n3 = Get ListView(self=@n2)
   > SetSelectedIndex(self=@n3, Index=0)
   > NavigateToIndex(self=@n3, Index=0)
   > Set LastFocusedWidget(LastFocusedWidget=Get ListView(self=@n2))
   PrintString(execute=@n1.1, InString=Navigation: nav down from tabs to disco button)
   > Return Node(ReturnValue=Select(
     Option 0=Get BackToDisco_Button() -> (BackToDisco_Button:WBP UIKit Button Loud Object Reference)
     Option 1=Get ViewTimeLimits_Button() -> (ViewTimeLimits_Button:WBP UIKit Button Regular Object Reference)
     Index=Get IsPlayTime Out()
    ))
   Return Node(execute=@n1.2)
   PrintString(execute=@n1.3, InString=Navigation: nav down from tabs to search tabs)
   > n4 = Branch(Condition=AND Boolean(A=Get IsTileViewSearchResultsNotEmpty(), B=Get IsPlayTime Out()))
   n5 = Get TileView(self=Get WBP_Results_PolyTileView())
   > SetSelectedIndex(self=@n5, Index=0)
   > NavigateToIndex(self=@n5, Index=0)
   > Return Node()
   n6 = Get Tabs_Search_Results()
   n7 = GetActiveTab!(self:Common Tab List Widget Base Object Reference=@n6) -> (ReturnValue:Name=None)
   Return Node(
    execute=@n4.else
    ReturnValue=Select(
     Option 0=GetTabButtonBaseByID!(self:Common Tab List Widget Base Object Reference=@n6, TabNameID:Name=@n7) -> (ReturnValue:Common Button Base Object Reference)
     Option 1=Get ViewTimeLimitsResults_Button() -> (ViewTimeLimitsResults_Button:WBP UIKit Button Regular Object Reference)
     Index=Get IsPlayTime Out()
    )
   )
  ]}
 GetActiveTabButtonObject {nodes [
   GetActiveTabButtonObject() {is=title K2Node_FunctionEntry}
   n1 = Get TabGroup_Search()
   > Return Node!(
    execute
    SearchTabButton:Common Button Base Object Reference=GetTabButtonBaseByID(self=@n1, TabNameID=GetActiveTab(self=@n1) -> (ReturnValue=None))
   )
  ]}
 GetNoResultCreatorSearchSwitcherState {nodes [
   n1 = GetNoResultCreatorSearchSwitcherState!() -> (then, CreatorResultCount:int, IsTileViewSearchResultsNotEmpty:bool=~) {is=title K2Node_FunctionEntry}
   > Return Node!(execute, ShowEmptyState:bool=AND Boolean(A="Equal (Integer)"(A=@n1), B=NOT Boolean(A=@n1.IsTileViewSearchResultsNotEmpty)))
  ]}
 GetNoResultIslandSearchSwitcherState {nodes [
   n1 = GetNoResultIslandSearchSwitcherState!() -> (then, IslandResultCount:int, IsTileViewSearchResultsNotEmpty:bool=~) {is=title K2Node_FunctionEntry}
   > Return Node(ShowEmptyState=AND Boolean(A="Equal (Integer)"(A=@n1), B=NOT Boolean(A=@n1.IsTileViewSearchResultsNotEmpty)))
  ]}
 GetPlayTimeState {nodes [
   n1 = GetPlayTimeState!() -> (then, PlayTimeState:EPlaytimeState Enum=Unblocked) {is=title K2Node_FunctionEntry}
   > Return Node!(execute, IsPlayTimeOut:bool="Equal (Enum)"(A:EPlaytimeState Enum=@n1, B:EPlaytimeState Enum=PlayTime_Blocked))
  ]}
 GetScrollIsAtTopOfList {nodes [
   GetScrollIsAtTopOfList() {is=title K2Node_FunctionEntry}
   n1 = GetScrollOffset!(self:List View Base Object Reference=Get TileView(self=Get WBP_Results_PolyTileView())) -> (ReturnValue:float)
   n2 = GetScrollOffset(self=Get ListView(self=Get WBP_Poly_MegaList()))
   n3 = GetActiveWidget(self=Get Switcher_List())
   n4 = Get NamedSlot_SwitcherListContent()
   > Return Node!(
    execute
    bIsAtTop:bool=Select(
     Option 0:bool="Nearly Equal (Float)"(A=@n2, B=0, ErrorTolerance=0)
     Option 1:bool="Nearly Equal (Float)"(A=@n1, B=0, ErrorTolerance=0)
     Index="Equal (Object)"(A=@n3, B=@n4)
    ) -> (ReturnValue:bool)
   )
  ]}
 GetShowCreateTimeEmptyStateSearchResults {nodes [
   n1 = GetShowCreateTimeEmptyStateSearchResults!() -> (then, IsPlayTimeOut:bool=~, IslandCount:int) {is=title K2Node_FunctionEntry}
   > Return Node!(execute, bShowEmptyState:bool=AND Boolean(A=@n1, B="integer <= integer"(A=@n1.IslandCount)))
  ]}
 GoBackToDisco {nodes [
   GoBackToDisco() {is=title K2Node_FunctionEntry}
   > SelectTabByIndex(self=Get TabGroup_Search(), Index=0, bSuppressClickFeedback=true)
   n1 = Get TabGroup_Search()
   > Set LastFocusedWidget(LastFocusedWidget=GetTabButtonBaseByID(self=@n1, TabNameID=GetActiveTab(self=@n1) -> (ReturnValue=None)))
   > RequestRefreshFocus()
  ]}
 HandlePossibleTransitionToMiniHeader {nodes [
   HandlePossibleTransitionToMiniHeader() {is=title K2Node_FunctionEntry}
   > Branch(Condition=NOT Boolean(A=Get Is Row Lock to Top() -> (Is Row Lock to Top:bool)))
   > TransitionToMiniHeader(IsMiniHeaderWithFullList=true)
  ]}
 LeftTabTriggerPressed {nodes [
   LeftTabTriggerPressed() {is=title K2Node_FunctionEntry}
   n1 = Get TabGroup_Search()
   > n2 = Branch(Condition="Equal (Integer)"(A=GetSelectedTabIndex(self=@n1) -> (ReturnValue=0)))
   n3 = Get LastFocusedWidget()
   n4 = IsValid(Object=@n3)
   n5 = Get SearchInputField()
   n6 = Get EditableText(self=Get Block_EditableText(self=Get InputField(self=@n5)))
   > n7 = Branch(Condition=OR Boolean!(A=NOT Boolean(A=@n4), B="Equal (Object)"(A=@n3, B=@n6)) -> (ReturnValue))
   > Set bShouldFocusOnMegaListItem(bShouldFocusOnMegaListItem=true)
   n8 = Get TabGroup_Search()
   > n9 = SelectTabByIndex(
    self=@n8
    Index=int - int(A=GetTabCount!(self:Common Tab List Widget Base Object Reference=@n8) -> (ReturnValue:int=0))
    bSuppressClickFeedback=true
   )
   TransitionToMiniHeader(execute=[@n9.then, @n11.then, @n10.then], IsMiniHeaderWithFullList=false)
   n10 = FocusOnSearchField(execute=@n7.else) {is=title K2Node_CallFunction}
   Set bShouldFocusOnMegaListItem(execute=@n2.else, bShouldFocusOnMegaListItem=true)
   > n11 = SelectPreviousTab!(execute, self:UIKit Tab Group Object Reference=Get TabGroup_Search(), bSuppressClickFeedback:bool=true) -> (then)
  ]}
 Navigate to discover_Touch {nodes [
   Navigate to discover_Touch!() -> (then, MyGeometry:Geometry Structure, MouseEvent:"Pointer Event Structure (by ref)") {is=title K2Node_FunctionEntry}
   > Branch(Condition=IsUsingTouch(PlayerController=GetOwningPlayer()))
   > Call On Request Navigate to Discover()
   > Return Node!(execute, ReturnValue:Event Reply Structure=Handled())
  ]}
 NavigateOutOfMegalist {nodes [
   NavigateOutOfMegalist() {is=title K2Node_FunctionEntry}
   n1 = Get InputField(self=Get SearchInputField())
   > Set LastFocusedWidget(LastFocusedWidget=Get EditableText(self=Get Block_EditableText(self=@n1)))
   > RequestRefreshFocus()
   > ResetHeader()
   > PrintString(InString=Navigation: Out of mega list)
   n2 = Get WBP_BottomBarHack_ClearResults()
   > Branch(Condition=IsVisible(self=@n2))
   > SetBottomButtonVisibility(self=@n2, ShouldButtonBeVisible=false)
  ]}
 NavigationOutOfSearchRow {nodes [
   NavigationOutOfSearchRow!() -> (then, Navigation:EUINavigation Enum) {is=title K2Node_FunctionEntry}
   > ClearSearch(self:Self Object Reference)
   > Set LastFocusedWidget()
   > PrintString(InString=Navigation: out of search row)
   > Call On List Scrolled Out!(execute:Exec, self:WBP Row Input Field With Tabs Object Reference) -> (then:Exec) {is=title K2Node_CallDelegate}
   > Return Node(ReturnValue:Widget Object Reference)
  ]}
 OnActivated {nodes [
   OnActivated() {is=title K2Node_FunctionEntry}
   > SetInitialTabIndex()
   > SetBottomButtonVisibility(
    self=[
     Get DecoyButton_TabRightTriggerNav() -> (DecoyButton_TabRightTriggerNav:WBP Bottom Bar Decoy Button Object Reference)
     Get DecoyButton_LeftTabTriggerNav() -> (DecoyButton_LeftTabTriggerNav:WBP Bottom Bar Decoy Button Object Reference)
     Get WBP_BottomBarHack_Back()
    ]
    ShouldButtonBeVisible=true
   )
   > SetBottomButtonVisibility(self=Get WBP_BottomBarHack_ClearBadges(), ShouldButtonBeVisible=false)
   > Set bShouldFocusOnMegaListItem(bShouldFocusOnMegaListItem=false)
   > ResetHeader()
   n1 = GetSelectedDiscoverTabIndex!(self:Discover Surface View Model Object Reference=Get FortDiscoverSurfaceViewModel()) -> (ReturnValue:int=0)
   n2 = GetSelectedTabIndex(self=Get TabGroup_Search()) -> (ReturnValue=0)
   n3 = OR Boolean(A=Get IsPlayTime Out(), B=Get IsListEmpty())
   n4 = Get OverlayList()
   n5 = Get Switcher_NoResults()
   > SetActiveWidget(
    self=Get Switcher_List()
    Widget=Select(
     Option 0:Widget Object Reference=Get LoadingResults_Spinner() -> (LoadingResults_Spinner:Overlay Object Reference)
     Option 1:Widget Object Reference=Select(Option 0=@n4, Option 1=@n5, Index=@n3)
     Index="Equal (Integer)"(A=@n2, B=@n1)
    ) -> (ReturnValue:Widget Object Reference)
   )
  ]}
 OnBackButtonHackClicked {nodes [OnBackButtonHackClicked!() -> (then, ButtonObject:WBP Bottom Bar Decoy Button Object Reference) {is=title K2Node_FunctionEntry}]}
 OnListScrolledOut {nodes [OnListScrolledOut() {is=title K2Node_FunctionEntry}]}
 OnMouseWheel {nodes [
   n1 = OnMouseWheel!() -> (then, MyGeometry:Geometry Structure, MouseEvent:"Pointer Event Structure (by ref)") {is=title K2Node_FunctionEntry}
   > Set L_MyGeomentry(L_MyGeomentry:Geometry Structure=@n1) -> (Output_Get:Geometry Structure)
   > Set L_MouseEvent(L_MouseEvent:Pointer Event Structure=@n1.MouseEvent) -> (Output_Get:Pointer Event Structure)
   > Set L_bIsTopOfTheList(L_bIsTopOfTheList:bool=GetScrollIsAtTopOfList() {is=title K2Node_CallFunction}) -> (Output_Get:bool)
   > n2 = Branch(Condition=ShouldUpdateScrollToTopForMouse!(MouseEvent:Pointer Event Structure=Get L_MouseEvent() -> (L_MouseEvent:Pointer Event Structure)) -> (ReturnValue:bool))
   > n3 = ScrollToTop(self=Get ListView(self=Get WBP_Poly_MegaList()))
   n4 = Branch(execute=[@n3.then, @n2.else], Condition=Get L_bIsTopOfTheList() -> (L_bIsTopOfTheList:bool))
   n5 = Get Wheel Delta!(Input:"Pointer Event Structure (by ref)"=Get L_MouseEvent()) -> (ReturnValue:float) {is=title K2Node_CallFunction}
   > n6 = Branch(Condition=AND Boolean(A="float > float"(A=@n5), B=Get L_bIsTopOfTheList()))
   n7 = Get Size(self=Get AnimationSpacer())
   > n8 = Branch(Condition=NOT Boolean(A=IsZero2D(A=@n7)))
   > ClearSearch()
   > Call On List Scrolled Out()
   > Return Node(ReturnValue:Event Reply Structure=Handled())
   ResetHeader(execute=@n8.else)
   > Return Node(ReturnValue=Handled())
   n9 = Branch(execute=@n6.else, Condition="Nearly Equal (Float)"(A=Get MegalistItem Offset(), B=0, ErrorTolerance=0))
   > n10 = ScrollToTop(self=Get ListView(self=Get WBP_Poly_MegaList()))
   Return Node(execute=[@n10.then, @n9.else], ReturnValue=Handled())
   n11 = Get Switcher_List()
   n12 = Branch(execute=@n4.else, Condition="Equal (Object)"(A=GetActiveWidget(self=@n11), B=Get NamedSlot_SwitcherListContent()))
   > Return Node(ReturnValue=Unhandled())
   n13 = Branch(execute=@n12.else, Condition=Get Is Row Lock to Top())
   > Return Node(ReturnValue=Handled())
   n14 = Parent: OnMouseWheel!(
    execute=@n13.else
    MyGeometry:Geometry Structure=Get L_MyGeomentry() -> (L_MyGeomentry:Geometry Structure)
    MouseEvent:"Pointer Event Structure (by ref)"=Get L_MouseEvent()
   ) -> (then, ReturnValue:Event Reply Structure) {is=title K2Node_CallParentFunction}
   > Return Node(ReturnValue=@n14)
  ]}
 OnRequestNavigateToDiscover {nodes [OnRequestNavigateToDiscover() {is=title K2Node_FunctionEntry}]}
 OnSearchPeekButtonClicked {nodes [OnSearchPeekButtonClicked() {is=title K2Node_FunctionEntry}]}
 OverrideTabInputAction {nodes [
   OverrideTabInputAction() {is=title K2Node_FunctionEntry}
   n1 = Get SearchBarActionInputNextTab() -> (SearchBarActionInputNextTab:Data Table Row Handle Structure)
   > SetInputAction!(
    execute
    self:Common Action Widget Object Reference=Get NextTabAction() -> (NextTabAction:Common Action Widget Object Reference)
    InputActionRow:Data Table Row Handle Structure=@n1
   ) -> (then)
   n2 = Get SearchBarActionInputPreviousTab() -> (SearchBarActionInputPreviousTab:Data Table Row Handle Structure)
   > SetInputAction(self=Get PrevTabAction() -> (PrevTabAction:Common Action Widget Object Reference), InputActionRow=@n2)
   > SetInputActions!(
    execute
    self:UIKit Tab Group Object Reference=Get TabGroup_Search()
    NextTab:Data Table Row Handle Structure=@n1
    PreviousTab:Data Table Row Handle Structure=@n2
   ) -> (then)
   > SetTriggeringInputAction!(
    execute
    self:Common Button Base Object Reference=Get DecoyButton_TabRightTriggerNav()
    InputActionRow:"Data Table Row Handle Structure (by ref)"=@n1
   ) -> (then)
   > SetTriggeringInputAction(self=Get DecoyButton_LeftTabTriggerNav(), InputActionRow=@n2)
  ]}
 PlaySearchResultsFadeAnimation {nodes [
   n1 = PlaySearchResultsFadeAnimation!() -> (then, SearchResultsScrollOffset:double=~) {is=title K2Node_FunctionEntry}
   > n2 = Branch(Condition="float > float"(A=@n1))
   n3 = Get Anim_SearchScrollingFade()
   > PlayAnimationForward(InAnimation=@n3)
   PlayAnimationReverse(execute=@n2.else, InAnimation=@n3)
  ]}
 Populate Additional Tab Content {nodes [
   Populate Additional Tab Content!() -> (then, TabId:Name, TabButton:Common Button Base Object Reference) {is=title K2Node_FunctionEntry}
   > n1 = Cast To WBP_UIKit_Button_Selectable!(execute, Object=Get TabButton() -> (TabButton:Common Button Base Object Reference)) -> (then, CastFailed, AsWBP UIKit Button Selectable:WBP UIKit Button Selectable Object Reference) {is=title K2Node_DynamicCast}
   > Set UIKitButtonSelectableRef(UIKitButtonSelectableRef:WBP UIKit Button Selectable Object Reference=@n1) -> (Output_Get:WBP UIKit Button Selectable Object Reference)
   > n2 = Construct Object from Class(Class=/FortUILibrary/Systems/ModularButton/Blocks/WBP_UIKit_Block_Text.WBP_UIKit_Block_Text_C) -> (ReturnValue:WBP UIKit Block Text Object Reference)
   > n3 = Set UIKitBlockTextRef(UIKitBlockTextRef:WBP UIKit Block Text Object Reference=@n2) -> (Output_Get:WBP UIKit Block Text Object Reference)
   > Add(
    TargetArray:Array of WBP UIKit Block Text Object References=Get TabBlockTextArray() -> (TabBlockTextArray:Array of WBP UIKit Block Text Object References)
    NewItem:"WBP UIKit Block Text Object Reference (by ref)"=@n3
   ) {is=title K2Node_CallArrayFunction}
   > n4 = AddChildToGrid!(
    execute
    self:Grid Panel Object Reference=Get Main!(self:WBP UIKit Button Generic Base Object Reference=Get UIKitButtonSelectableRef() -> (UIKitButtonSelectableRef:WBP UIKit Button Selectable Object Reference)) -> (Main:Grid Panel Object Reference)
    Content:Widget Object Reference=@n3
    InRow:int=1
    InColumn:int=2
   ) -> (then, ReturnValue:Grid Slot Object Reference)
   > SetHorizontalAlignment(self:Grid Slot Object Reference=@n4, InHorizontalAlignment=HAlign_Left)
   > SetVerticalAlignment(self:Grid Slot Object Reference=@n4, InVerticalAlignment=VAlign_Fill)
   > SetFontMaterial!(
    execute
    self:WBP UIKit Block Text Object Reference=Get UIKitBlockTextRef() -> (UIKitBlockTextRef:WBP UIKit Block Text Object Reference)
    In Material:Material Instance Object Reference=/FortUILibrary/UIKit/Buttons/TabButtons/_MaterialInstances/MI_UI_TopNav_Text
   ) -> (then)
   > RegisterModularBlock!(
    execute
    self:UIKit Modular Button Object Reference=Get UIKitButtonSelectableRef()
    Block:UIKit Block Object Reference=Get UIKitBlockTextRef()
   ) -> (then)
  ]}
 PopulateSearchResultTabs {nodes [
   PopulateSearchResultTabs() {is=title K2Node_FunctionEntry}
   > n1 = Construct Object from Class(Class=/Script/FortniteUI.UIKitTabGroupViewModel) -> (ReturnValue:UI Kit Tab Group View Model Object Reference)
   > n2 = Sequence!(execute) -> (then_0, then_1, then_2:Exec)
   > n3 = Construct Object from Class(Class=/Script/FortniteUI.UIKitTabGroupButtonViewModel) -> (ReturnValue:UI Kit Tab Group Button View Model Object Reference)
   > n4 = Set Label!(execute, Label:Text="LOCTEXT(\"Islands\")", self:UI Kit Selectable Button View Model Object Reference=@n3) -> (then, Output_Get:Text)
   > Set TabID!(
    execute
    TabID:Name=String To Name(InString=Get Source String from Text!(Text:Text=@n4) -> (ReturnValue:String) {is=title K2Node_CallFunction})
    self:UI Kit Tab Group Button View Model Object Reference=@n3
   ) -> (then, Output_Get:Name=None)
   n5 = Get Tabs!(self:UI Kit Tab Group View Model Object Reference=@n1) -> (Tabs:Array of UI Kit Tab Group Button View Model Object References)
   > Add(
    TargetArray:Array of UI Kit Tab Group Button View Model Object References=@n5
    NewItem:"UI Kit Tab Group Button View Model Object Reference (by ref)"=@n3
   )
   n6 = Construct Object from Class(execute=@n2.then_1, Class=&UIKitTabGroupButtonViewModel)
   > n7 = Set Label(Label="LOCTEXT(\"Developers\")", self=@n6)
   > Set TabID(TabID=String To Name(InString=Get Source String from Text(Text=@n7)), self=@n6) -> (Output_Get=None)
   > Add(TargetArray=@n5, NewItem=@n6)
   n8 = Get Tabs_Search_Results()
   Set Tabs Data!(
    execute=@n2.then_2
    self:WBP UIKit Tabs Base Object Reference=@n8
    InTabs Data:Array of UI Kit Tab Group Button View Model Object References=@n5
   ) -> (then) {is=title K2Node_CallFunction}
   > SelectTabByIndex(self=@n8, Index=0, bSuppressClickFeedback=true)
  ]}
 RefreshCurrentPageToTop {nodes [
   RefreshCurrentPageToTop() {is=title K2Node_FunctionEntry}
   > n1 = Switch on Int(Selection=GetActiveWidgetIndex(self=Get Switcher_List()) -> (ReturnValue=0)) {bHasDefaultPin=false}
   > TransitionToMiniHeader(IsMiniHeaderWithFullList=false)
   > ScrollToTop(self=Get ListView(self=Get WBP_Poly_MegaList()))
   > Set LastFocusedWidget(LastFocusedWidget=GetActiveTabButtonObject!() -> (SearchTabButton:Common Button Base Object Reference) {is=title K2Node_CallFunction})
   > Set RequestRefreshFocusAfterAnimation(RequestRefreshFocusAfterAnimation=true)
   > Branch(Condition=Get Is Row Lock to Top())
   > n2 = Branch(Condition="Equal (Enum)"(A:ECommonInputType Enum=Get Current Input Type(), B:ECommonInputType Enum=MouseAndKeyboard))
   > Set LastFocusedWidget()
   RequestRefreshFocus(execute=[@n4.then, @n2.else])
   n3 = SelectTabByIndex(execute=@n1.1, self=Get TabGroup_Search(), Index=0, bSuppressClickFeedback=true)
   n4 = TransitionToMiniHeader(execute=[@n6.then, @n3.then], IsMiniHeaderWithFullList=false)
   ScrollToTop(execute=@n1.3, self=Get TileView(self=Get WBP_Results_PolyTileView()))
   n5 = Get InputField(self=Get SearchInputField())
   > n6 = Set LastFocusedWidget(LastFocusedWidget=Get EditableText(self=Get Block_EditableText(self=@n5)))
  ]}
 Request Panels Hydration {nodes [
   Request Panels Hydration() {is=title K2Node_FunctionEntry}
   K2Node_Composite_0 = Get Last Visible Row Index!() -> (ReturnValue:int) {is=title K2Node_Composite}
   > Set Last Visible Row Index(Last Visible Row Index:int=@K2Node_Composite_0) -> (Output_Get:int=0)
   n1 = Get Last Hydration Index Trigger() -> (Last Hydration Index Trigger:int=0)
   n2 = int - int(
    A=Get Last Visible Row Index() -> (Last Visible Row Index:int=0) {is=title K2Node_VariableGet}
    B=Get Last Hydration Index Trigger() -> (Last Hydration Index Trigger=0)
   )
   n3 = Get LazyLoadLookaheadHydration!(self:Fort Runtime Options Object Reference=GetRuntimeOptions()) -> (LazyLoadLookaheadHydration:int=0)
   > Branch(Condition=OR Boolean(A="Equal (Integer)"(A=@n1, B=-1), B="integer >= integer"(A=@n2, B=@n3)))
   > RequestPanelsHydration!(
    execute
    self:Discover Surface View Model Object Reference=Get FortDiscoverSurfaceViewModel()
    InPanelIndexTo:int=int + int(A=Get Last Visible Row Index() -> (Last Visible Row Index=0), B=int * int(A=@n3, B=2))
   ) -> (then)
   > Set Last Hydration Index Trigger(Last Hydration Index Trigger:int=Get Last Visible Row Index() -> (Last Visible Row Index=0)) -> (Output_Get:int=0)
  ]}
 Get Last Visible Row Index {
  path=Request Panels Hydration.K2Node_Composite_0.Get Last Visible Row Index
  nodes [
   Inputs!()
   n1 = Get ListView(self=Get WBP_Poly_MegaList())
   n2 = Ceil to Integer64!(A:double=GetScrollOffset(self=@n1)) -> (ReturnValue:Integer64=0)
   n3 = integer64 - integer64(
    A:int=Length(TargetArray:Array of User Widget Object References=GetDisplayedEntryWidgets!(self:List View Base Object Reference=@n1) -> (ReturnValue:Array of User Widget Object References))
    B:int=1
   ) -> (ReturnValue:Integer64) {is=title K2Node_PromotableOperator}
   Outputs!(ReturnValue:int="To Integer (Integer64)"!(InInt:Integer64=integer64 + integer64(A:Integer64=@n2, B:Integer64=@n3) -> (ReturnValue:Integer64) {is=title K2Node_PromotableOperator}) -> (ReturnValue:int=0))
  ]
 }
 RightTabTriggerPressed {nodes [
   RightTabTriggerPressed() {is=title K2Node_FunctionEntry}
   > n1 = Branch(Condition=IsAnySearchActiveOrInProgress!(self:Discover Search Model Object Reference=Get FortDiscoverSearchViewModel()) -> (ReturnValue:bool))
   > SetActiveWidget(self=Get Switcher_List(), Widget=Select(Option 0=Get OverlayList(), Option 1=Get Switcher_NoResults(), Index=Get IsPlayTime Out()))
   > ClearSearch()
   n2 = Get WBP_Poly_MegaList()
   > Set bFirstRowSnapIsZeroIndex(bFirstRowSnapIsZeroIndex=true, self=@n2)
   n3 = Get ListView(self=@n2)
   > SetSelectedIndex(self=@n3, Index=0)
   > Navigate To Item!(
    execute
    self:List View Object Reference=@n3
    Item:Object Reference=Get Selected Item!(self:List View Object Reference=@n3) -> (ReturnValue:Object Reference) {is=title K2Node_CallFunction}
   ) -> (then)
   > Set LastFocusedWidget(LastFocusedWidget=Get ListView(self=@n2))
   > RequestRefreshFocus()
   n4 = Get TabGroup_Search()
   n5 = GetTabCount(self=@n4) -> (ReturnValue=0)
   n6 = Branch(execute=@n1.else, Condition="Not Equal (Integer)"(A=int - int(A=@n5), B=GetSelectedTabIndex(self=@n4) -> (ReturnValue=0)))
   > n7 = SelectNextTab!(execute, self:UIKit Tab Group Object Reference=Get TabGroup_Search(), bSuppressClickFeedback:bool=true) -> (then)
   TransitionToMiniHeader(execute=[@n7.then, @n8.then], IsMiniHeaderWithFullList=false)
   SelectTabByIndex(execute=@n6.else, self=Get TabGroup_Search(), Index=0, bSuppressClickFeedback=true)
   > n8 = FocusOnSearchField()
  ]}
 Set Material Instance Parameters {nodes [
   Set Material Instance Parameters() {is=title K2Node_FunctionEntry}
   > GetDynamicMaterial(self=Get Image_SoftEdgeGradient_Islands() -> (Image_SoftEdgeGradient_Islands:Image Object Reference))
  ]}
 Set No Result State {nodes [
   n1 = Set No Result State!() -> (then, IslandResultCount:int, CreatorResultCount:int) {is=title K2Node_FunctionEntry}
   n2 = AND Boolean(A="Equal (Integer)"(A=@n1), B="Equal (Integer)"(A=@n1.CreatorResultCount))
   > SetVisibility(
    self=Get WBP_Row_Header() -> (WBP_Row_Header:WBP Row Header Object Reference)
    InVisibility=Select(Option 0:ESlateVisibility Enum=SelfHitTestInvisible, Option 1:ESlateVisibility Enum=Hidden, Index=@n2) -> (ReturnValue:ESlateVisibility Enum)
   )
   > Return Node!(
    execute
    Switcher:Widget Object Reference=Select(
     Option 0:Widget Object Reference=Get VerticalBox_IslandBox() -> (VerticalBox_IslandBox:Vertical Box Object Reference)
     Option 1:Widget Object Reference=Get WBP_NoResults_SearchResults() -> (WBP_NoResults_SearchResults:WBP Discovery No Results Message Object Reference)
     Index=@n2
    ) -> (ReturnValue:Widget Object Reference)
   )
  ]}
 SetAnim_OutputBackgroundColor {nodes [
   n1 = SetAnim_OutputBackgroundColor!() -> (then, Anim Progress:double=~) {is=title K2Node_FunctionEntry}
   n2 = Select(
    Option 0:Linear Color Structure=Select(
     Option 0=(R=0.0144,G=0.1912,B=0.3663,A=1)
     Option 1=(R=0.9387,G=0,B=0.3864,A=1)
     Index=GetConfigDataBoolWithDefault!(EntryName:String=bEnableMangoFrontEndVisuals, bDefaultValue:bool=false) -> (ReturnValue:bool)
    )
    Option 1:Linear Color Structure=(R=0.7231,G=0.0382,B=0,A=1)
    Index=GetConfigDataBoolWithDefault(EntryName=bEnableCH7FrontEndVisuals, bDefaultValue=false)
   ) -> (ReturnValue:Linear Color Structure)
   n3 = Get CachedTab() -> (CachedTab:E_Discover_Background_Color Enum=NewEnumerator0)
   n4 = Get SelectedTab() -> (SelectedTab:E_Discover_Background_Color Enum=NewEnumerator0)
   > Set OutputBackgroundColor(OutputBackgroundColor:Linear Color Structure="Lerp Using HSV (LinearColor)"!(
     A:Linear Color Structure=Select!(
      NewEnumerator0:Linear Color Structure=@n2
      NewEnumerator1:Linear Color Structure=(R=0.0697,G=0.0152,B=0.3419,A=1)
      NewEnumerator2:Linear Color Structure=(R=0.016,G=0.3278,B=0.3278,A=1)
      NewEnumerator3:Linear Color Structure=(R=0.3278,G=0.016,B=0.1199,A=1)
      NewEnumerator4:Linear Color Structure=(R=0.3278,G=0.042,B=0.016,A=1)
      Index:E_Discover_Background_Color Enum=@n3
     ) -> (ReturnValue)
     B:Linear Color Structure=Select(
      NewEnumerator0=@n2
      NewEnumerator1=(R=0.0697,G=0.0152,B=0.3419,A=1)
      NewEnumerator2=(R=0.016,G=0.3278,B=0.3278,A=1)
      NewEnumerator3=(R=0.3278,G=0.016,B=0.1199,A=1)
      NewEnumerator4=(R=0.3278,G=0.042,B=0.016,A=1)
      Index=@n4
     )
     Alpha:float=@n1
    ) -> (ReturnValue:Linear Color Structure)) -> (Output_Get:Linear Color Structure)
   n5 = Get Image_SoftEdgeGradient_Islands()
   > SetVectorParameterValue(
    self=GetDynamicMaterial!(Brush:"Slate Brush Structure (by ref)"=Get Brush!(self:Image Object Reference=@n5) -> (Brush:Slate Brush Structure)) -> (ReturnValue)
    ParameterName=Primary Background Color
    Value=Get OutputBackgroundColor() -> (OutputBackgroundColor:Linear Color Structure)
   )
   n6 = Get Image_ListSoftBottomEdge() -> (Image_ListSoftBottomEdge:Image Object Reference)
   > SetVectorParameterValue(self=GetDynamicMaterial(Brush=Get Brush(self=@n6)), ParameterName=Primary Background Color, Value=Get OutputBackgroundColor())
  ]}
 SetBackgroundColorEnums {nodes [
   n1 = SetBackgroundColorEnums!() -> (then, Tab Index:int) {is=title K2Node_FunctionEntry}
   > Set SelectedTab(SelectedTab:E_Discover_Background_Color Enum=Select!(
     Option 0:E_Discover_Background_Color Enum=NewEnumerator0
     Option 1:E_Discover_Background_Color Enum=NewEnumerator1
     Option 2:E_Discover_Background_Color Enum=NewEnumerator2
     Option 3:E_Discover_Background_Color Enum=NewEnumerator3
     Option 4:E_Discover_Background_Color Enum=NewEnumerator4
     Index:int=@n1
    ) -> (ReturnValue:E_Discover_Background_Color Enum)) -> (Output_Get:E_Discover_Background_Color Enum=NewEnumerator0)
   > Set CachedTab(CachedTab:E_Discover_Background_Color Enum=Select(
     Option 0=NewEnumerator0
     Option 1=NewEnumerator1
     Option 2=NewEnumerator2
     Option 3=NewEnumerator3
     Option 4=NewEnumerator4
     Index=Get CachedTabIndex() -> (CachedTabIndex=0)
    )) -> (Output_Get:E_Discover_Background_Color Enum=NewEnumerator0)
  ]}
 SetCompactUIStyle {nodes [
   SetCompactUIStyle() {is=title K2Node_FunctionEntry}
   > Parent: SetCompactUIStyle() {is=title K2Node_CallParentFunction}
   > Branch(Condition=IsUsingCompactUI())
   n1 = Get SearchInputField()
   n2 = Get SIzeBox!(self:WBP UIKit Input Field Base Object Reference=Get InputField(self=@n1)) -> (SIzeBox:Size Box Object Reference)
   > SetHeightOverride(self=@n2, InHeightOverride=80)
   > ClearWidthOverride(self:Size Box Object Reference=@n2)
   > SetHorizontalAlignment(self=SlotAsGridSlot(Widget=@n1), InHorizontalAlignment=HAlign_Fill)
   > SetHorizontalAlignment(self=SlotAsGridSlot(Widget=Get TabGroup_Search()), InHorizontalAlignment=HAlign_Right)
   > SetColumnFill!(
    execute
    self:Grid Panel Object Reference=Get GP_SearchBar() -> (GP_SearchBar:Grid Panel Object Reference)
    ColumnIndex:int=0
    Coefficient:float=1
   ) -> (then)
   > Set MinButtonHeight!(execute, MinButtonHeight:int=80, self:WBP UIKit Tabs Base Object Reference=Get Tabs_Search_Results()) -> (then, Output_Get:int=0)
   > Set!(
    execute
    Padding_Left:float=132
    Padding_Top:float=0
    Padding_Right:float=0
    Padding_Bottom:float=16
    self:Grid Slot Object Reference=SlotAsGridSlot(Widget=Get Tabs_Search_Results())
   ) -> (then, Output_Get:Margin Structure)
  ]}
 SetInitialRefreshFocus {nodes [
   n1 = SetInitialRefreshFocus!() -> (then, ShouldRefreshFocus:bool=~) {is=title K2Node_FunctionEntry}
   > n2 = Sequence!(execute) -> (then_0, then_1)
   > n3 = GetConsoleVariableBoolValue!(execute, VariableName:String=Discovery.ActivateReduxOnRefreshFocus) -> (then, ReturnValue:bool)
   > Branch(Condition=@n3)
   > n4 = Branch(Condition=@n1)
   > ActivateWidget()
   DeactivateWidget(execute=@n4.else)
   Branch(execute=@n2.then_1, Condition=@n1)
   > RequestRefreshFocus()
  ]}
 SetInitialTabIndex {nodes [
   SetInitialTabIndex() {is=title K2Node_FunctionEntry}
   n1 = GetSelectedDiscoverTabIndex(self=Get FortDiscoverSurfaceViewModel()) -> (ReturnValue=0)
   n2 = GetSelectedTabIndex(self=Get TabGroup_Search()) -> (ReturnValue=0)
   > n3 = Set InitialTabIndex(InitialTabIndex:int=SelectInt(A="Max (Integer)"(A=0, B=@n2), B=@n1, bPickA="integer < integer"(A=@n1))) -> (Output_Get:int=0)
   > SelectTabByIndex(self=Get TabGroup_Search(), Index=@n3, bSuppressClickFeedback=true)
  ]}
 SetListQuerying {nodes [
   SetListQuerying!() -> (then, LoadingState:EDiscoverLoadingState Enum=Unset, IsListEmpty:bool) {is=title K2Node_FunctionEntry}
   n1 = Get LoadingState() -> (LoadingState:EDiscoverLoadingState Enum=Unset)
   > n2 = Branch(Condition=AND Boolean(A=Get IsListEmpty(), B="Equal (Enum)"(A:EDiscoverLoadingState Enum=@n1, B:EDiscoverLoadingState Enum=Loaded)))
   n3 = Cast To FortDiscoverTileItemVM!(
    execute=[@n20.else, @n2.then]
    Object=GetTabViewModel!(self:UIKit Tab Group Object Reference=Get TabGroup_Search(), TabID:Name=Get CurrentTabId() -> (CurrentTabId:Name=None)) -> (ReturnValue:MVVM Base Viewmodel Object Reference)
   ) -> (then, CastFailed, AsDiscover Tile Item View Model:Discover Tile Item View Model Object Reference) {is=title K2Node_DynamicCast}
   n4 = GetMnemonic!(self:Discover Tile Item View Model Object Reference=@n3) -> (ReturnValue:String)
   n5 = Find(
    TargetMap:Map of Strings to Discover Redux Tab Empty Results Text Entry Structures=Get DiscoverReduxTabEmptyResultsText!(self:Fort Runtime Options Object Reference=GetRuntimeOptions()) -> (DiscoverReduxTabEmptyResultsText:Map of Strings to Discover Redux Tab Empty Results Text Entry Structures)
    Key=@n4
   ) -> (Value:Discover Redux Tab Empty Results Text Entry Structure)
   > n6 = Branch(Condition=@n5.ReturnValue)
   n7 = Break Discover Redux Tab Empty Results Text Entry!(DiscoverReduxTabEmptyResultsTextEntry:"Discover Redux Tab Empty Results Text Entry Structure (by ref)"=@n5) -> (Title:Text, Subtitle:Text) {is=title K2Node_BreakStruct}
   n8 = TextIsEmpty(InText=@n7)
   n9 = TextIsEmpty(InText=@n7.Subtitle)
   > n10 = Branch(Condition=NOT Boolean(A=OR Boolean(A=@n8, B=@n9)))
   > n11 = SetMessageTextOverrides!(
    execute
    self:WBP Discovery Blank Tiles Rounded Corners Object Reference=Get WBP_Discovery_BlankTiles_RoundedCorners() -> (WBP_Discovery_BlankTiles_RoundedCorners:WBP Discovery Blank Tiles Rounded Corners Object Reference)
    InTitle:Text=@n7
    InSubTitle:Text=@n7.Subtitle
   ) -> (then)
   SetVisibility(
    execute=[@n19.then, @n11.then, @n21.then]
    self=Get ViewTimeLimits_Button()
    InVisibility=Select!(Option 0:ESlateVisibility Enum=Collapsed, Option 1:ESlateVisibility Enum=SelfHitTestInvisible, Index:bool=Get IsPlayTime Out()) -> (ReturnValue:ESlateVisibility Enum)
   )
   > SetVisibility(self=Get BackToDisco_Button(), InVisibility=Select(Option 0=SelfHitTestInvisible, Option 1=Collapsed, Index=Get IsPlayTime Out()))
   n12 = Get IsListEmpty()
   n13 = Get LoadingState() -> (LoadingState=Unset)
   n14 = "Equal (Integer)"(A=Get CachedTabIndex() -> (CachedTabIndex=0))
   n15 = Get IsListEmpty()
   > Return Node!(
    execute
    SwitcherActiveWidget:Widget Object Reference=Select(
     Option 0:Widget Object Reference=Parent: SetListQuerying!(LoadingState:EDiscoverLoadingState Enum=@n13, IsListEmpty:bool=@n12) -> (SwitcherActiveWidget:Widget Object Reference) {is=title K2Node_CallParentFunction}
     Option 1:Widget Object Reference=Get LoadingResults_Spinner()
     Index=AND Boolean(A=@n14, B=@n15)
    ) -> (ReturnValue:Widget Object Reference)
   )
   n16 = LogString(
    execute=@n10.else
    InString="To String (Text)"(InText=Format Text!(
      Format="LOCTEXT(\"FortRuntimeOptions - DiscoverReduxTabEmpty: Found entry for mnemonic {0} but no title or subtitle is set. Defaulting to generic text message.\")"
      0:String=@n4
     ) -> (Result))
   )
   n17 = Cast To WBP_Search_Tab_Button!(execute=[@n16.then, @n6.else], Object=GetTabButtonBaseByID(self=Get TabGroup_Search(), TabNameID=Get CurrentTabId() -> (CurrentTabId=None))) -> (then, CastFailed, AsWBP Search Tab Button:WBP Search Tab Button Object Reference) {is=title K2Node_DynamicCast}
   n18 = Get TabText!(self:WBP Search Tab Button Object Reference=@n17) -> (TabText:Text)
   > n19 = SetMessageTextOverrides(
    self=Get WBP_Discovery_BlankTiles_RoundedCorners()
    InTitle=Format Text!(Format="LOCTEXT(\"no {TabText} games\")", TabText:Text=@n18) -> (Result)
    InSubTitle=Format Text(Format="LOCTEXT(\"Your {TabText} games will appear here.\")", TabText=@n18)
   )
   n20 = Branch(execute=@n2.else, Condition=Get IsPlayTime Out())
   > SetMessageTextOverrides(
    self=Get WBP_Discovery_BlankTiles_RoundedCorners()
    InTitle="LOCTEXT(\"YOUR PLAY TIME IS UP\")"
    InSubTitle="LOCTEXT(\"You can test unpublished islands by searching for their island code directly.\")"
   )
   > n21 = Branch(Condition="Not Equal (Enum)"(A:EDiscoverLoadingState Enum=Get LoadingState() -> (LoadingState=Unset), B:EDiscoverLoadingState Enum=Loaded))
   Return Node(execute=@n21.else, SwitcherActiveWidget=Get Switcher_NoResults())
  ]}
 SetMegaListMinMaxDesiredHeight {nodes [
   SetMegaListMinMaxDesiredHeight() {is=title K2Node_FunctionEntry}
   > SetMinDesiredHeight!(
    execute
    self:Size Box Object Reference=Get SB_MegaListContentHeightLock() -> (SB_MegaListContentHeightLock:Size Box Object Reference)
    InMinDesiredHeight:float=600
   ) -> (then)
   > n1 = Branch(Condition=IsUsingCompactUI())
   > n2 = GetConsoleVariableFloatValue!(execute, VariableName:String=Frontend.Discover.MobileMaxHeightWithoutTopRow) -> (then, ReturnValue:float)
   > n3 = GetConsoleVariableFloatValue(VariableName=Frontend.Discover.MobileMaxHeightWithTopRow)
   n4 = Get Is Row Lock to Top()
   n5 = GetViewportSize!() -> (ReturnValue_X:double, ReturnValue_Y:double)
   n6 = GetViewportScale!() -> (ReturnValue:float)
   > SetMaxDesiredHeight!(
    execute
    self:Size Box Object Reference=Get SB_MegaListContentHeightLock()
    InMaxDesiredHeight:float=float * float(A=SelectFloat(A=@n3, B=@n2, bPickA=@n4), B=float / float(A=@n5.ReturnValue_Y, B=@n6)) {is=title K2Node_PromotableOperator}
   ) -> (then)
   SetMaxDesiredHeight(execute=@n1.else, self=Get SB_MegaListContentHeightLock(), InMaxDesiredHeight=1080)
  ]}
 SetNavOnTabSwitched {nodes [
   n1 = SetNavOnTabSwitched!() -> (then, LoadingState:EDiscoverLoadingState Enum=Unset) {is=title K2Node_FunctionEntry}
   > n2 = Switch on EDiscoverLoadingState!(execute, Selection:EDiscoverLoadingState Enum=@n1) -> (Unset:Exec, Loading:Exec, Loaded:Exec, Error:Exec) {is=title K2Node_SwitchEnum}
   n3 = GetListItems!(self:List View Object Reference=Get ListView(self=Get WBP_Poly_MegaList())) -> (ReturnValue:Array of Object References)
   n4 = Branch(
    execute=[@n2.Error, @n2.Loaded]
    Condition=AND Boolean(A=Is Not Empty(TargetArray:Array of Object References=@n3), B=Get bShouldFocusOnMegaListItem() -> (bShouldFocusOnMegaListItem:bool))
   )
   > Set bShouldFocusOnMegaListItem(bShouldFocusOnMegaListItem=false)
   n5 = Get WBP_Poly_MegaList()
   > Set bFirstRowSnapIsZeroIndex(bFirstRowSnapIsZeroIndex=true, self=@n5)
   n6 = Get ListView(self=@n5)
   > SetSelectedIndex(self=@n6, Index=0)
   > Navigate To Item(self=@n6, Item=Get Selected Item(self=@n6))
   > Set LastFocusedWidget(LastFocusedWidget=Get ListView(self=@n5))
   > RequestRefreshFocus()
   Set bShouldFocusOnMegaListItem(execute=@n4.else, bShouldFocusOnMegaListItem=false)
  ]}
 SetNoResultsSearchSwitcherState {nodes [
   SetNoResultsSearchSwitcherState!() -> (then, bShouldShowEmptyState:bool=~) {is=title K2Node_FunctionEntry}
   n1 = Get bShouldShowEmptyState() -> (bShouldShowEmptyState:bool)
   n2 = Get WBP_NoResults_SearchResults()
   n3 = Get VerticalBox_IslandBox()
   > SetActiveWidget(
    self=Get CommonWidgetSwitcher_Results() -> (CommonWidgetSwitcher_Results:Common Widget Switcher Legacy Object Reference)
    Widget=Select(Option 0=Select(Option 0=@n3, Option 1=@n2, Index=@n1), Option 1=Get VerticalBox_IslandBox(), Index=Get IsPlayTime Out())
   )
   > SetVisibility(self=Get WBP_Row_Header(), InVisibility=SelfHitTestInvisible)
  ]}
 SetPlayTimeState {nodes [
   n1 = SetPlayTimeState!() -> (then, IsPlayTimeOut:bool=~) {is=title K2Node_FunctionEntry}
   > Set with Broadcast IsPlayTime Out!(execute, IsPlayTime Out:bool=@n1) -> (then, Output_Get:bool)
   > UpdateEmptyStateForParentalControl()
   n2 = GetTabCount(self=Get TabGroup_Search()) -> (ReturnValue=0)
   > Branch(Condition="integer > integer"(A=@n2))
   > n3 = For Loop(LastIndex=@n2)
   > Branch(Condition="integer > integer"(A=@n3))
   n4 = GetTabIDbyIndex!(self:UIKit Tab Group Object Reference=Get TabGroup_Search(), Index:int=@n3) -> (ReturnValue:Name=None)
   > SetTabEnabled!(execute, self:Common Tab List Widget Base Object Reference=Get TabGroup_Search(), TabNameID:Name=@n4, bEnable:bool=true) -> (then)
   > SetTabEnabled(self=Get TabGroup_Search(), TabNameID=@n4, bEnable=NOT Boolean(A=Get IsPlayTime Out()))
  ]}
 SetReduxRowActivation {nodes [
   n1 = SetReduxRowActivation!() -> (then, ShouldActivateWidget:bool=~) {is=title K2Node_FunctionEntry}
   > n2 = GetConsoleVariableBoolValue(VariableName=Discovery.ActivateReduxOnRefreshFocus)
   > n3 = Branch(Condition=@n2)
   n4 = Branch(execute=@n3.else, Condition=@n1)
   > ActivateWidget()
   Branch(execute=@n4.else, Condition=IsActivated())
   > DeactivateWidget()
  ]}
 SetResultsTerm {nodes [
   n1 = SetResultsTerm!() -> (then, ResultsTerm:String) {is=title K2Node_FunctionEntry}
   > SetSubHeaderText!(
    execute
    self:WBP Row Header Object Reference=Get WBP_Row_Header()
    SubHeaderText:Text=Format Text!(Format="LOCTEXT(\"Matching \\\"{ResultsTerm}\\\"\")", ResultsTerm:String=@n1) -> (Result)
   ) -> (then)
  ]}
 SetRowHeaderText {nodes [
   n1 = SetRowHeaderText!() -> (then, TabIndex:int) {is=title K2Node_FunctionEntry}
   n2 = "Get (a copy)"(
    Array:Array of UI Kit Tab Group Button View Model Object References=Get LocalTabsViewModels!(self:UIKit Tab Group Object Reference=Get Tabs_Search_Results()) -> (LocalTabsViewModels:Array of UI Kit Tab Group Button View Model Object References)
    Dimension 1=@n1
   ) -> (Output:UI Kit Tab Group Button View Model Object Reference)
   > SetHeaderText!(
    execute
    self:WBP Row Header Object Reference=Get WBP_Row_Header()
    HeaderText:Text=Format Text!(Format="LOCTEXT(\"{Label} Results\")", Label:Text=Get Label!(self:UI Kit Selectable Button View Model Object Reference=@n2) -> (Label:Text)) -> (Result)
   ) -> (then)
  ]}
 SetRowLocksToTop {nodes [
   n1 = SetRowLocksToTop!() -> (then, IsRowLockToTop:bool=~) {is=title K2Node_FunctionEntry}
   > Set Is Row Lock to Top(Is Row Lock to Top:bool=@n1) -> (Output_Get:bool)
  ]}
 SetSearchCount {nodes [
   n1 = SetSearchCount!() -> (then, IslandCount:int, CreatorCount:int) {is=title K2Node_FunctionEntry}
   > n2 = For Each Loop(Array:Array of WBP UIKit Block Text Object References=Get TabBlockTextArray()) -> (Array Element:WBP UIKit Block Text Object Reference)
   > SetText!(
    execute
    self:WBP UIKit Block Text Object Reference=@n2
    Text:Text=Select(
     Option 0:Text=Format Text!(Format="LOCTEXT(\"({IslandsCount})\")", IslandsCount:int=@n1) -> (Result)
     Option 1:Text=Format Text!(Format="LOCTEXT(\"({CreatorsCount})\")", CreatorsCount:int=@n1.CreatorCount) -> (Result)
     Index:int=@n2.Array Index
    ) -> (ReturnValue:Text)
   ) -> (then)
   Return Node!(execute=@n2.Completed, SearchTabTextArray:Array of WBP UIKit Block Text Object References=Get TabBlockTextArray())
  ]}
 SetSearchInProgress {nodes [
   SetSearchInProgress!() -> (then, CreatorSearchInProgress:bool=~, IslandSearchInProgress:bool=~) {is=title K2Node_FunctionEntry}
   n1 = OR Boolean(A=Get CreatorSearchInProgress() -> (CreatorSearchInProgress:bool), B=Get IslandSearchInProgress() -> (IslandSearchInProgress:bool))
   > n2 = Branch(Condition=@n1)
   > n3 = SetActiveWidget(self=Get Switcher_List(), Widget=Get NamedSlot_SwitcherListContent())
   Return Node!(
    execute=[@n3.then, @n2.else]
    SwitcherActiveWidget:Widget Object Reference=Select(
     Option 0:Widget Object Reference=Get SearchResultsGrid() -> (SearchResultsGrid:Grid Panel Object Reference)
     Option 1:Widget Object Reference=Get SearchSpinner() -> (SearchSpinner:Grid Panel Object Reference)
     Index:bool=@n1
    ) -> (ReturnValue:Widget Object Reference)
   )
  ]}
 SetShowCreateTimeEmptyStateSearchResults {nodes [
   n1 = SetShowCreateTimeEmptyStateSearchResults!() -> (then, bShowEmptyState:bool=~) {is=title K2Node_FunctionEntry}
   > SetActiveWidget(
    self=Get CommonWidgetSwitcher_OutOfTimeSearch() -> (CommonWidgetSwitcher_OutOfTimeSearch:Common Widget Switcher Legacy Object Reference)
    Widget=Select(
     Option 0=Get O_ResultTileList() -> (O_ResultTileList:Overlay Object Reference)
     Option 1=Get GP_NoResultsCreateTime() -> (GP_NoResultsCreateTime:Grid Panel Object Reference)
     Index=@n1
    )
   )
   > Branch(Condition=Get bShowEmptyState() -> (bShowEmptyState:bool))
   > SetVisibility(self=Get WBP_Row_Header(), InVisibility=SelfHitTestInvisible)
   > SetVisibility(self=Get Tabs_Search_Results(), InVisibility=Collapsed)
  ]}
 SetTabNotificationData {nodes [
   SetTabNotificationData!() -> (then, TotalUnseenCount:Array of Fort Creative Discovery Notification Entry Structures) {is=title K2Node_FunctionEntry}
   n1 = Get TabGroup_Search()
   > n2 = For Each Loop(Array:Array of Widget Object References=GetAllChildren(self=Get TabsContainer!(self:WBP UIKit Tab Group Base Object Reference=@n1) -> (TabsContainer:Grid Panel Object Reference))) -> (Array Element:Widget Object Reference)
   > n3 = Cast To WBP_Search_Tab_Button(Object=@n2)
   > Set L_SearchTab(L_SearchTab:WBP Search Tab Button Object Reference=@n3) -> (Output_Get:WBP Search Tab Button Object Reference)
   > Branch(Condition=Is Not Empty(TargetArray:Array of Fort Creative Discovery Notification Entry Structures=Get TotalUnseenCount() -> (TotalUnseenCount:Array of Fort Creative Discovery Notification Entry Structures)))
   > n4 = For Each Loop(Array:Array of Fort Creative Discovery Notification Entry Structures=Get TotalUnseenCount()) -> (Array Element:Fort Creative Discovery Notification Entry Structure)
   n5 = Break Fort Creative Discovery Notification Entry!(FortCreativeDiscoveryNotificationEntry:"Fort Creative Discovery Notification Entry Structure (by ref)"=@n4) -> (ReferenceLinkCode:String, Total:int, NotificationTypeCount:Array of Fort Creative Discovery Notification Type Count Structures) {is=title K2Node_BreakStruct}
   n6 = "To String (Text)"(InText=Get TabText(self=Get L_SearchTab() -> (L_SearchTab:WBP Search Tab Button Object Reference)))
   > Branch(Condition="Not Equal (Integer)"(A=FindSubstring(SearchIn=@n5, Substring=@n6), B=-1))
   > SetNotificationCount!(execute, self:WBP Search Tab Button Object Reference=Get L_SearchTab(), Count:int=@n5.Total) -> (then)
   MakeLiteralInt!(Value:int=39) -> (ReturnValue:int=0)
  ]}
 SetTabsNoResultsSize {nodes [
   n1 = SetTabsNoResultsSize!() -> (then, DiscoveryBlankTilesRoundedCorners:WBP Discovery Blank Tiles Rounded Corners Object Reference) {is=title K2Node_FunctionEntry}
   > Set VerticalEntrySpacing!(
    execute
    VerticalEntrySpacing:float=16
    self:List View Object Reference=Get TileView_BlankTiles!(self:WBP Discovery Blank Tiles Rounded Corners Object Reference=@n1) -> (TileView_BlankTiles:Tile View Object Reference)
   ) -> (then, Output_Get:float)
   > SetEntryTileSize!(execute, self:WBP Discovery Blank Tiles Rounded Corners Object Reference=@n1, EntrySize:Vector 2D Structure=(X=256,Y=144)) -> (then)
  ]}
 SetupNavigationRules {nodes [
   SetupNavigationRules() {is=title K2Node_FunctionEntry}
   > SetNavigationRuleCustom!(
    execute
    self:Widget Object Reference=[Get TabGroup_Search(), Get SearchInputField()]
    Direction:EUINavigation Enum=Down
    InCustomDelegate:Delegate=Create Event()
   ) -> (then)
   > SetNavigationRuleCustom(self=[Get ViewTimeLimits_Button(), Get BackToDisco_Button(), Get WBP_Results_PolyTileView()], Direction=Up, InCustomDelegate=Create Event())
   n1 = Get Tabs_Search_Results()
   > SetNavigationRuleCustom(self=@n1, Direction=Up, InCustomDelegate=Create Event())
   > SetNavigationRuleCustom(self=@n1, Direction=Down, InCustomDelegate=Create Event())
   > SetNavigationRuleCustom(self=[Get SearchInputField(), Get TabGroup_Search()], Direction=Up, InCustomDelegate=Create Event())
   > SetNavigationRuleCustom(self=Get WBP_Results_PolyTileView(), Direction=Up, InCustomDelegate=Create Event())
  ]}
 SetupTabSurface {nodes [
   SetupTabSurface() {is=title K2Node_FunctionEntry}
   > n1 = Cast To WBP_Search_Tab_Button(Object=GetTabButtonBaseByID(self=Get TabGroup_Search(), TabNameID=Get CurrentTabId() -> (CurrentTabId=None)))
   > n2 = Set SelectabTab(SelectabTab:WBP Search Tab Button Object Reference=@n1) -> (Output_Get:WBP Search Tab Button Object Reference)
   n3 = Get FortDiscoverSurfaceTileItemVM!(self:WBP Search Tab Button Object Reference=Get SelectabTab() -> (SelectabTab:WBP Search Tab Button Object Reference)) -> (FortDiscoverSurfaceTileItemVM:Fort Discover Surface Tile Item VM Object Reference)
   > SetFortDiscoverSurfaceTileItemVM!(execute, self:WBP Search Tab Button Object Reference=@n2, Viewmodel:Fort Discover Surface Tile Item VM Object Reference=@n3) -> (then)
   n4 = Get FortDiscoverSurfaceViewModel()
   > SetSurfaceName!(
    execute
    self:Discover Surface View Model Object Reference=@n4
    InSurfaceName:Name=GetReferenceSurfaceName!(self:Fort Discover Surface Tile Item VM Object Reference=@n3) -> (ReturnValue:Name=None)
   ) -> (then)
   > Set Last Hydration Index Trigger(Last Hydration Index Trigger=-1) -> (Output_Get=0)
   n5 = Get SelectabTab()
   > SetBottomButtonVisibility(
    self=Get WBP_BottomBarHack_ClearBadges()
    ShouldButtonBeVisible="integer > integer"(A=Get TabNotificationCount!(self:WBP Search Tab Button Object Reference=@n5) -> (TabNotificationCount:int=0))
   )
   > FireDiscoverTabEnteredAnalyticEvent(self:Discover Surface View Model Object Reference=@n4)
   > Set SendAnalyticsOnExit(SendAnalyticsOnExit=true)
   > Set SendSearchTabEnterAnalytics(SendSearchTabEnterAnalytics=true)
  ]}
 ShouldUpdateScrollToTopForMouse {nodes [
   ShouldUpdateScrollToTopForMouse!() -> (then, MouseEvent:Pointer Event Structure) {is=title K2Node_FunctionEntry}
   n1 = Get Current Input Type()
   n2 = GetScrollOffset(self=Get ListView(self=Get WBP_Poly_MegaList()))
   n3 = Get Wheel Delta(Input=Get MouseEvent() -> (MouseEvent:Pointer Event Structure))
   > Return Node!(
    execute
    ReturnValue:bool=AND Boolean!(
     A="float < float"(A=@n2, B=0.5)
     B="Equal (Enum)"(A:ECommonInputType Enum=@n1, B:ECommonInputType Enum=MouseAndKeyboard)
     C:bool="float > float"(A=@n3)
    ) -> (ReturnValue)
   )
  ]}
 UnbindEvents {nodes [
   UnbindEvents() {is=title K2Node_FunctionEntry}
   n1 = Get TileView(self=Get WBP_Results_PolyTileView())
   > Unbind Event from On Entry Generated(self:List View Base Object Reference=@n1, Delegate=Create Event()) {is=title K2Node_RemoveDelegate}
   > Unbind Event from On List View Scrolled(self:List View Object Reference=@n1, Delegate=Create Event()) {is=title K2Node_RemoveDelegate}
   > Unbind Event from Navigate Up Out Of Mega List(self:WBP Poly Mega List Object Reference=Get WBP_Poly_MegaList(), Delegate=Create Event()) {is=title K2Node_RemoveDelegate}
   > Unbind all Events from On Back Button Clicked!(execute:Exec, self:WBP Discover Surface Base Object Reference) -> (then:Exec) {is=title K2Node_ClearDelegate}
   > Unbind Event from On Input Method Changed(self=Get CommonInputSubsystem(), Delegate=Create Event())
   n2 = Get SearchInputField()
   > Unbind Event from On Clicked(self:Common Button Base Object Reference=Get Button_Clear(self=@n2), Delegate=Create Event()) {is=title K2Node_RemoveDelegate}
   > Unbind Event from On Search Added to Focus Path(self:WBP UIKit Filtered Input Field Object Reference=@n2, Delegate=Create Event()) {is=title K2Node_RemoveDelegate}
   > Unbind Event from On List View Scrolled(self=Get ListView(self=Get WBP_Poly_MegaList()), Delegate=Create Event())
   > UnbindToAthenaFrontendTabChange!(execute, Context:Widget Object Reference=Self-Reference()) -> (then)
   > Unbind Event from On List View Touch End(self:List View Object Reference=Get ListView(self=Get WBP_Poly_MegaList()), Delegate=Create Event()) {is=title K2Node_RemoveDelegate}
   > Unbind Event from On Switcher List Active Widget Changed(self:WBP Discover Surface Base Object Reference, Delegate=Create Event()) {is=title K2Node_RemoveDelegate}
  ]}
 UpFromSearchResultTabs {nodes [
   UpFromSearchResultTabs!() -> (then, Navigation:EUINavigation Enum) {is=title K2Node_FunctionEntry}
   n1 = Get InputField(self=Get SearchInputField())
   > Return Node(ReturnValue:Widget Object Reference=Get EditableText(self=Get Block_EditableText(self=@n1)))
  ]}
 UpToTabs {nodes [
   UpToTabs!() -> (then, Navigation:EUINavigation Enum) {is=title K2Node_FunctionEntry}
   n1 = Select!(
    Option 0:Common Tab List Widget Base Object Reference
    Option 1:Common Tab List Widget Base Object Reference=Get TabGroup_Search()
    Option 2:Common Tab List Widget Base Object Reference
    Option 3:Common Tab List Widget Base Object Reference=Get Tabs_Search_Results()
    Index:int=GetActiveWidgetIndex(self=Get Switcher_List()) -> (ReturnValue=0)
   ) -> (ReturnValue:Common Tab List Widget Base Object Reference)
   > PrintString(InString=Append!(A="Navigation: Up to selected tab = ", B=GetDisplayName(Object=@n1)) -> (ReturnValue))
   n2 = GetSelectedTabId!(self:Common Tab List Widget Base Object Reference=@n1) -> (ReturnValue:Name=None)
   n3 = Get Block_EditableText(self=Get InputField(self=Get SearchInputField()))
   > Return Node(ReturnValue=Select!(
     Option 0:Widget Object Reference=GetTabButtonBaseByID(self=@n1, TabNameID=@n2)
     Option 1:Widget Object Reference=Get EditableText(self=@n3)
     Index:bool=Get IsPlayTime Out()
    ) -> (ReturnValue:Widget Object Reference))
  ]}
 UpdateEmptyStateForParentalControl {nodes [
   UpdateEmptyStateForParentalControl() {is=title K2Node_FunctionEntry}
   > n1 = Branch(Condition=Get IsPlayTime Out())
   > n2 = SetMessageTextOverrides(
    self=Get WBP_Discovery_BlankTiles_RoundedCorners()
    InTitle="LOCTEXT(\"YOUR PLAY TIME IS UP\")"
    InSubTitle="LOCTEXT(\"You can test unpublished islands by searching for their island code directly.\")"
   )
   SetVisibility(
    execute=[@n15.then, @n11.then, @n2.then]
    self=Get ViewTimeLimits_Button()
    InVisibility=Select(Option 0:ESlateVisibility Enum=Collapsed, Option 1:ESlateVisibility Enum=SelfHitTestInvisible, Index=Get IsPlayTime Out()) -> (ReturnValue:ESlateVisibility Enum)
   )
   > SetVisibility(self=Get BackToDisco_Button(), InVisibility=Select(Option 0=SelfHitTestInvisible, Option 1=Collapsed, Index=Get IsPlayTime Out()))
   n3 = Cast To FortDiscoverTileItemVM(execute=@n1.else, Object=GetTabViewModel(self=Get TabGroup_Search(), TabID=Get CurrentTabId() -> (CurrentTabId=None)))
   n4 = GetMnemonic(self=@n3)
   n5 = Find(TargetMap=Get DiscoverReduxTabEmptyResultsText(self=GetRuntimeOptions()), Key=@n4)
   > n6 = Branch(Condition=@n5.ReturnValue)
   n7 = Break Discover Redux Tab Empty Results Text Entry(DiscoverReduxTabEmptyResultsTextEntry=@n5)
   n8 = TextIsEmpty(InText=@n7)
   n9 = TextIsEmpty(InText=@n7.Subtitle)
   > n10 = Branch(Condition=NOT Boolean(A=OR Boolean(A=@n8, B=@n9)))
   > n11 = SetMessageTextOverrides(self=Get WBP_Discovery_BlankTiles_RoundedCorners(), InTitle=@n7, InSubTitle=@n7.Subtitle)
   n12 = LogString(
    execute=@n10.else
    InString="To String (Text)"(InText=Format Text!(
      Format="LOCTEXT(\"FortRuntimeOptions - DiscoverReduxTabEmpty: Found entry for mnemonic {0} but no title or subtitle is set. Defaulting to generic text message.\")"
      0:String=@n4
     ) -> (Result))
   )
   n13 = Cast To WBP_Search_Tab_Button(execute=[@n12.then, @n6.else], Object=GetTabButtonBaseByID(self=Get TabGroup_Search(), TabNameID=Get CurrentTabId() -> (CurrentTabId=None)))
   n14 = Get TabText(self=@n13)
   > n15 = SetMessageTextOverrides(
    self=Get WBP_Discovery_BlankTiles_RoundedCorners()
    InTitle=Format Text!(Format="LOCTEXT(\"no {TabText} games\")", TabText:Text=@n14) -> (Result)
    InSubTitle=Format Text(Format="LOCTEXT(\"Your {TabText} games will appear here.\")", TabText=@n14)
   )
  ]}
 UpdateTabPromptVisibility {nodes [
   UpdateTabPromptVisibility() {is=title K2Node_FunctionEntry}
   n1 = Get IsPlayTime Out()
   > SetVisibility(
    self=[Get PrevTabAction(), Get NextTabAction()]
    InVisibility=Select!(
     MouseAndKeyboard:ESlateVisibility Enum=Collapsed
     Gamepad:ESlateVisibility Enum=Select!(Option 0:ESlateVisibility Enum=SelfHitTestInvisible, Option 1:ESlateVisibility Enum=Collapsed, Index:bool=@n1) -> (ReturnValue)
     Touch:ESlateVisibility Enum=Collapsed
     Count:ESlateVisibility Enum=Collapsed
     Index:ECommonInputType Enum=Get Current Input Type()
    ) -> (ReturnValue)
   )
   n2 = Get Current Input Type()
   > Set Grid Slot Column and Row!(
    execute:Exec
    Widget:Widget Object Reference=Get SearchInputField()
    Row:int
    RowSpan:int
    Column:int=SelectInt(
     A=0
     B=Select!(MouseAndKeyboard:int=0, Gamepad:int=1, Touch:int=0, Count:int=0, Index:ECommonInputType Enum=@n2) -> (ReturnValue:int)
     bPickA=IsUsingCompactUI()
    )
    ColumnSpan:int
   ) -> (then:Exec) {is=title K2Node_MacroInstance}
   > Set Grid Slot Column and Row(Widget=Get TabGroup_Search(), Column=Select(MouseAndKeyboard=1, Gamepad=2, Touch=1, Count=1, Index=Get Current Input Type()), ColumnSpan=1)
   n3 = IsUsingCompactUI()
   n4 = "Equal (Enum)"(A=Get Current Input Type(), B=Gamepad)
   > Set Grid Slot Padding!(
    execute:Exec
    Widget:Widget Object Reference=Get SearchInputField()
    Padding_Left:float=SelectFloat(A=52, B=0, bPickA=AND Boolean!(A=@n3, B=@n4) -> (ReturnValue))
    Padding_Top:float=0
    Padding_Right:float=0
    Padding_Bottom:float=0
   ) -> (then:Exec) {is=title K2Node_MacroInstance}
  ]}
 UpdateTileSizeAndShapeThenTileView {nodes [
   n1 = UpdateTileSizeAndShapeThenTileView!() -> (then, TileBase:WBP Discover Tile Base Object Reference) {is=title K2Node_FunctionEntry}
   n2 = SelectInt(A=8, B=3, bPickA=UseNumberOfColumns!() -> (ReturnValue:bool))
   n3 = "Equal (Enum)"(A:EDiscoverSearchMode Enum=Get SearchMode() -> (SearchMode:EDiscoverSearchMode Enum=Island), B:EDiscoverSearchMode Enum=Creator)
   > Set Tile Size!(
    execute
    self:WBP Discover Tile Base Object Reference=@n1
    InSize:Vector 2D Structure=ColumnSizeToTileSize!(self:WBP Discover Tile Base Object Reference=@n1, NewColumnSize:int=SelectInt(A=@n2, B=5, bPickA=@n3), IsSquareTiles:bool=@n3) -> (NewTileSize:Vector 2D Structure)
   ) -> (then) {is=title K2Node_CallFunction}
   n4 = BreakVector2D(InVec=Get Dynamic Tile Size!(self:WBP Discover Tile Base Object Reference=@n1) -> (DynamicTileSize:Vector 2D Structure) {is=title K2Node_CallFunction})
   n5 = Get TileView(self=Get WBP_Results_PolyTileView())
   > SetEntryWidth!(execute, self:Tile View Object Reference=@n5, NewWidth:float=@n4) -> (then)
   > SetEntryHeight!(execute, self:Tile View Object Reference=@n5, NewHeight:float=@n4.Y) -> (then)
  ]}
 UpdateVisual {nodes [
   UpdateVisual() {is=title K2Node_FunctionEntry}
   > RemoveFromParent(self=[
     Get HeaderImage() -> (HeaderImage:Image Object Reference)
     Get Background() -> (Background:Image Object Reference)
     Get HeaderLoadingThrobber() -> (HeaderLoadingThrobber:WBP UIKit Throbber Object Reference)
    ])
   > SetColorAndOpacity!(
    execute
    self=Get Text_HeaderRow!(self:WBP Row Header Object Reference=Get WBP_Row_Header()) -> (Text_HeaderRow:Common Text Object Reference)
    InColorAndOpacity_SpecifiedColor:Linear Color Structure=(R=1,G=1,B=1,A=1)
    InColorAndOpacity_ColorUseRule:ESlateColorStylingMode Enum
   ) -> (then)
   n1 = Get InputField(self=Get SearchInputField())
   > SetPlacement!(
    execute
    self:WBP UIKit Input Field Base Object Reference=@n1
    Widget:Widget Object Reference=Get Block_Icon!(self:WBP UIKit Input Field Base Object Reference=@n1) -> (Block_Icon:WBP UIKit Block Image Object Reference)
    Placement:S UI Placement Structure=Get BlockIconPlacement() -> (BlockIconPlacement:S UI Placement Structure)
   ) -> (then)
   > SetPlacement(
    self=@n1
    Widget=Get Block_EditableText(self=@n1)
    Placement=Get BlockEditableTextPlacement() -> (BlockEditableTextPlacement:S UI Placement Structure)
   )
   n2 = IsUsingCompactUI()
   n3 = SelectFloat(A=24, B=10, bPickA=@n2)
   n4 = Get InputField(self=Get SearchInputField())
   > SetVectorParameterValue!(
    execute
    self=Get BackgroundDynamicMaterial!(self:WBP UIKit Block Background Base Object Reference=Get Block_Background!(self:WBP UIKit Input Field Base Object Reference=@n4) -> (Block_Background:WBP UIKit Block Background Base Object Reference)) -> (BackgroundDynamicMaterial:Material Instance Dynamic Object Reference)
    ParameterName=Corner Radius
    Value_R:float=@n3
    Value_G:float=@n3
    Value_B:float=@n3
    Value_A:float=@n3
   ) -> (then)
   n5 = SelectFloat(A=36, B=20, bPickA=@n2)
   n6 = Get SearchInputField()
   > SetCornerRadiuses!(
    execute
    self:WBP UIKit Block Outline Object Reference=Get Block_Outline!(self:WBP UIKit Input Field Base Object Reference=Get InputField(self=@n6)) -> (Block_Outline:WBP UIKit Block Outline Object Reference)
    TopLeft:float=@n5
    TopRight:float=@n5
    BottomLeft:float=@n5
    BottomRight:float=@n5
    RadiusInPixels:bool=true
   ) -> (then)
   > SetTabsNoResultsSize!(
    execute
    self
    DiscoveryBlankTilesRoundedCorners:WBP Discovery Blank Tiles Rounded Corners Object Reference=Get WBP_Discovery_BlankTiles_RoundedCorners()
   ) -> (then) {is=title K2Node_CallFunction}
   > SetTabsNoResultsSize(DiscoveryBlankTilesRoundedCorners=Get WBP_Discovery_BlankTiles_RoundedCorners_1() -> (WBP_Discovery_BlankTiles_RoundedCorners_1:WBP Discovery Blank Tiles Rounded Corners Object Reference))
   > SetDesiredSizeOverride!(execute, self:Image Object Reference=Get Image_ListSoftBottomEdge(), DesiredSize:Vector 2D Structure=(X=32,Y=50)) -> (then)
   > SetClipping!(
    execute
    self:Widget Object Reference=Get ScrollBox_Tabs!(self:WBP UIKit Tab Group Base Object Reference=Get TabGroup_Search()) -> (ScrollBox_Tabs:Scroll Box Object Reference)
    InClipping:EWidgetClipping Enum=Inherit
   ) -> (then)
  ]}
]
variables [
 Anim_OutputBackgroundColor {type=number, default=1}
 BroadcastTabSearchSelectedIndex {}
 bShouldFocusOnMegaListItem {type=boolean}
 CachedTab {type=E_Discover_Background_Color, default=Discover}
 CachedTabIndex {type=integer}
 Current Input Type {type=ECommonInputType, default=MouseAndKeyboard}
 Current Page {type=integer}
 CurrentTabId {type=string, default=None}
 DiscoverTileBaseReference {type=WBP_Discover_Tile_Base_C, default=null}
 InitialTabIndex {type=integer}
 Is Row Lock to Top {type=boolean}
 IsBladeExpandContractAnimationFinished {type=boolean}
 IsDirectionUp {type=boolean}
 IsFirstActivation {type=boolean, default=true}
 IsPlayTime Out {type=boolean}
 IsTileViewSearchResultsNotEmpty {type=boolean}
 Last Active Widget {type=Widget, default=null}
 Last Hydration Index Trigger {type=integer, default=-1}
 LastFocusedWidget {type=Widget, default=null}
 MegalistItem Offset {type=number}
 OnBackButtonHackClicked {}
 OnListScrolledOut {}
 OnRequestNavigateToDiscover {}
 OnSearchPeekButtonClicked {}
 OutputBackgroundColor {type=LinearColor, default=<rgba 0 0 0 0>}
 RequestRefreshFocusAfterAnimation {type=boolean}
 SearchBarActionInputNextTab {type=DataTableRowHandle, default {dataTable=&InputActionDataTable, rowName=DiscoverReduxTabRight}}
 SearchBarActionInputPreviousTab {type=DataTableRowHandle, default {dataTable=&InputActionDataTable, rowName=DiscoverReduxTabLeft}}
 SearchMode {type=EDiscoverSearchMode, default=Island}
 SelectedTab {type=E_Discover_Background_Color, default=Discover}
 SendAnalyticsOnExit {type=boolean}
 SendSearchTabEnterAnalytics {type=boolean, default=true}
 TabBlockTextArray {type="array<WBP_UIKit_Block_Text_C>", default []}
]
widgets {
 info {
  inheritedWidgetCount=0
  namedSlotCount=7
  parentClass=/DiscoveryBrowser/WBP_Discover_Surface_Base.WBP_Discover_Surface_Base_C
  rootWidgetClass=null
  widgetCount=43
 }
 widgets [
  Overlay_0 {widgetClassPath=&Overlay, widget=Overlay_0}
  GP_SearchBar {
   widgetClassPath=&GridPanel
   bIsVariable=true
   parent=Overlay_0
   widget=GP_SearchBar
   properties {navigation=&self:WidgetTree.GP_SearchBar.WidgetNavigation_0, columnFill [1,0], rowFill [1]}
   slot {class=OverlaySlot, properties {horizontalAlignment=HAlign_Fill, padding=<ltrb 0 0 0 24>}}
  }
  HorizontalBox_1 {
   widgetClassPath=&HorizontalBox
   parent=GP_SearchBar
   widget=HorizontalBox_1
   slot {class=GridSlot, properties {verticalAlignment=VAlign_Center}}
  }
  PrevTabAction {
   widgetClassPath=&CommonActionWidget
   bIsVariable=true
   parent=HorizontalBox_1
   widget=PrevTabAction
   properties {visibility=Collapsed}
   slot {class=HorizontalBoxSlot, properties {padding=<ltrb 0 0 12 0>}}
  }
  SearchInputField {
   widgetClassPath=/FortUILibrary/UIKit/InputField/WBP_UIKit_FilteredInputField
   bIsVariable=true
   parent=HorizontalBox_1
   widget=SearchInputField
   properties {
    bIsFocusable=true
    hideValidationInfo=true
    hintText=Search Discover
    isSearchField=true
    useHintText=true
    clearButtonPadding=<ltrb 0 0 12 0>
    desiredImageSize=<56 56>
    icon Properties {
     material=/FortUILibrary/Systems/InputField/MaterialInstances/MI_UIKit_InputFieldIcon
     image Size=<32 32>
     placement {
      horizontal Alignment=HAlign_Left
      vertical Alignment=VAlign_Center
      grid {column=0, column Span=0, layer=0, row=1, row Span=0, nudge=<0 0>}
      padding=<ltrb 16 0 0 0>
     }
    }
    inputActionPromptPadding=<ltrb 0 0 12 0>
    scaleOverride {enable=true, value=1.2}
   }
   slot {class=HorizontalBoxSlot, properties {size {sizeRule=Fill, value=1}}}
  }
  HorizontalBox_2 {
   widgetClassPath=&HorizontalBox
   parent=GP_SearchBar
   widget=HorizontalBox_2
   slot {class=GridSlot, properties {column=2, columnSpan=1, verticalAlignment=VAlign_Center, padding=<ltrb 24 0 0 0>}}
  }
  TabGroup_Search {
   widgetClassPath=/FortUILibrary/UIKit/Tabs/WBP_UIKit_TabGroup
   bIsVariable=true
   parent=HorizontalBox_2
   widget=TabGroup_Search
   properties {
    bIsFocusable=true
    bShouldWrapNavigation=false
    backplateBrightness=Dark
    buttonClass=/DiscoveryBrowser/WBP_Search_Tab_Button.WBP_Search_Tab_Button_C
    buttonHorizontalAlignment=HAlign_Left
    hideMouseAndKeyboardPrompts=true
    hidePrompts=true
    inputActions=Custom
    isNaturalNavigation=true
    promptBackplateCornerRadius=Interactive Component
    tabGroupHorizontalAlignment=HAlign_Fill
    tabSpacing=12
    useInitialTabs=true
    initialTabsData [
     {bEnableTabId=true, tabId=CreativeDiscoverySurface_Nested_Category_BR, tabViewModel=null}
     {bEnableTabId=true, tabId=CreativeDiscoverySurface_Nested_Collab_Lego, tabViewModel=null}
     {bEnableTabId=true, tabId=CreativeDiscoverySurface_Search_V2, tabViewModel=null}
     {bEnableTabId=true, tabId=CreativeDiscoverySurface_Nested_Category_3v3, tabViewModel=null}
     {bEnableTabId=true, tabId=Empty, tabViewModel=null}
    ]
    minButtonSize=<0 72>
    nextTabInputActionData {dataTable=null, rowName=DiscoverReduxTabRight}
    previousTabInputActionData {dataTable=null, rowName=DiscoverReduxTabLeft}
    tabGroupPadding=<ltrb 4 4 4 4>
   }
   slot {class=HorizontalBoxSlot}
  }
  NextTabAction {
   widgetClassPath=&CommonActionWidget
   bIsVariable=true
   parent=HorizontalBox_2
   widget=NextTabAction
   properties {visibility=Collapsed}
   slot {class=HorizontalBoxSlot, properties {padding=<ltrb 12 0 0 0>}}
  }
  SearchResult_Switcher_0 {widgetClassPath=&CommonWidgetSwitcherLegacy, bIsVariable=true, widget=SearchResult_Switcher_0}
  SearchSpinner {
   widgetClassPath=&GridPanel
   bIsVariable=true
   parent=SearchResult_Switcher_0
   widget=SearchSpinner
   properties {columnFill [1], rowFill [1]}
   slot {class=WidgetSwitcherSlot, properties {padding=<ltrb 0 0 0 80>}}
  }
  UIKit_Throbber_Search {
   widgetClassPath=/FortUILibrary/UIKit/Throbber/WBP_UIKit_Throbber
   bIsVariable=true
   parent=SearchSpinner
   widget=UIKit_Throbber_Search
   properties {displayBackground=false, throbberSize=Large}
   slot {class=GridSlot, properties {horizontalAlignment=HAlign_Center, verticalAlignment=VAlign_Center}}
  }
  SearchResultsGrid {
   widgetClassPath=&GridPanel
   bIsVariable=true
   parent=SearchResult_Switcher_0
   widget=SearchResultsGrid
   properties {columnFill [1], rowFill [0,1]}
   slot {class=WidgetSwitcherSlot}
  }
  Tabs_Search_Results {
   widgetClassPath=/FortUILibrary/UIKit/Tabs/WBP_UIKit_Tabs
   bIsVariable=true
   parent=SearchResultsGrid
   widget=Tabs_Search_Results
   properties {
    bIsFocusable=true
    backplateCornerRadius=Button
    buttonHorizontalAlignment=HAlign_Left
    initialSelectedTabIndex=0
    inputActions=None
    isNaturalNavigation=true
    minButtonHeight=30
    minButtonWidth=50
    navigation=&self:WidgetTree.Tabs_Search_Results.WidgetNavigation_0
    useBackplate=true
    localTabsViewModels [null,null]
    tabGroupPadding=<ltrb 4 4 4 4>
   }
   slot {class=GridSlot, properties {horizontalAlignment=HAlign_Left, layer=1, nudge=<-50 0>, padding=<ltrb 132 0 0 48>}}
  }
  CommonWidgetSwitcher_Results {
   widgetClassPath=&CommonWidgetSwitcherLegacy
   bIsVariable=true
   parent=SearchResultsGrid
   widget=CommonWidgetSwitcher_Results
   slot {class=GridSlot, properties {layer=2, row=1}}
  }
  VerticalBox_IslandBox {
   widgetClassPath=&VerticalBox
   bIsVariable=true
   parent=CommonWidgetSwitcher_Results
   widget=VerticalBox_IslandBox
   properties {clipping=ClipToBoundsWithoutIntersecting}
   slot {class=WidgetSwitcherSlot}
  }
  WBP_Row_Header {
   widgetClassPath=&/WBP_Row_Header
   bIsVariable=true
   parent=VerticalBox_IslandBox
   widget=WBP_Row_Header
   slot {class=VerticalBoxSlot}
  }
  CommonWidgetSwitcher_OutOfTimeSearch {
   widgetClassPath=&CommonWidgetSwitcherLegacy
   bIsVariable=true
   parent=VerticalBox_IslandBox
   widget=CommonWidgetSwitcher_OutOfTimeSearch
   slot {class=VerticalBoxSlot, properties {size {sizeRule=Fill, value=1}}}
  }
  O_ResultTileList {
   widgetClassPath=&Overlay
   bIsVariable=true
   parent=CommonWidgetSwitcher_OutOfTimeSearch
   widget=O_ResultTileList
   slot {class=WidgetSwitcherSlot}
  }
  CLIPPING_AND_PADDING_FIX {
   widgetClassPath=&VerticalBox
   parent=O_ResultTileList
   widget=CLIPPING_AND_PADDING_FIX
   properties {bOverride_Cursor=true, clipping=ClipToBounds}
   slot {class=OverlaySlot, properties {horizontalAlignment=HAlign_Fill, verticalAlignment=VAlign_Fill, padding=<ltrb 58 0 80 0>}}
  }
  WBP_Results_PolyTileView {
   widgetClassPath=/DiscoveryBrowser/WBP_Results_PolyTileView
   bIsVariable=true
   parent=CLIPPING_AND_PADDING_FIX
   widget=WBP_Results_PolyTileView
   properties {
    navigation=&self:WidgetTree.WBP_Results_PolyTileView.WidgetNavigation_0
    desiredFocusWidget {widgetName=None}
   }
   slot {class=VerticalBoxSlot, properties {size {sizeRule=Fill, value=1}}}
  }
  Spacer_99 {
   widgetClassPath=&Spacer
   parent=CLIPPING_AND_PADDING_FIX
   widget=Spacer_99
   properties {size=<1 350>}
   slot {class=VerticalBoxSlot}
  }
  Image_SoftEdgeGradient_Islands {
   widgetClassPath=&Image
   bIsVariable=true
   parent=O_ResultTileList
   widget=Image_SoftEdgeGradient_Islands
   properties {
    renderOpacity=0
    visibility=SelfHitTestInvisible
    brush {
     bIsDynamicallyLoaded=false
     drawAs=Image
     imageType=FullColor
     mirroring=NoMirror
     resourceName=None
     resourceObject=/DiscoveryBrowser/TabScreens/DiscoverHome/Materials/MI_UI_Discovery_NebulousBackgroundWithTopMask
     tiling=NoTile
     imageSize=<32 50>
     margin=<ltrb 0 0 0 0>
     outlineSettings {
      bUseBrushTransparency=false
      roundingType=HalfHeightRadius
      width=0
      color {colorUseRule=UseColor_Specified, specifiedColor=<rgba 0 0 0 0>}
      cornerRadii=<0 0 0 0>
     }
     tintColor {colorUseRule=UseColor_Specified, specifiedColor=<rgba 1 1 1 1>}
     uVRegion {bIsValid=false, max=<0 0>, min=<0 0>}
    }
   }
   slot {class=OverlaySlot, properties {horizontalAlignment=HAlign_Fill}}
  }
  GP_NoResultsCreateTime {
   widgetClassPath=&GridPanel
   bIsVariable=true
   parent=CommonWidgetSwitcher_OutOfTimeSearch
   widget=GP_NoResultsCreateTime
   properties {columnFill [1], rowFill [1]}
   slot {class=WidgetSwitcherSlot}
  }
  WBP_Discovery_BlankTiles_RoundedCorners_1 {
   widgetClassPath=/DiscoveryBrowser/WBP_Discovery_BlankTiles_RoundedCorners
   bIsVariable=true
   parent=GP_NoResultsCreateTime
   widget=WBP_Discovery_BlankTiles_RoundedCorners_1
   properties {
    entries Count=24
    flavorSubtitleOverride=You can test unpublished islands by searching for their island code directly.
    flavorTitleOverride=YOUR PLAY TIME IS UP
   }
   slot {class=GridSlot, properties {verticalAlignment=VAlign_Top, nudge=<-12 0>, padding=<ltrb 80 0 80 0>}}
  }
  HorizontalBox {
   widgetClassPath=&HorizontalBox
   namedSlotHost=WBP_Discovery_BlankTiles_RoundedCorners_1
   widget=HorizontalBox
   properties {bOverride_Cursor=true}
  }
  ViewTimeLimitsResults_Button {
   widgetClassPath=&WBP_UIKit_Button_Regular
   bIsVariable=true
   parent=HorizontalBox
   widget=ViewTimeLimitsResults_Button
   properties {
    bHideInputAction=true
    navigation=&self:WidgetTree.ViewTimeLimitsResults_Button.WidgetNavigation_0
    text=View Time Limits
   }
   slot {
    class=HorizontalBoxSlot
    properties {horizontalAlignment=HAlign_Center, verticalAlignment=VAlign_Bottom, padding=<ltrb 0 32 0 0>, size {sizeRule=Fill, value=1}}
   }
  }
  WBP_NoResults_SearchResults {
   widgetClassPath=/DiscoveryBrowser/WBP_Discovery_MessageNoResults.WBP_Discovery_NoResultsMessage
   bIsVariable=true
   parent=CommonWidgetSwitcher_Results
   widget=WBP_NoResults_SearchResults
   properties {flavorMessageBase="Please search for a different word or phrase ", flavorTitleBase=No Results Found}
   slot {class=WidgetSwitcherSlot, properties {verticalAlignment=VAlign_Center, padding=<ltrb 0 0 0 80>}}
  }
  Discovery_BlankTiles_SwipePanel {
   widgetClassPath=/Script/FortniteUI.FortSwipePanel
   bIsVariable=true
   widget=Discovery_BlankTiles_SwipePanel
   properties {thresholdUnit=Inches, swipeThreshold=<0 0.25>}
  }
  GridPanel_1 {
   widgetClassPath=&GridPanel
   parent=Discovery_BlankTiles_SwipePanel
   widget=GridPanel_1
   properties {columnFill [1], rowFill [1]}
   slot {class=FortSwipePanelSlot}
  }
  WBP_Discovery_BlankTiles_RoundedCorners {
   widgetClassPath=&WBP_Discovery_BlankTiles_RoundedCorners
   bIsVariable=true
   parent=GridPanel_1
   widget=WBP_Discovery_BlankTiles_RoundedCorners
   properties {entries Count=24}
   slot {class=GridSlot, properties {verticalAlignment=VAlign_Top, padding=<ltrb 80 0 80 0>}}
  }
  HorizontalBox_3 {
   widgetClassPath=&HorizontalBox
   namedSlotHost=WBP_Discovery_BlankTiles_RoundedCorners
   widget=HorizontalBox_3
   properties {bOverride_Cursor=true}
  }
  BackToDisco_Button {
   widgetClassPath=/FortUILibrary/UIKit/Buttons/WBP_UIKit_Button_Loud
   bIsVariable=true
   parent=HorizontalBox_3
   widget=BackToDisco_Button
   properties {text=back to discover}
   slot {
    class=HorizontalBoxSlot
    properties {horizontalAlignment=HAlign_Center, verticalAlignment=VAlign_Bottom, padding=<ltrb 0 32 0 0>, size {sizeRule=Fill, value=1}}
   }
  }
  ViewTimeLimits_Button {
   widgetClassPath=&WBP_UIKit_Button_Regular
   bIsVariable=true
   parent=HorizontalBox_3
   widget=ViewTimeLimits_Button
   properties {
    bHideInputAction=true
    navigation=&self:WidgetTree.ViewTimeLimits_Button.WidgetNavigation_0
    text=View Time Limits
    visibility=Collapsed
   }
   slot {
    class=HorizontalBoxSlot
    properties {horizontalAlignment=HAlign_Center, verticalAlignment=VAlign_Bottom, padding=<ltrb 0 32 0 0>, size {sizeRule=Fill, value=1}}
   }
  }
  Overlay_2 {widgetClassPath=&Overlay, widget=Overlay_2}
  HorizontalBox_0 {
   widgetClassPath=&HorizontalBox
   parent=Overlay_2
   widget=HorizontalBox_0
   properties {bOverride_Cursor=true}
   slot {class=OverlaySlot, properties {horizontalAlignment=HAlign_Right, verticalAlignment=VAlign_Bottom}}
  }
  WBP_BottomBarHack_ClearResults {
   widgetClassPath=/Game/UI/Frontend/BottomBar/WBP_BottomBarDecoyButton
   bIsVariable=true
   parent=HorizontalBox_0
   widget=WBP_BottomBarHack_ClearResults
   properties {minWidth=120, triggeringInputAction {dataTable=null, rowName=DiscoverReduxClearResults}}
   slot {class=HorizontalBoxSlot}
  }
  WBP_BottomBarHack_ClearBadges {
   widgetClassPath=&WBP_BottomBarDecoyButton
   bIsVariable=true
   parent=HorizontalBox_0
   widget=WBP_BottomBarHack_ClearBadges
   properties {minWidth=120, triggeringInputAction {dataTable=null, rowName=DiscoverReduxClearBadges}}
   slot {class=HorizontalBoxSlot}
  }
  DecoyButton_LeftTabTriggerNav {
   widgetClassPath=&WBP_BottomBarDecoyButton
   bIsVariable=true
   parent=HorizontalBox_0
   widget=DecoyButton_LeftTabTriggerNav
   properties {
    bDisplayInActionBar=false
    minWidth=120
    triggeringInputAction {dataTable=&InputActionDataTable, rowName=DiscoverReduxTabLeft}
   }
   slot {class=HorizontalBoxSlot}
  }
  DecoyButton_TabRightTriggerNav {
   widgetClassPath=&WBP_BottomBarDecoyButton
   bIsVariable=true
   parent=HorizontalBox_0
   widget=DecoyButton_TabRightTriggerNav
   properties {
    bDisplayInActionBar=false
    minWidth=120
    triggeringInputAction {dataTable=&InputActionDataTable, rowName=DiscoverReduxTabRight}
   }
   slot {class=HorizontalBoxSlot}
  }
  HACK_DiscoFocusCatcher {
   widgetClassPath=&Button
   parent=Overlay_2
   widget=HACK_DiscoFocusCatcher
   properties {
    renderOpacity=0
    visibility=SelfHitTestInvisible
    widgetStyle {
     clickedSlateSound {resourceObject=null}
     disabled {
      bIsDynamicallyLoaded=false
      drawAs=NoDrawType
      imageType=NoImage
      mirroring=NoMirror
      resourceName=None
      resourceObject=null
      tiling=NoTile
      imageSize=<0 0>
      margin=<ltrb 0 0 0 0>
      outlineSettings {
       bUseBrushTransparency=false
       roundingType=HalfHeightRadius
       width=0
       color {colorUseRule=UseColor_Specified, specifiedColor=<rgba 0 0 0 0>}
       cornerRadii=<0 0 0 0>
      }
      tintColor {colorUseRule=UseColor_Specified, specifiedColor=<rgba 1 1 1 1>}
      uVRegion {bIsValid=false, max=<0 0>, min=<0 0>}
     }
     disabledForeground {colorUseRule=UseColor_Specified, specifiedColor=<rgba 0.5271 0.5271 0.5271 1>}
     hovered {
      bIsDynamicallyLoaded=false
      drawAs=RoundedBox
      imageType=NoImage
      mirroring=NoMirror
      resourceName=None
      resourceObject=null
      tiling=NoTile
      imageSize=<1 1>
      margin=<ltrb 0 0 0 0>
      outlineSettings {
       bUseBrushTransparency=true
       roundingType=FixedRadius
       width=1
       color {colorUseRule=UseColor_Specified, specifiedColor=<rgba 0.7243 0.7243 0.7243 1>}
       cornerRadii=<4 4 4 4>
      }
      tintColor {colorUseRule=UseColor_Specified, specifiedColor=<rgba 0.7243 0.7243 0.7243 1>}
      uVRegion {bIsValid=false, max=<0 0>, min=<0 0>}
     }
     hoveredForeground {colorUseRule=UseColor_Specified, specifiedColor=<rgba 0 1 1 1>}
     hoveredSlateSound {resourceObject=null}
     normal {
      bIsDynamicallyLoaded=false
      drawAs=RoundedBox
      imageType=NoImage
      mirroring=NoMirror
      resourceName=None
      resourceObject=null
      tiling=NoTile
      imageSize=<1 1>
      margin=<ltrb 0 0 0 0>
      outlineSettings {
       bUseBrushTransparency=true
       roundingType=FixedRadius
       width=1
       color {colorUseRule=UseColor_Specified, specifiedColor=<rgba 0.6951 0.6951 0.6951 1>}
       cornerRadii=<4 4 4 4>
      }
      tintColor {colorUseRule=UseColor_Specified, specifiedColor=<rgba 0.4955 0.4955 0.4955 1>}
      uVRegion {bIsValid=false, max=<0 0>, min=<0 0>}
     }
     normalForeground {colorUseRule=UseColor_Specified, specifiedColor=<rgba 0 1 1 1>}
     normalPadding=<ltrb 0 0 0 0>
     pressed {
      bIsDynamicallyLoaded=false
      drawAs=RoundedBox
      imageType=NoImage
      mirroring=NoMirror
      resourceName=None
      resourceObject=null
      tiling=NoTile
      imageSize=<1 1>
      margin=<ltrb 0 0 0 0>
      outlineSettings {
       bUseBrushTransparency=true
       roundingType=FixedRadius
       width=1
       color {colorUseRule=UseColor_Specified, specifiedColor=<rgba 0.7243 0.7243 0.7243 1>}
       cornerRadii=<4 4 4 4>
      }
      tintColor {colorUseRule=UseColor_Specified, specifiedColor=<rgba 0.3843 0.3843 0.3843 1>}
      uVRegion {bIsValid=false, max=<0 0>, min=<0 0>}
     }
     pressedForeground {colorUseRule=UseColor_Specified, specifiedColor=<rgba 0 1 1 1>}
     pressedPadding=<ltrb 12 1.5 12 1.5>
     pressedSlateSound {resourceObject=null}
    }
   }
   slot {class=OverlaySlot, properties {horizontalAlignment=HAlign_Right, verticalAlignment=VAlign_Bottom}}
  }
  PanelForArrow {widgetClassPath=&GridPanel, bIsVariable=true, widget=PanelForArrow, properties {columnFill [1], rowFill [1]}}
  Touch_Border {
   widgetClassPath=&Border
   parent=PanelForArrow
   widget=Touch_Border
   properties {
    horizontalAlignment=HAlign_Center
    verticalAlignment=VAlign_Center
    background {
     bIsDynamicallyLoaded=false
     drawAs=Image
     imageType=NoImage
     mirroring=NoMirror
     resourceName=None
     resourceObject=null
     tiling=NoTile
     imageSize=<32 32>
     margin=<ltrb 0 0 0 0>
     outlineSettings {
      bUseBrushTransparency=false
      roundingType=HalfHeightRadius
      width=0
      color {colorUseRule=UseColor_Specified, specifiedColor=<rgba 0 0 0 0>}
      cornerRadii=<0 0 0 0>
     }
     tintColor {colorUseRule=UseColor_Specified, specifiedColor=<rgba 1 1 1 0>}
     uVRegion {bIsValid=false, max=<0 0>, min=<0 0>}
    }
    padding=<ltrb 0 0 0 0>
   }
   slot {class=GridSlot, properties {nudge=<0 -48>}}
  }
  WBP_UIKit_ViewMore_ArrowButton {
   widgetClassPath=/FortUILibrary/UIKit/Buttons/ArrowButtons/WBP_UIKit_ViewMore_ArrowButton
   bIsVariable=true
   parent=Touch_Border
   widget=WBP_UIKit_ViewMore_ArrowButton
   properties {style=/FortUILibrary/Foundation/Buttons/ButtonStyles/ButtonStyle-Empty.ButtonStyle-Empty_C}
   slot {class=BorderSlot, properties {horizontalAlignment=HAlign_Center, verticalAlignment=VAlign_Center, padding=<ltrb 0 0 0 0>}}
  }
 ]
}
