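# Base CommonUserWidget for Discovery browser rows, wrapping content in a named slot and defining a shared row interface.
# When a cvar macro selects rollback, PreConstruct sets custom Up/Down boundary navigation on the slot.
# DoCustomNavigation broadcasts OnNavigateToNewOuterRow with the direction and placeholder subrow values.
# It also declares a scroll-offset dispatcher, a zero-returning GetNewSubrowIndexData stub and a cvar-gated helper that stops provider visibility impressions.
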WidgetBlueprint /DiscoveryBrowser/Rows/WBP_Row_Poly_Base
graphs [
 DoCustomNavigation {nodes [
   n1 = DoCustomNavigation!() -> (then, Navigation:EUINavigation Enum) {is=title K2Node_FunctionEntry}
   > n2 = Switch on EUINavigation!(execute, Selection:EUINavigation Enum=@n1) -> (Left:Exec, Right:Exec, Up:Exec, Down:Exec, Next:Exec, Previous:Exec, Invalid:Exec) {is=title K2Node_SwitchEnum}
   Call On Navigate to New Outer Row!(
    execute:Exec=[@n2.Down, @n2.Up]
    self:WBP Row Poly Base Object Reference
    IsDirectionUp:bool="Equal (Enum)"(A:EUINavigation Enum=@n1, B:EUINavigation Enum=Up)
    PreviousSubrowSelectedIndex:int
    PreviousNumItemsPerSubrow:int=1
    PreviousRowWasViewAllTileSelected:bool
   ) -> (then:Exec) {is=title K2Node_CallDelegate}
   > Return Node!(execute, ReturnValue:Widget Object Reference)
  ]}
 EventGraph {nodes [
   Event PreConstruct()
   > n1 = Branch Natural Nav Rollback Cvar!(In:Exec) -> (Use Natural Nav:Exec, Rollback:Exec) {is=title K2Node_MacroInstance}
   SetNavigationRuleCustomBoundary!(
    execute=@n1.Rollback
    self:Widget Object Reference=Get CustomNavigationSlot() -> (CustomNavigationSlot:Named Slot Object Reference)
    Direction:EUINavigation Enum=Up
    InCustomDelegate:Delegate=Create Event()
   ) -> (then)
   > SetNavigationRuleCustomBoundary(self=Get CustomNavigationSlot(), Direction=Down, InCustomDelegate=Create Event())
  ]}
 GetNewSubrowIndexData {nodes [
   GetNewSubrowIndexData() {is=title K2Node_FunctionEntry}
   > Return Node!(
    execute
    NewSubrowLinesCount:int=0
    NewSubrowNumItemsPerRow:int=0
    DisplayedEntries:int=0
    NewSubrowFirstVisibleItemIndex:int=0
    SelectedItemIndex:int=0
   )
  ]}
 OnNavigateToNewOuterRow {nodes [OnNavigateToNewOuterRow!() -> (then, IsDirectionUp:bool=~, PreviousSubrowSelectedIndex:int, PreviousNumItemsPerSubrow:int, PreviousRowWasViewAllTileSelected:bool=~) {is=title K2Node_FunctionEntry}]}
 OnUpdateOuterListScrollOffset {nodes [OnUpdateOuterListScrollOffset!() -> (then, RowIndex:int, SubRowIndex:int, IsViewAll:bool=~) {is=title K2Node_FunctionEntry}]}
 StopVisibilityOnEntryReleased {nodes [
   n1 = StopVisibilityOnEntryReleased!() -> (then, InProviderVM:Discover Provider View Model Object Reference) {is=title K2Node_FunctionEntry}
   > n2 = GetConsoleVariableBoolValue!(execute, VariableName:String=Athena.DiscoverSurface.RowStopVisibilityOnEntryReleased) -> (then, ReturnValue:bool)
   > Branch(Condition=@n2)
   > StopVisibleImpression(self:Discover Provider View Model Object Reference=@n1)
  ]}
]
variables [
 OnNavigateToNewOuterRow {}
 OnUpdateOuterListScrollOffset {}
]
widgets {
 info {inheritedWidgetCount=0, parentClass=&CommonUserWidget, rootWidgetClass=&NamedSlot, widgetCount=1}
 widgets [CustomNavigationSlot {widgetClassPath=&NamedSlot, bIsVariable=true, widget=CustomNavigationSlot}]
}
