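# Discovery browser billboard tile widget showing a featured item's art, title, live/countdown badges, play/trailer button and context menu.
# Setting the tile view model starts the countdown timer, rating-card slot and impression tracking, fetching spot images if needed.
# Clicks open a nested discovery surface, or a video screen when time-locked or age-restricted.
# Dynamic materials take art and hover/focus state; rollback-cvar navigation broadcasts outer-row changes.
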WidgetBlueprint /DiscoveryBrowser/Tiles/Billboard/WBP_Discovery_IP_Tile_Billboard
graphs [
 CloseContextMenu {nodes [
   CloseContextMenu() {is=title K2Node_FunctionEntry}
   > Close(self:Menu Anchor Object Reference=Get Anchor!(self:UIKit Context Menu Anchor Object Reference=Get Billboard_ContextMenu() -> (Billboard_ContextMenu:WBP UIKit Context Menu Anchor Button Object Reference)) -> (Anchor:Menu Anchor Object Reference))
  ]}
 DoCustomBoundaryNavigation {nodes [
   n1 = DoCustomBoundaryNavigation!() -> (then, Navigation:EUINavigation Enum) {is=title K2Node_FunctionEntry}
   > n2 = Switch on EUINavigation!(execute, Selection:EUINavigation Enum=@n1) -> (Left:Exec, Right:Exec, Up:Exec, Down:Exec, Next:Exec, Previous:Exec, Invalid:Exec) {is=title K2Node_SwitchEnum}
   Call On Navigate to New Outer Row!(
    execute:Exec=[@n2.Down, @n2.Up]
    self:WBP Discovery IP Tile Billboard Object Reference
    IsDirectionUp:bool="Equal (Enum)"(A:EUINavigation Enum=@n1, B:EUINavigation Enum=Up)
    PreviousSubrowSelectedIndex:int
    PreviousNumItemsPerSubrow:int=1
    PreviousRowWasViewAllTileSelected:bool
   ) -> (then:Exec) {is=title K2Node_CallDelegate}
   > Return Node!(execute, NewParam:Widget Object Reference)
  ]}
 EventGraph {nodes [
   "On Clicked (Billboard_Button)"!() -> (OutputDelegate, then, Button:Common Button Base Object Reference) {is=title K2Node_ComponentBoundEvent}
   n1 = Get FortDiscoverSurfaceTileItemVM() -> (FortDiscoverSurfaceTileItemVM:Fort Discover Surface Tile Item VM Object Reference)
   n2 = GetActivityTimeLockedReason!(self:Fort Discover Link Entry Tile Item VM Object Reference=@n1) -> (ReturnValue:EFortActivityIsLockedReason Enum=None)
   > n3 = Branch(Condition=NOR Boolean!(
     A:bool=IsActivityTimeLocked!(self:Fort Discover Link Entry Tile Item VM Object Reference=@n1) -> (ReturnValue:bool)
     B:bool="Equal (Enum)"(A:EFortActivityIsLockedReason Enum=@n2, B:EFortActivityIsLockedReason Enum=AgeRating)
    ) -> (ReturnValue:bool))
   > HandleCategoryTileClicked!(
    execute
    self:Fort Activity Selector Object Reference=Get FortActivitySelector() -> (ReturnValue:Fort Activity Selector Object Reference) {is=title K2Node_GetSubsystem}
    FortPC:Fort Player Controller Object Reference=GetFirstLocalFortPlayerController!(bIncludeNonSpectators:bool=true, bIncludeSpectators:bool=false) -> (ReturnValue:Fort Player Controller Object Reference)
    InTileItemVM:Fort Discover Surface Tile Item VM Object Reference=Get FortDiscoverSurfaceTileItemVM()
   ) -> (then)
   > n4 = PushPrimaryContentClassAsync!(
    execute:Exec
    WidgetClass:Common Activatable Widget Class Reference=/DiscoveryBrowser/WBP_Discover_NestedSurface.WBP_Discover_NestedSurface_C
   ) -> (then:Exec, OnPrimaryContentPushed:Exec, ActivatableWidget:Common Activatable Widget Object Reference) {is=title K2Node_AsyncAction}
   n5 = Cast To WBP_Discover_NestedSurface!(execute=@n4.OnPrimaryContentPushed, Object=@n4) -> (then, CastFailed, AsWBP Discover Nested Surface:WBP Discover Nested Surface Object Reference) {is=title K2Node_DynamicCast}
   n6 = Get FortDiscoverSurfaceTileItemVM()
   > SetFortDiscoverSurfaceTileItemVM!(execute, self:WBP Discover Nested Surface Object Reference=@n5, Viewmodel:Fort Discover Surface Tile Item VM Object Reference=@n6) -> (then)
   > On Disover Surface Tile Item VM Set(self:WBP Discover Nested Surface Object Reference=@n5)
   n7 = Get FortDiscoverSurfaceViewModel!(self:WBP Discover Nested Surface Object Reference=@n5) -> (FortDiscoverSurfaceViewModel:Discover Surface View Model Object Reference)
   > SetParentItem!(execute, self:Discover Surface View Model Object Reference=@n7, InParentItem:Fort Discover Surface Tile Item VM Object Reference=@n6) -> (then)
   > SetSurfaceName!(
    execute
    self:Discover Surface View Model Object Reference=@n7
    InSurfaceName:Name=GetReferenceSurfaceName!(self:Fort Discover Surface Tile Item VM Object Reference=@n6) -> (ReturnValue:Name=None)
   ) -> (then)
   n8 = PushPrimaryContentClassAsync(execute=@n3.else, WidgetClass=/DiscoveryBrowser/ActivityVideoScreen.ActivityVideoScreen_C)
   n9 = Cast To ActivityVideoScreen!(execute=@n8.OnPrimaryContentPushed, Object=@n8) -> (then, CastFailed, AsActivity Video Screen:Activity Video Screen Object Reference) {is=title K2Node_DynamicCast}
   > PlayVideo_LinkEntryVM!(
    execute
    self:Activity Video Screen Object Reference=@n9
    LinkEntryViewModel:Fort Discover Link Entry Tile Item VM Object Reference=Get FortDiscoverSurfaceTileItemVM()
   ) -> (then)
   "On Context Menu Opened (Billboard_ContextMenu)"() {is=title K2Node_ComponentBoundEvent}
   > UpdateActivitySelector(self:Fort Activity Tile Details Display Object Reference)
   n10 = OnDiscoverSurfaceTileIPViewModelSet!() -> (OutputDelegate, then, SurfaceTileViewModel:Fort Discover Surface Tile Item VM Object Reference) {is=title K2Node_CustomEvent}
   > Set FortDiscoverSurfaceTileItemVM(FortDiscoverSurfaceTileItemVM:Fort Discover Surface Tile Item VM Object Reference=@n10.SurfaceTileViewModel) -> (Output_Get:Fort Discover Surface Tile Item VM Object Reference)
   > OnDiscoverSurfaceViewModelSetup()
   n11 = Event PreConstruct()
   > Parent: PreConstruct(IsDesignTime=@n11.IsDesignTime)
   > SetStyle!(
    execute
    self:Common Button Base Object Reference
    InStyle:Common Button Style Class Reference=/FortUILibrary/UIKit/Buttons/Blocks/ButtonStyle_UIKit_Empty.ButtonStyle_UIKit_Empty_C
   ) -> (then)
   > n12 = GetDynamicMaterial(self=Get Billboard_Background_Image() -> (Billboard_Background_Image:Image Object Reference))
   > Set BackgroundDynamicMaterial(BackgroundDynamicMaterial:Material Instance Dynamic Object Reference=@n12) -> (Output_Get:Material Instance Dynamic Object Reference)
   > n13 = GetDynamicMaterial(self=Get Billboard_Foreground_Image() -> (Billboard_Foreground_Image:Image Object Reference))
   > Set ForegroundDynamicMaterial(ForegroundDynamicMaterial:Material Instance Dynamic Object Reference=@n13) -> (Output_Get:Material Instance Dynamic Object Reference)
   > SetHideInputAction!(
    execute
    self:Common Button Base Object Reference=Get Billboard_Button() -> (Billboard_Button:WBP UIKit Button Loud Object Reference)
    bInHideInputAction:bool=true
   ) -> (then)
   n14 = Get Block_Background!(self:WBP UIKit Button Generic Base Object Reference=Get ContextMenuButton!(self:WBP UIKit Context Menu Anchor Button Base Object Reference=Get Billboard_ContextMenu()) -> (ContextMenuButton:WBP UIKit Button Generic Base Object Reference)) -> (Block_Background:WBP UIKit Block Background Base Object Reference)
   n15 = Get BackgroundDynamicMaterial!(self:WBP UIKit Block Background Base Object Reference=@n14) -> (BackgroundDynamicMaterial:Material Instance Dynamic Object Reference)
   > SetScalarParameterValue(self=@n15, ParameterName=Corner Radius, Value=0.3)
   n16 = Get OutlineDynamicMaterial!(self:WBP UIKit Block Background Base Object Reference=@n14) -> (OutlineDynamicMaterial:Material Instance Dynamic Object Reference)
   > SetScalarParameterValue(self=[@n15, @n16], ParameterName=Radius In Pixels, Value=0)
   > SetVectorParameterValue!(execute, self=@n16, ParameterName=Corner Radius, Value_R:float=0.325, Value_G:float=0.325, Value_B:float=0.325, Value_A:float=0.325) -> (then)
   n17 = Get Billboard_Highlight() -> (Billboard_Highlight:WBP UIKit Highlight Object Reference)
   > Set with Broadcast HighlightMatchFont!(
    execute
    HighlightMatchFont:Slate Font Info Structure=Get Font!(self:Text Object Reference=Get Text_Highlight!(self:WBP UIKit Highlight Base Object Reference=@n17) -> (Text_Highlight:Common Text Object Reference)) -> (Font:Slate Font Info Structure)
   ) -> (then, Output_Get:Slate Font Info Structure)
   > SetVisibility(self=[Get Billboard_ContextMenu(), Get Billboard_CCU() -> (Billboard_CCU:WBP UIKit Badge Object Reference)], InVisibility=Collapsed)
   > n18 = Branch Natural Nav Rollback Cvar!(In:Exec) -> (Use Natural Nav:Exec, Rollback:Exec) {is=title K2Node_MacroInstance}
   SetNavigationRuleCustomBoundary!(
    execute=@n18.Rollback
    self:Widget Object Reference=[
     Get Billboard_Button()
     Get Billboard_Countdown() -> (Billboard_Countdown:WBP UIKit Badge Object Reference)
     Get GP_Root() -> (GP_Root:Grid Panel Object Reference)
    ]
    Direction:EUINavigation Enum=Up
    InCustomDelegate:Delegate=Create Event()
   ) -> (then)
   > SetNavigationRuleCustomBoundary(self=[Get Billboard_Button(), Get Billboard_ContextMenu(), Get GP_Root()], Direction=Down, InCustomDelegate=Create Event())
   Event Construct()
   n19 = Get Billboard_Highlight()
   > Set with Broadcast HighlightMatchFont(HighlightMatchFont=Get Font(self=Get Text_Highlight(self=@n19)))
   Event On Focused() {is=title K2Node_Event}
   > Parent: On Focused() {is=title K2Node_CallParentFunction}
   n20 = Get ForegroundDynamicMaterial() -> (ForegroundDynamicMaterial:Material Instance Dynamic Object Reference)
   > n21 = SetScalarParameterValue(self=@n20, ParameterName=Is Focused, Value=1)
   n22 = Get FortDiscoverSurfaceTileItemVM()
   StartActiveImpression(execute=[@n25.then, @n21.then], self:Discover Tile Item View Model Object Reference=@n22)
   Event On Unfocused() {is=title K2Node_Event}
   > Parent: On Unfocused() {is=title K2Node_CallParentFunction}
   > n23 = SetScalarParameterValue(self=@n20, ParameterName=Is Focused, Value=0)
   StopActiveImpression(execute=[@n23.then, @n26.then], self:Discover Tile Item View Model Object Reference=@n22)
   Event On Hovered()
   > Parent: On Hovered() {is=title K2Node_CallParentFunction}
   n24 = Get ForegroundDynamicMaterial()
   > n25 = SetScalarParameterValue(self=@n24, ParameterName=Is Hovered, Value=1)
   Event On Unhovered()
   > Parent: On Unhovered() {is=title K2Node_CallParentFunction}
   > n26 = SetScalarParameterValue(self=@n24, ParameterName=Is Hovered, Value=0)
   Event OnInitialized()
   > SetMobileStyle() {is=title K2Node_CallFunction}
   n27 = Event OnMouseEnter()
   > Parent: OnMouseEnter!(execute, MyGeometry:Geometry Structure=@n27.MyGeometry, MouseEvent:"Pointer Event Structure (by ref)"=@n27.MouseEvent) -> (then) {is=title K2Node_CallParentFunction}
   n28 = ScreenToLocal!(Geometry:"Geometry Structure (by ref)"=@n27.MyGeometry, ScreenPosition:Vector 2D Structure, bIncludeWindowPosition:bool=false) -> (LocalCoordinate_X:double, LocalCoordinate_Y:double)
   n29 = Get ForegroundDynamicMaterial()
   > SetScalarParameterValue(self=@n29, ParameterName=ScreenPosX, Value=@n28)
   > SetScalarParameterValue(self=@n29, ParameterName=ScreenPosY, Value=@n28.LocalCoordinate_Y)
  ]}
 GetASpotOverrideText {nodes [
   n1 = GetASpotOverrideText!() -> (then, DefaultText:Text, ASpotOverrideText:Text) {is=title K2Node_FunctionEntry}
   > Return Node!(execute, Text:Text=SelectText(A=@n1, B=@n1.ASpotOverrideText, bPickA=TextIsEmpty(InText=@n1.ASpotOverrideText)))
  ]}
 GetBackgroundImageLoadingState_SurfaceVM {nodes [
   n1 = GetBackgroundImageLoadingState_SurfaceVM!() -> (then, BackgroundImageLoadingState:EActivityImageLoadingState Enum=Unset, BackgroundTexture:Texture Object Reference) {is=title K2Node_FunctionEntry}
   > n2 = Switch on EActivityImageLoadingState!(execute, Selection:EActivityImageLoadingState Enum=@n1) -> (Unset:Exec, Loading:Exec, Loaded:Exec, Error:Exec) {is=title K2Node_SwitchEnum}
   Return Node!(execute=[@n2.Error, @n2.Unset], Texture:Texture Object Reference=/Game/UI/DiscoveryShared/Art/T_UI_Discovery_DefaultEmpty)
   Return Node(execute=@n2.Loaded, Texture=@n1.BackgroundTexture)
  ]}
 GetBillboardButtonText {nodes [
   n1 = GetBillboardButtonText!() -> (then, HasVideo:bool=~, ActivityLockedReason:EFortActivityIsLockedReason Enum=None, IsActivityTimeLocked:bool=~) {is=title K2Node_FunctionEntry}
   n2 = "Equal (Enum)"(A=@n1.ActivityLockedReason, B=AgeRating)
   > Return Node!(execute, Text:Text=SelectText(A="LOCTEXT(\"Play\")", B="LOCTEXT(\"View Trailer\")", bPickA=NOR Boolean(A=@n2, B=@n1.IsActivityTimeLocked)))
  ]}
 GetCCUText {nodes [
   n1 = GetCCUText!() -> (then, CCUCount:int) {is=title K2Node_FunctionEntry}
   > Return Node(Text=SelectText(
     A=GetShorthandQuantity!(Quantity:int=@n1, MinimumFractionalDigits:int=0, MaximumFractionalDigits:int=1, bShouldRoundDown:bool=false) -> (ReturnValue:Text)
     B="To Text (Integer)"(Value=@n1)
     bPickA="integer > integer"(A=@n1, B=999)
    ))
  ]}
 GetCCUVisibility {nodes [
   n1 = GetCCUVisibility!() -> (then, CCUCount:int, TimerState:ECountdownTimerState Enum=None) {is=title K2Node_FunctionEntry}
   n2 = "integer <= integer"!(A:int=@n1, B:int=0) -> (ReturnValue:bool) {is=title K2Node_CallFunction}
   n3 = "Not Equal (Enum)"!(A:ECountdownTimerState Enum=@n1.TimerState, B:ECountdownTimerState Enum=Live) -> (ReturnValue:bool)
   > Return Node!(execute, Visibility:ESlateVisibility Enum=Select(Option 0=HitTestInvisible, Option 1=Collapsed, Index=OR Boolean(A=@n2, B=@n3)))
  ]}
 GetCountdownText {nodes [
   n1 = GetCountdownText!() -> (then, Time Remaining:Text) {is=title K2Node_FunctionEntry}
   > Return Node!(execute, ReturnValue:Text=TextToUpper!(InText:"Text (by ref)"=@n1) -> (ReturnValue:Text))
  ]}
 GetIsLiveVisible {nodes [
   n1 = GetIsLiveVisible!() -> (then, Timer State:ECountdownTimerState Enum=None) {is=title K2Node_FunctionEntry}
   > Return Node!(execute, IsLiveVisible:bool="Equal (Enum)"(A:ECountdownTimerState Enum=@n1, B:ECountdownTimerState Enum=Live))
  ]}
 GetPopOutImageLoadingState_SurfaceVM {nodes [
   n1 = GetPopOutImageLoadingState_SurfaceVM!() -> (then, PopOutImageLoadingState:EActivityImageLoadingState Enum=Unset, PopOutTexture:Texture Object Reference) {is=title K2Node_FunctionEntry}
   > n2 = Switch on EActivityImageLoadingState(Selection=@n1)
   Return Node!(execute=[@n2.Error, @n2.Unset], Texture:Texture Object Reference=&T_UI_Discovery_DefaultEmpty)
   Return Node(execute=@n2.Loaded, Texture=@n1.PopOutTexture)
  ]}
 OnDiscoverSurfaceViewModelSetup {nodes [
   OnDiscoverSurfaceViewModelSetup() {is=title K2Node_FunctionEntry}
   n1 = Get FortDiscoverSurfaceTileItemVM()
   > SetFortDiscoverSurfaceTileItemVM(self:Self Object Reference, Viewmodel=@n1)
   > SetupTimer!(execute, self, InFortDiscoverLinkEntryTileVM:Fort Discover Link Entry Tile Item VM Object Reference=@n1) -> (then)
   > ShowActivityDataFromViewModel!(
    execute
    self:Fort Squeegee Widget Injection Slot Object Reference=Get SqueegeeInjectionSlot() -> (SqueegeeInjectionSlot:Squeegee Injection Slot Object Reference)
    GameActivityViewModel:Fort Discover Link Entry Tile Item VM Object Reference=@n1
   ) -> (then)
   > StartVisibilityImpression(self:Discover Tile Item View Model Object Reference=@n1)
   > n2 = Branch(Condition=HasFetchedASpotImages!(self:Fort Discover Link Entry Tile Item VM Object Reference=Get FortDiscoverSurfaceTileItemVM()) -> (ReturnValue:bool))
   FetchASpotImages(execute=@n2.else, self:Fort Discover Link Entry Tile Item VM Object Reference=Get FortDiscoverSurfaceTileItemVM())
  ]}
 OnNavigateToNewOuterRow {nodes [OnNavigateToNewOuterRow!() -> (then, IsDirectionUp:bool=~, PreviousSubrowSelectedIndex:int, PreviousNumItemsPerSubrow:int, PreviousRowWasViewAllTileSelected:bool=~) {is=title K2Node_FunctionEntry}]}
 SetBackgroundImageTexture {nodes [
   n1 = SetBackgroundImageTexture!() -> (then, BackgroundImageTexture:Texture Object Reference) {is=title K2Node_FunctionEntry}
   > SetTextureParameterValue(self=Get BackgroundDynamicMaterial!() -> (BackgroundDynamicMaterial), ParameterName=Background Image, Value=@n1)
  ]}
 SetContextMenuFlowRiderTelemetry {nodes [
   n1 = SetContextMenuFlowRiderTelemetry!() -> (then, LinkID:Online Link Id Structure) {is=title K2Node_FunctionEntry}
   > n2 = LinkIdToString!(execute, LinkString:"Online Link Id Structure (by ref)"=@n1) -> (then, ReturnValue:String)
   > SetParentIDForContextMenuTelemetry!(
    execute
    self:WBP UIKit Context Menu Anchor Button Base Object Reference=Get Billboard_ContextMenu()
    ParentID:String=Append(A=Island ID:, B=@n2)
    IsParentIdEGPI:bool=false
   ) -> (then)
  ]}
 SetCountdownAndLiveVisibility {nodes [
   n1 = SetCountdownAndLiveVisibility!() -> (then, IsLiveVisible:bool=~) {is=title K2Node_FunctionEntry}
   n2 = Select(Option 0=Collapsed, Option 1=HitTestInvisible, Index=@n1)
   > SetVisibility(self=Get Billboard_Highlight(), InVisibility=@n2)
   > SetVisibility(self=Get Billboard_Background_LiveEffect() -> (Billboard_Background_LiveEffect:Image Object Reference), InVisibility=@n2)
   n3 = "Equal (Enum)"(
    A=Get TimerState!(self:Countdown Timer View Model Object Reference=Get FortCountdownTimerVM() -> (FortCountdownTimerVM:Countdown Timer View Model Object Reference)) -> (TimerState:ECountdownTimerState Enum=None)
    B=None
   )
   > SetVisibility(self=Get Billboard_Countdown(), InVisibility=Select(Option 0=HitTestInvisible, Option 1=Collapsed, Index=OR Boolean(A=@n1, B=@n3)))
  ]}
 SetMobileStyle {nodes [
   SetMobileStyle() {is=title K2Node_FunctionEntry}
   > Branch(Condition=IsUsingCompactUI())
   > Set Min Width!(execute, Min Width:int=90, self:WBP UIKit Context Menu Anchor Button Base Object Reference=Get Billboard_ContextMenu()) -> (then, Output_Get:int=0)
   > Branch(Condition=IsTabletAspectRatio!() -> (ReturnValue:bool))
   n1 = Get MinSize() -> (MinSize:Spacer Object Reference)
   > SetSize!(
    execute
    self:Spacer Object Reference=@n1
    InSize_X:double=1628
    InSize_Y:double=Get!(self:Spacer Object Reference=@n1) -> (Size_X:double, Size_Y:double).Size_Y
   ) -> (then)
  ]}
 SetPopOutImageTexture {nodes [
   n1 = SetPopOutImageTexture!() -> (then, PopOutImageTexture:Texture Object Reference) {is=title K2Node_FunctionEntry}
   > SetTextureParameterValue(self=Get ForegroundDynamicMaterial(), ParameterName=Foreground Image, Value=@n1)
  ]}
 SetupContextMenu {nodes [
   SetupContextMenu() {is=title K2Node_FunctionEntry}
   > n1 = Construct Object from Class(Class=/Script/FortniteUI.UIKitContextMenuActionContext) -> (ReturnValue:UIKit Context Menu Action Context Object Reference)
   > n2 = Set ContextMenuActionContext(ContextMenuActionContext:UIKit Context Menu Action Context Object Reference=@n1) -> (Output_Get:UIKit Context Menu Action Context Object Reference)
   > SetContext!(
    execute
    self:WBP UIKit Context Menu Anchor Button Base Object Reference=Get Billboard_ContextMenu()
    InteractionContext_Data:UIKit Context Menu Action Context Object Reference=@n2
   ) -> (then)
  ]}
 SetupTimer {nodes [
   n1 = SetupTimer!() -> (then, InFortDiscoverLinkEntryTileVM:Fort Discover Link Entry Tile Item VM Object Reference) {is=title K2Node_FunctionEntry}
   n2 = Get FortCountdownTimerVM()
   > SetupViewModel!(
    execute
    self:Countdown Timer View Model Object Reference=@n2
    InFutureTime:"Date Time Structure (by ref)"=Select(
     Option 0:Date Time Structure=GetNextStartTime!(self:Fort Discover Link Entry Tile Item VM Object Reference=@n1) -> (ReturnValue:Date Time Structure)
     Option 1:Date Time Structure=GetNextAspotCountdownTimerEnd!(self:Fort Discover Link Entry Tile Item VM Object Reference=@n1) -> (ReturnValue:Date Time Structure)
     Index=HasActiveASpotCountdownTimer!(self:Fort Discover Link Entry Tile Item VM Object Reference=@n1) -> (ReturnValue:bool)
    ) -> (ReturnValue:Date Time Structure)
    InLiveText:"Text (by ref)"=Get LiveText() -> (LiveText:Text)
   ) -> (then)
   > SetTimerUnitDisplayType!(execute, self:Countdown Timer View Model Object Reference=@n2, InTimerUnitDisplayType:ETimespanUnitDisplayFormat Enum=Abbreviated) -> (then)
   > SetApproximateLongTimers!(execute, self:Countdown Timer View Model Object Reference=@n2, bInApproximateLongTimers:bool=false) -> (then)
   > StartTimer(self:Countdown Timer View Model Object Reference=@n2)
  ]}
]
variables [
 BackgroundDynamicMaterial {type=MaterialInstanceDynamic}
 ContextMenuActionContext {type=UIKitContextMenuActionContext, default=null}
 Favorited_Sound {type=SoundBase, default=/Game/Sounds/Fort_UI_Sounds/FrontEnd/MultiProduct/Common/Buttons/MSS_UI_MP_Button_Click_Favorite}
 ForegroundDynamicMaterial {type=MaterialInstanceDynamic}
 HighlightMatchFont {
  type=SlateFontInfo
  default {
   bForceMonospaced=false
   bMaterialIsStencil=false
   fontMaterial=null
   fontObject=&HeadingNow
   letterSpacing=10
   monospacedWidth=1
   size=15.75
   skewAmount=0
   typefaceFontName=86 Bold Italic
   outlineSettings {
    bApplyOutlineToDropShadows=false
    bMiteredCorners=false
    bSeparateFillAlpha=false
    outlineMaterial=null
    outlineSize=0
    outlineColor=<rgba 0 0 0 1>
   }
  }
 }
 LiveText {type=string, default=Live}
 OnNavigateToNewOuterRow {}
 ParentPanelName {type=string, default=""}
 Unfavorited_Sound {type=SoundBase, default=/Game/Sounds/Fort_UI_Sounds/FrontEnd/MultiProduct/Common/Buttons/MSS_UI_MP_Button_Click_Unfavorite}
]
widgets {
 info {
  inheritedWidgetCount=8
  parentClass=/Script/DiscoveryBrowserUI.FortActivityTileDetailsDisplay
  rootWidgetClass=&GridPanel
  widgetCount=29
 }
 widgets [
  GP_Root {widgetClassPath=&GridPanel, bIsVariable=true, widget=GP_Root, properties {columnFill [1], rowFill [1]}}
  MinSize {
   widgetClassPath=&Spacer
   bIsVariable=true
   parent=GP_Root
   widget=MinSize
   properties {size=<1760 352>}
   slot {class=GridSlot, properties {layer=-10}}
  }
  ThumbnailOverlay_Video {widgetClassPath=&Overlay, parent=GP_Root, widget=ThumbnailOverlay_Video, properties {visibility=Collapsed}, slot {class=GridSlot}}
  ActivityVideoCycleWidget {
   widgetClassPath=/DiscoveryBrowser/Tiles/Islands/WBP_IslandTile_VideoCycle
   bInherited=true
   bIsVariable=true
   parent=ThumbnailOverlay_Video
   widget=ActivityVideoCycleWidget
   properties {movieRestartDelay=0.1, movieStartDelay=0.1, visibility=Hidden}
   slot {class=OverlaySlot, properties {horizontalAlignment=HAlign_Fill, verticalAlignment=VAlign_Fill}}
  }
  Billboard_Background_LiveEffect {
   widgetClassPath=&Image
   bIsVariable=true
   parent=GP_Root
   widget=Billboard_Background_LiveEffect
   properties {
    visibility=SelfHitTestInvisible
    brush {
     bIsDynamicallyLoaded=false
     drawAs=Image
     imageType=FullColor
     mirroring=NoMirror
     resourceName=None
     resourceObject=/DiscoveryBrowser/Rows/Billboard/MaterialInstances/MI_UI_LiveEvent
     tiling=NoTile
     imageSize=<32 32>
     margin=<ltrb 0 0 0 0>
     outlineSettings {
      bUseBrushTransparency=false
      roundingType=HalfHeightRadius
      width=0
      color {colorUseRule=UseColor_Specified, specifiedColor=<rgba 0 0 0 0>}
      cornerRadii=<0 0 0 0>
     }
     tintColor {colorUseRule=UseColor_Specified, specifiedColor=<rgba 1 1 1 1>}
     uVRegion {bIsValid=false, max=<0 0>, min=<0 0>}
    }
   }
   slot {class=GridSlot, properties {layer=2}}
  }
  Billboard_Background_Image {
   widgetClassPath=&Image
   bIsVariable=true
   parent=GP_Root
   widget=Billboard_Background_Image
   properties {
    visibility=SelfHitTestInvisible
    brush {
     bIsDynamicallyLoaded=false
     drawAs=Image
     imageType=FullColor
     mirroring=NoMirror
     resourceName=None
     resourceObject=/DiscoveryBrowser/Rows/Billboard/MaterialInstances/MI_UI_Discovery_Row_Billboard_Background
     tiling=NoTile
     imageSize=<32 32>
     margin=<ltrb 0 0 0 0>
     outlineSettings {
      bUseBrushTransparency=false
      roundingType=HalfHeightRadius
      width=0
      color {colorUseRule=UseColor_Specified, specifiedColor=<rgba 0 0 0 0>}
      cornerRadii=<0 0 0 0>
     }
     tintColor {colorUseRule=UseColor_Specified, specifiedColor=<rgba 1 1 1 1>}
     uVRegion {bIsValid=false, max=<0 0>, min=<0 0>}
    }
   }
   slot {class=GridSlot}
  }
  Billboard_Foreground_Image {
   widgetClassPath=&Image
   bIsVariable=true
   parent=GP_Root
   widget=Billboard_Foreground_Image
   properties {
    visibility=SelfHitTestInvisible
    brush {
     bIsDynamicallyLoaded=false
     drawAs=Image
     imageType=FullColor
     mirroring=NoMirror
     resourceName=None
     resourceObject=/DiscoveryBrowser/Rows/Billboard/MaterialInstances/MI_UI_Discovery_Row_Billboard_Foreground
     tiling=NoTile
     imageSize=<32 32>
     margin=<ltrb 0 0 0 0>
     outlineSettings {
      bUseBrushTransparency=false
      roundingType=HalfHeightRadius
      width=0
      color {colorUseRule=UseColor_Specified, specifiedColor=<rgba 0 0 0 0>}
      cornerRadii=<0 0 0 0>
     }
     tintColor {colorUseRule=UseColor_Specified, specifiedColor=<rgba 1 1 1 1>}
     uVRegion {bIsValid=false, max=<0 0>, min=<0 0>}
    }
   }
   slot {class=GridSlot, properties {layer=1, padding=<ltrb 0 -32 0 0>}}
  }
  StackBox_0 {
   widgetClassPath=&StackBox
   parent=GP_Root
   widget=StackBox_0
   properties {orientation=Orient_Vertical}
   slot {class=GridSlot, properties {horizontalAlignment=HAlign_Left, layer=10, verticalAlignment=VAlign_Center, padding=<ltrb 64 0 0 0>}}
  }
  SIZEHACKDONOTCOPY {
   widgetClassPath=&ScaleBox
   parent=StackBox_0
   widget=SIZEHACKDONOTCOPY
   properties {stretch=UserSpecified, userSpecifiedScale=1.2}
   slot {class=StackBoxSlot, properties {horizontalAlignment=HAlign_Left, verticalAlignment=VAlign_Top, padding=<ltrb 0 0 0 4>}}
  }
  HorizontalBox_1 {widgetClassPath=&HorizontalBox, parent=SIZEHACKDONOTCOPY, widget=HorizontalBox_1, slot {class=ScaleBoxSlot}}
  Billboard_Highlight {
   widgetClassPath=/FortUILibrary/UIKit/Highlight/WBP_UIKit_Highlight
   bIsVariable=true
   parent=HorizontalBox_1
   widget=Billboard_Highlight
   properties {highlightSize=Compact, highlightText=Live}
   slot {class=HorizontalBoxSlot, properties {padding=<ltrb 0 0 8 0>}}
  }
  Billboard_Countdown {
   widgetClassPath=/FortUILibrary/UIKit/Badge/WBP_UIKit_Badge
   bIsVariable=true
   parent=HorizontalBox_1
   widget=Billboard_Countdown
   properties {
    badgeText=Badge_wrongFont
    badgeType=Text
    indicator Type=Utility
    indicatorConfig=Icon
    indicatorIcon=/FortUILibrary/Textures/MaterialInstances/M_UI_AnimatedClock
   }
   slot {class=HorizontalBoxSlot, properties {padding=<ltrb 0 0 8 0>}}
  }
  Billboard_CCU {
   widgetClassPath=&WBP_UIKit_Badge
   bIsVariable=true
   parent=HorizontalBox_1
   widget=Billboard_CCU
   properties {badgeText=264K_wrongFont_look_in_VM, badgeType=Text, indicatorConfig=None}
   slot {class=HorizontalBoxSlot}
  }
  Billboard_Title {
   widgetClassPath=&CommonTextBlock
   bIsVariable=true
   parent=StackBox_0
   widget=Billboard_Title
   properties {
    lineHeightPercentage=0.75
    text="{Title Goes Here}"
    textTransformPolicy=ToUpper
    wrapTextAt=720
    font {
     bForceMonospaced=false
     bMaterialIsStencil=false
     fontMaterial=null
     fontObject=&HeadingNow
     letterSpacing=-40
     monospacedWidth=1
     size=27
     skewAmount=0
     typefaceFontName=86 Bold Italic
     outlineSettings {
      bApplyOutlineToDropShadows=false
      bMiteredCorners=false
      bSeparateFillAlpha=false
      outlineMaterial=/FortUILibrary/MaterialInstances/Text/MI_UI_Block_Text_Outline
      outlineSize=3
      outlineColor=<rgba 0 0 0 1>
     }
    }
    shadowOffset=<0 0>
   }
   slot {class=StackBoxSlot, properties {verticalAlignment=VAlign_Top}}
  }
  Billboard_Description {
   widgetClassPath=&CommonTextBlock
   bIsVariable=true
   parent=StackBox_0
   widget=Billboard_Description
   properties {
    lineHeightPercentage=0.85
    text=Collapsed visibility default wwwwwwwwwwwwwwwwwwwwwwwwwwwwwwwwwwwwwwww
    visibility=Collapsed
    wrapTextAt=720
    font {
     bForceMonospaced=false
     bMaterialIsStencil=false
     fontMaterial=null
     fontObject=&HeadingNow
     letterSpacing=0
     monospacedWidth=1
     size=15.75
     skewAmount=0
     typefaceFontName=75 Medium
     outlineSettings {
      bApplyOutlineToDropShadows=false
      bMiteredCorners=false
      bSeparateFillAlpha=false
      outlineMaterial=&MI_UI_Block_Text_Outline
      outlineSize=3
      outlineColor=<rgba 0 0 0 1>
     }
    }
    shadowOffset=<0 0>
   }
   slot {class=StackBoxSlot, properties {verticalAlignment=VAlign_Top, padding=<ltrb 0 16 0 0>}}
  }
  O_Buttons {
   widgetClassPath=&Overlay
   parent=StackBox_0
   widget=O_Buttons
   slot {class=StackBoxSlot, properties {horizontalAlignment=HAlign_Left, verticalAlignment=VAlign_Top, padding=<ltrb 0 24 0 0>}}
  }
  WBP_CaptureForPostBufferUpdate {
   widgetClassPath=&WBP_CaptureForPostBufferUpdate
   bIsVariable=true
   parent=O_Buttons
   slot {class=OverlaySlot, properties {horizontalAlignment=HAlign_Fill, verticalAlignment=VAlign_Fill}}
  }
  GP_Buttons {
   widgetClassPath=&GridPanel
   parent=O_Buttons
   widget=GP_Buttons
   properties {bOverride_Cursor=true, columnFill [1], rowFill [1]}
   slot {class=OverlaySlot}
  }
  Billboard_Button {
   widgetClassPath=/FortUILibrary/UIKit/Buttons/WBP_UIKit_Button_Loud
   bIsVariable=true
   parent=GP_Buttons
   widget=Billboard_Button
   properties {minWidth=264, text=Join Event}
   slot {class=GridSlot, properties {padding=<ltrb 0 0 12 0>}}
  }
  Billboard_ContextMenu {
   widgetClassPath=/FortUILibrary/UIKit/ContextMenu/WBP_UIKit_ContextMenuAnchorButton
   bIsVariable=true
   parent=GP_Buttons
   widget=Billboard_ContextMenu
   properties {
    bIsFocusable=true
    buttonClass=/DiscoveryBrowser/Tiles/Islands/WBP_Discovery_ContextualButton.WBP_Discovery_ContextualButton_C
    buttonIcon=/FortUILibrary/Textures/Icons/T_UI_Icon_UIKit_ContextMenu_SDF
    buttonText=""
    min Height=56
    min Width=56
    navigation=&self:WidgetTree.Billboard_ContextMenu.WidgetNavigation_0
    useBottomBarButtons=true
    button Padding=<ltrb 0 0 0 0>
   }
   slot {class=GridSlot, properties {column=1}}
  }
  SizeBox_0 {
   widgetClassPath=&SizeBox
   parent=GP_Root
   widget=SizeBox_0
   properties {bOverride_HeightOverride=true, bOverride_WidthOverride=true, heightOverride=64, widthOverride=64}
   slot {class=GridSlot, properties {horizontalAlignment=HAlign_Right, layer=2, verticalAlignment=VAlign_Bottom, padding=<ltrb 0 0 24 24>}}
  }
  SqueegeeInjectionSlot {
   widgetClassPath=/Game/UI/Plugin/Squeegee/SqueegeeInjectionSlot
   bIsVariable=true
   parent=SizeBox_0
   widget=SqueegeeInjectionSlot
   properties {
    pluginId=Squeegee
    pluginWidgetId=RatingCardLogoBorderless
    styleParams {bShowDetails=true, fontSize=23, paddingSize=0, iconSize=<64 64>, iconSkewSize=<11 58>}
   }
   slot {class=SizeBoxSlot}
  }
  Text_DebugId {
   widgetClassPath=&TextBlock
   bInherited=true
   parent=GP_Root
   widget=Text_DebugId
   properties {
    autoWrapText=true
    justification=Center
    text=EventId: epicgames_Some_Event_Id EventWindowId: Some_event_window_id EventWindow Playlist: Playlists_name_example
    visibility=Collapsed
    wrapTextAt=400
    wrappingPolicy=AllowPerCharacterWrapping
    font {
     bForceMonospaced=false
     bMaterialIsStencil=false
     fontMaterial=null
     fontObject=&Roboto
     letterSpacing=0
     monospacedWidth=1
     size=16
     skewAmount=0
     typefaceFontName=Bold
     outlineSettings {
      bApplyOutlineToDropShadows=false
      bMiteredCorners=false
      bSeparateFillAlpha=false
      outlineMaterial=null
      outlineSize=3
      outlineColor=<rgba 0 0 0 1>
     }
    }
   }
   slot {class=GridSlot, properties {layer=20}}
  }
  Text_ActivityName {widgetClassPath=&CommonTextBlock, bInherited=true, widget=null}
  Text_PlayerCount {widgetClassPath=&CommonTextBlock, bInherited=true, widget=null}
  Button_Favorite {widgetClassPath=&CommonButtonBase, bInherited=true, widget=null}
  Button_Details {widgetClassPath=&CommonButtonBase, bInherited=true, widget=null}
  ActivityBrowserTag_EpicOriginal {widgetClassPath=/Script/FortniteUI.FortActivityBrowserTag, bInherited=true, widget=null}
  InputActionWidget {widgetClassPath=&CommonActionWidget, bInherited=true, widget=null}
 ]
}
